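# Player animation Blueprint for the CoffeePuff hookshot item, overriding Item and Locomotion anim layers on the Fortnite player skeleton.
# While firing, it applies a full-body aim offset and, in the lower body, either additive throw/hit/pull states or pelvis twist.
# Otherwise, lower-body correctives reuse PoppySoap equip and pose animations with a hip offset.
# Full locomotion adds an idle loop masked by slope damping, plus equip clavicle correction.
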AnimBlueprint /CoffeePuffGameplay/Assets/Animations/CoffeePuff_Player_AnimBP
target_skeleton=&Fortnite_M_Avg_Player_Skeleton
graphs [
 AnimGraph {nodes [Output Pose()]}
 EventGraph {links [], nodes []}
 ItemAimOffset {nodes [
   Input Pose!() -> (Pose, AimOffsetAlpha:double, Yaw:double, Pitch:double) {name=InPose_AimOffset}
   Input Pose!() -> (Pose) {name=InPose_PreAimOffsets}
   Output Pose() {layerGroup=Item Override}
  ]}
 ItemFallAdditive {nodes [
   Input Pose() {name=InPoseItemFallAdditive}
   Output Pose() {layerGroup=Item Override}
  ]}
 ItemFallLandAdditive {nodes [
   Input Pose() {name=InPoseItemFallLandAdditive}
   Output Pose() {layerGroup=Item Override}
  ]}
 ItemFallLevel2Override {nodes [
   Input Pose() {name=InPose_FallLevel2}
   Output Pose() {layerGroup=Item Override}
  ]}
 ItemFinalPoseOverride {nodes [
   Input Pose() {name=InFinalPose}
   Output Pose() {layerGroup=Item Override}
  ]}
 ItemFullBodyAimOffsets {nodes [
   n1 = Input Pose!() -> (Pose, FullBody_AimOffsetAlpha:double, Aim_Yaw:double, Aim_Pitch:double) {name=InPoseFullBodyAimOffsets}
   n2 = AimOffset Player 'CoffeePuff_Player_Hookshot_AO_CMM'!(X:float=@n1.Aim_Yaw, Y:float=@n1.Aim_Pitch, BasePose:Pose Link Structure=@n1, Alpha:float=@n1.FullBody_AimOffsetAlpha) -> (Pose:Pose Link Structure) {
    blendSpace=/Game/Animation/Game/MainPlayer/Poses/Medium/Male/CoffeePuff/AO/CoffeePuff_Player_Hookshot_AO_CMM
    alphaBoolBlend {blendOption=Linear}
   }
   Output Pose(Result=Blend Poses by bool(BlendPose_0=@n2, BlendPose_1=Input Pose!() -> (Pose) {name=InPosePreFullBodyAimOffsets}, bActiveValue=Get bIsFiring() -> (bIsFiring:bool)) {blendType=Linear}) {layerGroup=Item Override}
  ]}
 ItemFullBodyOverride {nodes [
   n1 = "Transform (Modify) Bone - Bone: pelvis"!(
    ComponentPose=Local To Component(LocalPose=Use cached pose 'Item_Fullbody'())
    Alpha=1
    Translation="0,0,0"
    Rotation_Roll:float=0
    Rotation_Pitch:float=0
    Rotation_Yaw:float=Get CustomPelvisTwist() -> (CustomPelvisTwist:double)
    Scale="1,1,1"
   ) -> (Pose) {
    rotationMode=BMM_Additive
    translationMode=BMM_Additive
    alphaBoolBlend {blendOption=Linear}
    boneToModify {boneName=pelvis}
    translation {z=-0.6}
   }
   Output Pose(Result=Blend Poses by bool(
     BlendPose_0=Component To Local(ComponentPose=@n1)
     BlendPose_1=Use cached pose 'Item_Fullbody'()
     bActiveValue=Get bIsInVehicle() -> (bIsInVehicle:bool)
    ) {blendType=Linear}) {layerGroup=Item Override}
   Save cached pose 'Item_Fullbody'(Pose=Input Pose() {name=InPoseFullBody}) {cacheName=Item_Fullbody}
  ]}
 ItemIdleOverride {nodes [
   Input Pose() {name=InPose_Idle}
   Output Pose() {layerGroup=Item Override}
  ]}
 ItemJetPackJumpAdditive {nodes [Output Pose() {layerGroup=Item Override}]}
 ItemJetPackStartAdditive {nodes [Output Pose() {layerGroup=Item Override}]}
 ItemJumpLoopAdditive {nodes [
   Input Pose() {name=InPoseItemJumpLoopAdditive}
   Output Pose() {layerGroup=Item Override}
  ]}
 ItemJumpOverride {nodes [
   Input Pose() {name=InPoseJump}
   Output Pose() {layerGroup=Item Override}
  ]}
 ItemJumpUpAdditive {nodes [
   Input Pose() {name=InPoseItemJumpUpAdditive}
   Output Pose() {layerGroup=Item Override}
  ]}
 ItemPreAimOffsets {nodes [
   Input Pose() {name=InPosePreAimOffsets}
   Output Pose() {layerGroup=Item Override}
  ]}
 ItemPreIK {nodes [
   Input Pose() {name=InPose_PreIK}
   Output Pose() {layerGroup=Item Override}
  ]}
 ItemRightHandIK {nodes [
   Input Pose() {name=InPosePreRightHandIK}
   Input Pose() {name=InPosePostRightHandIK}
   Output Pose() {layerGroup=Item Override}
  ]}
 ItemSkeletalControl {nodes [
   Input Pose()
   Output Pose() {layerGroup=Item Override}
  ]}
 ItemSwimJumpFallLoop {nodes [Output Pose() {layerGroup=Item Override}]}
 ItemSwimJumpSurfaceEnd {nodes [Output Pose() {layerGroup=Item Override}]}
 ItemSwimJumpSurfaceLoop {nodes [Output Pose() {layerGroup=Item Override}]}
 ItemUpperBody {nodes [
   Input Pose!() -> (Pose, InputParam:Fort Anim Input Adjusted Aim Structure) {name=InPoseUpperBody}
   Output Pose() {layerGroup=Item Override}
  ]}
 ItemZipLineStartAdditive {nodes [Output Pose() {layerGroup=Item Override}]}
 LocomotionLayer_Additives {nodes [Output Pose() {layerGroup=LocomotionOverride}]}
 LocomotionLayer_BaseLocomotion {nodes [
   Input Pose!() -> (Pose) {name=InPose_BaseLocomotion}
   Output Pose() {layerGroup=LocomotionOverride}
  ]}
 LocomotionLayer_FinalPose {nodes [
   Input Pose() {name=InPoseFinalPose}
   Output Pose() {layerGroup=LocomotionOverride}
  ]}
 LocomotionLayer_FullBody {nodes [
   Input Pose() {name=InPose_FullBody}
   Output Pose() {layerGroup=LocomotionOverride}
  ]}
 LocomotionLayer_FullLocomotion {nodes [
   Output Pose(Result=Blend Poses by bool(
     BlendPose_0=Use cached pose 'UpperBodyIdlePose'()
     BlendPose_1=Use cached pose 'LocoLayer_FullLoco'()
     BlendTime_0=0.4
     BlendTime_1=0.3
     bActiveValue=Get bAdditiveLoopBlend() -> (bAdditiveLoopBlend:bool)
    ) {blendType=CubicInOut}) {layerGroup=LocomotionOverride}
   Save cached pose 'LocoLayer_FullLoco'(Pose=Input Pose() {name=InPose_FullLocomotion}) {cacheName=LocoLayer_FullLoco, cachePoseName=FullBody}
   n1 = Layered blend per bone(
    BasePose="Sequence Player (additive) 'PoppySoap_Idle_Loop'"() {sequence=/PoppySoap/Assets/Animation/PoppySoap_Idle_Loop}
    BlendPoses_0=Additive Identity Pose()
    BlendWeights_0=Get LowerBodyAdditiveLoopSlopeDampFactor() -> (LowerBodyAdditiveLoopSlopeDampFactor:double)
   ) {
    bMeshSpaceRotationBlend=true
    blendMode=BlendMask
    blendMasks [&Fortnite_M_Avg_Player_Skeleton:LowerBody]
    layerSetup []
   }
   AnimGraphNode_StateMachine_1 = CorrectivePoseBlend() {is=title AnimGraphNode_StateMachine, maxTransitionsPerFrame=1}
   Save cached pose 'UpperBodyIdlePose'(Pose=Apply Additive(Base=@AnimGraphNode_StateMachine_1, Additive=@n1, Alpha=Get IdleAdditiveAlpha() -> (IdleAdditiveAlpha:double)) {alphaBoolBlend {blendOption=Linear}}) {cacheName=UpperBodyIdlePose}
  ]}
 CorrectivePoseBlend {
  entry=PassThrough
  path=LocomotionLayer_FullLocomotion.AnimGraphNode_StateMachine_1.CorrectivePoseBlend
  ids=id
  nodes [
   AnimStateNode_0 = PassThrough() {
    is=title AnimStateNode
    stateEntered {guid=#1}
    stateFullyBlended {guid=#2}
    stateLeft {guid=#3}
   }
   AnimStateNode_2 = EquipCorrective() {
    is=title AnimStateNode
    stateEntered {guid=#4}
    stateFullyBlended {guid=#5}
    stateLeft {guid=#6}
   }
   AnimStateTransitionNode_11 = AnimStateTransitionNode() {blendMode=CubicInOut, crossfadeDuration=0.4}
   AnimStateTransitionNode_3 = AnimStateTransitionNode() {
    transitionEnd {guid=#7}
    transitionInterrupt {guid=#8}
    transitionStart {guid=#9}
   }
  ]
  links [
   {
    dst=PassThrough
    rule=LocomotionLayer_FullLocomotion.AnimGraphNode_StateMachine_1.CorrectivePoseBlend.AnimStateTransitionNode_11.Transition
    src=EquipCorrective
    via=AnimStateTransitionNode_11
   }
   {
    dst=EquipCorrective
    rule=LocomotionLayer_FullLocomotion.AnimGraphNode_StateMachine_1.CorrectivePoseBlend.AnimStateTransitionNode_3.Transition
    src=PassThrough
    via=AnimStateTransitionNode_3
   }
  ]
 }
 PassThrough {
  path=LocomotionLayer_FullLocomotion.AnimGraphNode_StateMachine_1.CorrectivePoseBlend.AnimStateNode_0.PassThrough
  nodes [Output Animation Pose(Result=Use cached pose 'LocoLayer_FullLoco'())]
 }
 EquipCorrective {
  path=LocomotionLayer_FullLocomotion.AnimGraphNode_StateMachine_1.CorrectivePoseBlend.AnimStateNode_2.EquipCorrective
  nodes [Output Animation Pose(Result=Layered blend per bone(
     BasePose=Use cached pose 'LocoLayer_FullLoco'()
     BlendPoses_0=Sequence Player 'PoppySoap_Equip_MeshSpace'() {bLoopAnimation=false, sequence=/PoppySoap/Assets/Animation/PoppySoap_Equip_MeshSpace}
     BlendWeights_0=Get EquipBlendOut() -> (EquipBlendOut:float)
    ) {layerSetup [{branchFilters [{boneName=clavicle_l}, {boneName=clavicle_r}]}]})]
 }
 Transition {
  path=LocomotionLayer_FullLocomotion.AnimGraphNode_StateMachine_1.CorrectivePoseBlend.AnimStateTransitionNode_11.Transition
  nodes [Result(bCanEnterTransition=Get bEquipCorrectiveToPassThrough() -> (bEquipCorrectiveToPassThrough:bool))]
 }
 Transition {
  path=LocomotionLayer_FullLocomotion.AnimGraphNode_StateMachine_1.CorrectivePoseBlend.AnimStateTransitionNode_3.Transition
  nodes [Result(bCanEnterTransition=Get bPassThroughToEquipCorrective() -> (bPassThroughToEquipCorrective:bool))]
 }
 LocomotionLayer_GroundLandAdditive {nodes [
   Input Pose()
   Output Pose() {layerGroup=LocomotionOverride}
  ]}
 LocomotionLayer_InAir {nodes [
   Input Pose() {name=InPose_InAir}
   Input Pose() {name=InPose_InAirSourcePose}
   Output Pose() {layerGroup=LocomotionOverride}
  ]}
 LocomotionLayer_LowerBody {nodes [
   AnimGraphNode_StateMachine_0 = HookShotStates() {is=title AnimGraphNode_StateMachine}
   n1 = Apply Additive(Base=Use cached pose 'LocoLayer_LowerBody'(), Additive=@AnimGraphNode_StateMachine_0, Alpha=Get IdleAdditiveAlpha()) {alphaBoolBlend {blendOption=Linear}}
   n2 = "Transform (Modify) Bone - Bone: pelvis"(
    ComponentPose=Local To Component(LocalPose=Use cached pose 'LocoLayer_LowerBody'())
    Alpha=1
    Translation="0,0,0"
    Rotation_Roll=0
    Rotation_Pitch=0
    Rotation_Yaw=Get CustomPelvisTwist()
    Scale="1,1,1"
   ) {
    rotationMode=BMM_Additive
    translationMode=BMM_Additive
    alphaBoolBlend {blendOption=Linear}
    boneToModify {boneName=pelvis}
    translation {z=-0.6}
   }
   n3 = Component To Local(ComponentPose=@n2)
   n4 = Get bIsStoppingOrIdle() -> (bIsStoppingOrIdle:bool)
   n5 = Get bIsFiring()
   Output Pose(Result=Blend Poses by bool(
     BlendPose_0=Use cached pose 'CorrectedPose'()
     BlendPose_1=Blend Poses by bool(BlendPose_0=@n1, BlendPose_1=@n3, BlendTime_0=0.5, BlendTime_1=0.5, bActiveValue=@n4) {blendType=CubicInOut}
     BlendTime_0=0.2
     BlendTime_1=0.3
     bActiveValue=NOT Boolean(A=@n5)
    ) {blendType=CircularOut, childUpateMode=AlwaysTickChildren}) {layerGroup=LocomotionOverride}
   AnimGraphNode_StateMachine_1 = CorrectivePoseBlend() {maxTransitionsPerFrame=1}
   n6 = "Transform (Modify) Bone - Bone: pelvis"!(
    ComponentPose=Local To Component(LocalPose=@AnimGraphNode_StateMachine_1)
    Alpha=1
    Translation_X:double=0
    Translation_Y:double=0
    Translation_Z:double=Get HipOffsetVertical() -> (HipOffsetVertical:double)
    Rotation:Rotator="0,0,0"
    Scale="1,1,1"
   ) -> (Pose) {
    rotationMode=BMM_Additive
    translationMode=BMM_Additive
    alphaBoolBlend {blendOption=Linear}
    boneToModify {boneName=pelvis}
    translation {z=-0.6}
   }
   Save cached pose 'CorrectedPose'(Pose=Blend Poses by bool(
     BlendPose_0=Component To Local(ComponentPose=@n6)
     BlendPose_1=Use cached pose 'LocoLayer_LowerBody'()
     BlendTime_0=0.5
     BlendTime_1=0.4
     bActiveValue=Get bApplyLowerBodyCorrective() -> (bApplyLowerBodyCorrective:bool)
    ) {blendType=CubicInOut}) {cacheName=CorrectedPose}
   Save cached pose 'LocoLayer_LowerBody'(Pose=Input Pose() {name=InPose_LowerBody}) {cacheName=LocoLayer_LowerBody}
  ]}
 HookShotStates {
  entry=Passthrough
  path=LocomotionLayer_LowerBody.AnimGraphNode_StateMachine_0.HookShotStates
  ids=id
  nodes [
   AnimStateNode_0 = HookShot_Throw() {is=title AnimStateNode}
   AnimStateNode_1 = Hit_Success() {is=title AnimStateNode}
   AnimStateNode_2 = Pull_Success() {is=title AnimStateNode}
   AnimStateNode_3 = Pull_Fail() {is=title AnimStateNode}
   AnimStateNode_5 = Passthrough() {is=title AnimStateNode}
   AnimStateTransitionNode_0 = AnimStateTransitionNode() {crossfadeDuration=0}
   AnimStateTransitionNode_1 = AnimStateTransitionNode() {crossfadeDuration=0}
   AnimStateTransitionNode_10 = AnimStateTransitionNode() {crossfadeDuration=0}
   AnimStateTransitionNode_13 = AnimStateTransitionNode() {blendMode=CircularOut, crossfadeDuration=0}
   AnimStateTransitionNode_14 = AnimStateTransitionNode() {crossfadeDuration=0}
   AnimStateTransitionNode_15 = AnimStateTransitionNode()
   AnimStateTransitionNode_16 = AnimStateTransitionNode() {bAutomaticRuleBasedOnSequencePlayerInState=true, crossfadeDuration=0}
   AnimStateTransitionNode_17 = AnimStateTransitionNode() {crossfadeDuration=0}
   AnimStateTransitionNode_18 = AnimStateTransitionNode() {crossfadeDuration=0}
   AnimStateTransitionNode_8 = AnimStateTransitionNode() {crossfadeDuration=0}
   AnimStateTransitionNode_9 = AnimStateTransitionNode() {crossfadeDuration=0}
  ]
  links [
   {
    dst=Pull_Success
    rule=LocomotionLayer_LowerBody.AnimGraphNode_StateMachine_0.HookShotStates.AnimStateTransitionNode_10.Transition
    src=Hit_Success
    via=AnimStateTransitionNode_10
   }
   {
    dst=Hit_Success
    rule=LocomotionLayer_LowerBody.AnimGraphNode_StateMachine_0.HookShotStates.AnimStateTransitionNode_8.Transition
    src=HookShot_Throw
    via=AnimStateTransitionNode_8
   }
   {
    dst=Passthrough
    rule=LocomotionLayer_LowerBody.AnimGraphNode_StateMachine_0.HookShotStates.AnimStateTransitionNode_15.Transition
    src=HookShot_Throw
    via=AnimStateTransitionNode_15
   }
   {
    dst=Pull_Fail
    rule=LocomotionLayer_LowerBody.AnimGraphNode_StateMachine_0.HookShotStates.AnimStateTransitionNode_9.Transition
    src=HookShot_Throw
    via=AnimStateTransitionNode_9
   }
   {
    dst=Hit_Success
    rule=LocomotionLayer_LowerBody.AnimGraphNode_StateMachine_0.HookShotStates.AnimStateTransitionNode_0.Transition
    src=Passthrough
    via=AnimStateTransitionNode_0
   }
   {
    dst=HookShot_Throw
    rule=LocomotionLayer_LowerBody.AnimGraphNode_StateMachine_0.HookShotStates.AnimStateTransitionNode_14.Transition
    src=Passthrough
    via=AnimStateTransitionNode_14
   }
   {
    dst=Pull_Fail
    rule=LocomotionLayer_LowerBody.AnimGraphNode_StateMachine_0.HookShotStates.AnimStateTransitionNode_17.Transition
    src=Passthrough
    via=AnimStateTransitionNode_17
   }
   {
    dst=Pull_Success
    rule=LocomotionLayer_LowerBody.AnimGraphNode_StateMachine_0.HookShotStates.AnimStateTransitionNode_1.Transition
    src=Passthrough
    via=AnimStateTransitionNode_1
   }
   {
    dst=HookShot_Throw
    rule=LocomotionLayer_LowerBody.AnimGraphNode_StateMachine_0.HookShotStates.AnimStateTransitionNode_18.Transition
    src=Pull_Fail
    via=AnimStateTransitionNode_18
   }
   {
    dst=Passthrough
    rule=LocomotionLayer_LowerBody.AnimGraphNode_StateMachine_0.HookShotStates.AnimStateTransitionNode_13.Transition
    src=Pull_Fail
    via=AnimStateTransitionNode_13
   }
   {
    dst=Passthrough
    rule=LocomotionLayer_LowerBody.AnimGraphNode_StateMachine_0.HookShotStates.AnimStateTransitionNode_16.Transition
    src=Pull_Success
    via=AnimStateTransitionNode_16
   }
  ]
 }
 HookShot_Throw {
  path=LocomotionLayer_LowerBody.AnimGraphNode_StateMachine_0.HookShotStates.AnimStateNode_0.HookShot_Throw
  nodes [Output Animation Pose(Result="Sequence Player (additive) 'CoffeePuff_Player_Hookshot_Throw'"() {bLoopAnimation=false, sequence=&/CoffeePuff_Player_Hookshot_Throw})]
 }
 Hit_Success {
  path=LocomotionLayer_LowerBody.AnimGraphNode_StateMachine_0.HookShotStates.AnimStateNode_1.Hit_Success
  nodes [Output Animation Pose(Result="Sequence Player (additive) 'CoffeePuff_Player_Hookshot_Hit'"() {bLoopAnimation=false, sequence=&/CoffeePuff_Player_Hookshot_Hit})]
 }
 Pull_Success {
  path=LocomotionLayer_LowerBody.AnimGraphNode_StateMachine_0.HookShotStates.AnimStateNode_2.Pull_Success
  nodes [Output Animation Pose(Result="Sequence Player (additive) 'CoffeePuff_Player_Hookshot_Pull_Success'"() {bLoopAnimation=false, sequence=&/CoffeePuff_Player_Hookshot_Pull_Success})]
 }
 Pull_Fail {
  path=LocomotionLayer_LowerBody.AnimGraphNode_StateMachine_0.HookShotStates.AnimStateNode_3.Pull_Fail
  nodes [Output Animation Pose(Result="Sequence Player (additive) 'CoffeePuff_Player_Hookshot_Pull_Fail'"() {bLoopAnimation=false, sequence=&/CoffeePuff_Player_Hookshot_Pull_Fail})]
 }
 Passthrough {
  path=LocomotionLayer_LowerBody.AnimGraphNode_StateMachine_0.HookShotStates.AnimStateNode_5.Passthrough
  nodes [Output Animation Pose(Result=Additive Identity Pose())]
 }
 Transition {
  path=LocomotionLayer_LowerBody.AnimGraphNode_StateMachine_0.HookShotStates.AnimStateTransitionNode_0.Transition
  nodes [Result(bCanEnterTransition=Get bHookShot_HitSuccess() -> (bHookShot_HitSuccess:bool))]
 }
 Transition {
  path=LocomotionLayer_LowerBody.AnimGraphNode_StateMachine_0.HookShotStates.AnimStateTransitionNode_1.Transition
  nodes [Result(bCanEnterTransition=Get bHookShot_PullSuccess() -> (bHookShot_PullSuccess:bool))]
 }
 Transition {
  path=LocomotionLayer_LowerBody.AnimGraphNode_StateMachine_0.HookShotStates.AnimStateTransitionNode_10.Transition
  nodes [Result(bCanEnterTransition=Get bHookShot_PullSuccess())]
 }
 Transition {
  path=LocomotionLayer_LowerBody.AnimGraphNode_StateMachine_0.HookShotStates.AnimStateTransitionNode_13.Transition
  nodes [Result(bCanEnterTransition=NOT Boolean(A=Get bHookShot_PullFail() -> (bHookShot_PullFail:bool)))]
 }
 Transition {
  path=LocomotionLayer_LowerBody.AnimGraphNode_StateMachine_0.HookShotStates.AnimStateTransitionNode_14.Transition
  nodes [Result(bCanEnterTransition=Get bHookShot_Throw() -> (bHookShot_Throw:bool))]
 }
 Transition {
  path=LocomotionLayer_LowerBody.AnimGraphNode_StateMachine_0.HookShotStates.AnimStateTransitionNode_15.Transition
  nodes [Result(bCanEnterTransition=NOT Boolean(A=Get bIsFiring()))]
 }
 Transition {path=LocomotionLayer_LowerBody.AnimGraphNode_StateMachine_0.HookShotStates.AnimStateTransitionNode_16.Transition, nodes [Result()]}
 Transition {
  path=LocomotionLayer_LowerBody.AnimGraphNode_StateMachine_0.HookShotStates.AnimStateTransitionNode_17.Transition
  nodes [Result(bCanEnterTransition=Get bHookShot_PullFail())]
 }
 Transition {
  path=LocomotionLayer_LowerBody.AnimGraphNode_StateMachine_0.HookShotStates.AnimStateTransitionNode_18.Transition
  nodes [Result(bCanEnterTransition=Get bHookShot_Throw())]
 }
 Transition {
  path=LocomotionLayer_LowerBody.AnimGraphNode_StateMachine_0.HookShotStates.AnimStateTransitionNode_8.Transition
  nodes [Result(bCanEnterTransition=Get bHookShot_HitSuccess())]
 }
 Transition {
  path=LocomotionLayer_LowerBody.AnimGraphNode_StateMachine_0.HookShotStates.AnimStateTransitionNode_9.Transition
  nodes [Result(bCanEnterTransition=Get bHookShot_PullFail())]
 }
 CorrectivePoseBlend {
  entry=PassThrough
  path=LocomotionLayer_LowerBody.AnimGraphNode_StateMachine_1.CorrectivePoseBlend
  ids=id
  nodes [
   AnimStateNode_0 = PassThrough() {
    stateEntered {guid=#1}
    stateFullyBlended {guid=#2}
    stateLeft {guid=#3}
   }
   AnimStateNode_1 = CorrectiveAdditive() {is=title AnimStateNode}
   AnimStateNode_2 = EquipAdditive() {
    is=title AnimStateNode
    stateEntered {guid=#4}
    stateFullyBlended {guid=#5}
    stateLeft {guid=#6}
   }
   AnimStateTransitionNode_0 = AnimStateTransitionNode() {
    blendMode=CircularOut
    crossfadeDuration=0.5
    blendProfileWrapper {blendProfile=&Fortnite_M_Avg_Player_Skeleton:FastFeet}
   }
   AnimStateTransitionNode_11 = AnimStateTransitionNode() {blendMode=CubicInOut, crossfadeDuration=0.4}
   AnimStateTransitionNode_2 = AnimStateTransitionNode() {crossfadeDuration=0.45}
   AnimStateTransitionNode_3 = AnimStateTransitionNode() {
    transitionEnd {guid=#7}
    transitionInterrupt {guid=#8}
    transitionStart {guid=#9}
   }
   AnimStateTransitionNode_4 = AnimStateTransitionNode()
  ]
  links [
   {
    dst=PassThrough
    rule=LocomotionLayer_LowerBody.AnimGraphNode_StateMachine_1.CorrectivePoseBlend.AnimStateTransitionNode_2.Transition
    src=CorrectiveAdditive
    via=AnimStateTransitionNode_2
   }
   {
    dst=CorrectiveAdditive
    rule=LocomotionLayer_LowerBody.AnimGraphNode_StateMachine_1.CorrectivePoseBlend.AnimStateTransitionNode_4.Transition
    src=EquipAdditive
    via=AnimStateTransitionNode_4
   }
   {
    dst=PassThrough
    rule=LocomotionLayer_LowerBody.AnimGraphNode_StateMachine_1.CorrectivePoseBlend.AnimStateTransitionNode_11.Transition
    src=EquipAdditive
    via=AnimStateTransitionNode_11
   }
   {
    dst=CorrectiveAdditive
    rule=LocomotionLayer_LowerBody.AnimGraphNode_StateMachine_1.CorrectivePoseBlend.AnimStateTransitionNode_0.Transition
    src=PassThrough
    via=AnimStateTransitionNode_0
   }
   {
    dst=EquipAdditive
    rule=LocomotionLayer_LowerBody.AnimGraphNode_StateMachine_1.CorrectivePoseBlend.AnimStateTransitionNode_3.Transition
    src=PassThrough
    via=AnimStateTransitionNode_3
   }
  ]
 }
 PassThrough {
  path=LocomotionLayer_LowerBody.AnimGraphNode_StateMachine_1.CorrectivePoseBlend.AnimStateNode_0.PassThrough
  nodes [Output Animation Pose(Result=Use cached pose 'LocoLayer_LowerBody'())]
 }
 CorrectiveAdditive {
  path=LocomotionLayer_LowerBody.AnimGraphNode_StateMachine_1.CorrectivePoseBlend.AnimStateNode_1.CorrectiveAdditive
  nodes [Output Animation Pose(Result=Apply Additive(
     Base=Use cached pose 'LocoLayer_LowerBody'()
     Additive="Sequence Player (additive) 'PoppySoap_NonTargeted_Pose_Add'"() {bLoopAnimation=false, sequence=/PoppySoap/Assets/Animation/PoppySoap_NonTargeted_Pose_Add}
    ) {alphaBoolBlend {blendOption=Linear}})]
 }
 EquipAdditive {
  path=LocomotionLayer_LowerBody.AnimGraphNode_StateMachine_1.CorrectivePoseBlend.AnimStateNode_2.EquipAdditive
  nodes [
   n1 = "Sequence Player (additive) 'PoppySoap_Equip'"() {bLoopAnimation=false, sequence=/PoppySoap/Assets/Animation/PoppySoap_Equip}
   n2 = Use cached pose 'LocoLayer_LowerBody'()
   n3 = Layered blend per bone(
    BasePose=Use cached pose 'LocoLayer_LowerBody'()
    BlendPoses_0=Sequence Player 'PoppySoap_Equip_MeshSpace'() {bLoopAnimation=false, sequence=&PoppySoap_Equip_MeshSpace}
   ) {layerSetup [{branchFilters [{boneName=clavicle_l}, {boneName=clavicle_r}]}]}
   n4 = Use cached pose 'LocoLayer_LowerBody'()
   n5 = Get bIsCrouching()
   Output Animation Pose(Result=Blend Poses by bool(
     BlendPose_0=Blend Poses by bool(BlendPose_0=@n3, BlendPose_1=@n4, BlendTime_0=0.2, BlendTime_1=0.2, bActiveValue=@n5) {blendType=Linear}
     BlendPose_1=Apply Additive(Base=@n2, Additive=@n1) {alphaBoolBlend {blendOption=Linear}}
     BlendTime_0=0.2
     BlendTime_1=0.2
     bActiveValue=Get bIsMoving2D()
    ) {blendType=Linear})
  ]
 }
 Transition {
  path=LocomotionLayer_LowerBody.AnimGraphNode_StateMachine_1.CorrectivePoseBlend.AnimStateTransitionNode_0.Transition
  nodes [Result(bCanEnterTransition=Get bApplyLowerBodyCorrective())]
 }
 Transition {
  path=LocomotionLayer_LowerBody.AnimGraphNode_StateMachine_1.CorrectivePoseBlend.AnimStateTransitionNode_11.Transition
  nodes [Result(bCanEnterTransition=Get bEquipAdditiveToPassThrough() -> (bEquipAdditiveToPassThrough:bool))]
 }
 Transition {
  path=LocomotionLayer_LowerBody.AnimGraphNode_StateMachine_1.CorrectivePoseBlend.AnimStateTransitionNode_2.Transition
  nodes [Result(bCanEnterTransition=NOT Boolean(A=Get bApplyLowerBodyCorrective()))]
 }
 Transition {
  path=LocomotionLayer_LowerBody.AnimGraphNode_StateMachine_1.CorrectivePoseBlend.AnimStateTransitionNode_3.Transition
  nodes [Result(bCanEnterTransition=Get bIsEquipMontagePlaying() -> (bIsEquipMontagePlaying:bool))]
 }
 Transition {
  path=LocomotionLayer_LowerBody.AnimGraphNode_StateMachine_1.CorrectivePoseBlend.AnimStateTransitionNode_4.Transition
  nodes [Result(bCanEnterTransition=Get bEquipAdditiveToCorrectiveAdditive() -> (bEquipAdditiveToCorrectiveAdditive:bool))]
 }
 LocomotionLayer_UpperBody {nodes [
   Input Pose() {name=InPose_UpperBody}
   Output Pose() {layerGroup=LocomotionOverride}
  ]}
]
variables []
