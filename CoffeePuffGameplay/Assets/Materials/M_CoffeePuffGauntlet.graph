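# Material for the CoffeePuff gauntlet, layering animated fire, heat, cracks and Fresnel glow over textured PBR.
# Panning fire noise, distorted by Perlin, and static noise build masks that alter surface attributes and blend in emissive.
# Animation-driven FireEffect bias parameters intensify left and right sides separately, split by pre-skinned local X.
# A pixel-depth rim term adds distant glow; emissive passes through EyeAdaptationInverse, and everything feeds MF_CharacterParent_2023_PostFX.
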Material /CoffeePuffGameplay/Assets/Materials/M_CoffeePuffGauntlet
graph {
 properties {
  bAutomaticallySetUsageInEditor=false
  bUseMaterialAttributes=true
  bUsedWithMeshDeformer=true
  bUsedWithNanite=true
  bUsedWithSkeletalMesh=true
  previewMesh=/CoffeePuffGameplay/Assets/Meshes/Gauntlet_CoffeePuff_Pickup
  usedShadingModels=Default Lit
 }
 nodes [
  Add(A=Add(), B=Fresnel())
  n1 = Subtract(A=NamedRerouteUsage())
  n2 = ScalarParameter() {group=Animation - Dont change, parameterName=FireEffect_Left_Bias}
  n3 = Add(A=@n1, B=@n2)
  n4 = ScalarParameter() {defaultValue=-0.1, group=Animation - Dont change, parameterName=FireEffect_Bias}
  n5 = ScalarParameter() {group=Animation - Dont change, parameterName=FireEffect_Right_Bias}
  n6 = Add(A=@n1, B=@n5)
  n7 = LinearInterpolate(A=Multiply(A=Add(A=@n3, B=@n4), B=@n2), B=Multiply(A=Add(A=@n6, B=@n4), B=@n5), Alpha=NamedRerouteUsage())
  Add(A=@n7)
  n8 = Multiply(A=NamedRerouteUsage(), B=ScalarParameter() {defaultValue=47.4011, group=Cracks, parameterName=Cracks Surface})
  Add(A=@n8)
  n9 = ScalarParameter() {defaultValue=0.8, group=Texture, parameterName=Flame Speed}
  Add(A=@n9)
  AppendVector()
  Constant()
  Constant()
  DotProduct()
  n10 = Power(Base=Add(A=NamedRerouteUsage(), B=NamedRerouteUsage()) {constB=0}) {constExponent=1.2}
  n11 = PixelDepth()
  n12 = ScalarParameter() {defaultValue=494.8452, parameterName=Distance}
  n13 = Saturate(None=Divide(A=Subtract(A=@n11.R, B=@n12), B=ScalarParameter() {defaultValue=245.8024, parameterName=Param_4}))
  n14 = DotProduct(A=CameraVectorWS(), B=Transform(None=NamedRerouteUsage()))
  n15 = Power(Base=Multiply(A=Add(A=@n13) {constB=0.3}, B=OneMinus(None=@n14))) {constExponent=3}
  n16 = Multiply(A=Multiply(A=@n15, B=ScalarParameter() {defaultValue=20}), B=VectorParameter() {defaultValue=<rgba 1 0.0875 0 1>}.RGB)
  n17 = Add(A=NamedRerouteDeclaration(None=Add(A=@n10, B=@n16)) {name=Fire Emissive, nodeColor=<rgba 1 0 0.8203 1>}, B=NamedRerouteUsage()) {constB=0}
  n18 = Multiply(A=NamedRerouteUsage(), B=ScalarParameter() {defaultValue=50, group=Fresnel, parameterName=Fresnel Int})
  n19 = NamedRerouteUsage()
  n20 = Constant() {r=2}
  n21 = Multiply(A=MaterialFunctionCall(In=Power(Base=@n19, Exp=@n20), Contrast=Constant() {r=0.2}) {materialFunction=&CheapContrast}.Result) {constB=0.5}
  n22 = Add(A=NamedRerouteUsage(), B=NamedRerouteUsage())
  n23 = Multiply(A=@n22, B=@n22)
  n24 = Multiply(
   A=Multiply(A=Multiply(A=@n18, B=@n21), B=VectorParameter() {group=Fresnel, parameterName=Fresnel Color, defaultValue=<rgba 1 0.1673 0.0524 1>}.RGB)
   B=LinearInterpolate(Alpha=Saturate(None=@n23)) {constA=0.2}
  )
  n25 = Multiply(
   A=@n24
   B=Clamp(None=ScalarParameter() {group=Animation - Dont change, parameterName=FireEffect_Right_Bias}) {clampMode=CMODE_ClampMin, minDefault=0.5}
  )
  n26 = ScalarParameter() {group=Animation - Dont change, parameterName=FireEffect_Left_Bias, sliderMax=1}
  n27 = ScalarParameter() {group=Animation - Dont change, parameterName=FireEffect_Right_Bias, sliderMax=1}
  n28 = LinearInterpolate(Alpha=Saturate(None=Add(A=@n26, B=@n27))) {constA=1, constB=3}
  n29 = Multiply(
   A=@n24
   B=Clamp(None=ScalarParameter() {group=Animation - Dont change, parameterName=FireEffect_Left_Bias}) {clampMode=CMODE_ClampMin, minDefault=0.5}
  )
  n30 = Add(
   A=LinearInterpolate(B=Add(A=@n17) {constB=0}, Alpha=NamedRerouteUsage())
   B=LinearInterpolate(A=Multiply(A=@n25, B=@n28), B=Multiply(A=@n29, B=@n28), Alpha=NamedRerouteUsage())
  )
  n31 = ScalarParameter() {defaultValue=0.3, group=PBR, parameterName=Eye Adapt, sliderMax=1}
  n32 = Saturate(None=Add(
    A=ScalarParameter() {group=Animation - Dont change, parameterName=FireEffect_Left_Bias, sliderMax=1}
    B=ScalarParameter() {group=Animation - Dont change, parameterName=FireEffect_Right_Bias, sliderMax=1}
   ))
  n33 = Multiply(A=Add(A=NamedRerouteUsage()) {constB=0.2}, B=LinearInterpolate(Alpha=@n32) {constA=0.4, constB=-1}) {constB=0.3}
  n34 = GetMaterialAttributes(None=NamedRerouteUsage()) {attributeGetTypes [
    #1
    #2
    #3
    #4
   ]}
  n35 = SetMaterialAttributes!(
   MaterialAttributes
   Base Color=LinearInterpolate(A=@n34.Base Color, B=Multiply(A=@n33, B=@n34.Base Color), Alpha=NamedRerouteUsage()) {constB=0, desc=Base Color}
   Metallic=LinearInterpolate(A=@n34.Metallic, Alpha=NamedRerouteUsage()) {constB=0, desc=Metal}
   Specular=LinearInterpolate(A=@n34.Specular, Alpha=NamedRerouteUsage()) {constB=0, desc=Spec}
   Roughness=LinearInterpolate(A=@n34.Roughness, Alpha=NamedRerouteUsage()) {desc=Rough}
  ) {attributeSetTypes [
    #1
    #2
    #3
    #4
   ]}
  MaterialFunctionCall(
   Material=SetMaterialAttributes!(MaterialAttributes=@n35, Emissive Color=EyeAdaptationInverse(LightValueInput=@n30, AlphaInput=@n31).EyeAdaptationInverse) {attributeSetTypes [#5]}
   M=Constant()
   Fresnel=Fresnel(ExponentIn=Constant() {r=1}) {baseReflectFraction=0}
  ) {materialFunction=&MF_CharacterParent_2023_PostFX}
  n36 = TextureCoordinate()
  n37 = ScalarParameter() {defaultValue=0.3, group=Texture, parameterName=Tilling Flames, sliderMax=2}
  n38 = MaterialFunctionCall(Float2=Multiply(A=Multiply(A=@n36, B=@n37)) {constB=0.6}) {materialFunction=&BreakOutFloat2Components}
  n39 = AppendVector(A=Multiply(A=@n38.R) {constB=4}, B=Multiply(A=@n38.G))
  n40 = Multiply(A=Time(), B=@n9)
  n41 = NamedRerouteDeclaration(None=TextureSample(UVs=Panner(Coordinate=@n39, Time=@n40) {speedX=0.2, speedY=0.3}) {texture=/Game/Global/Textures/Default/Noise/T_FireNoise_01}.R) {name=Noise, nodeColor=<rgba 1 0.3387 0 1>}
  n42 = MaterialFunctionCall(Float2=@n37) {materialFunction=&BreakOutFloat2Components}
  n43 = Multiply(A=AppendVector(A=Multiply(A=@n42.R) {constB=2}, B=Multiply(A=@n42.G)), B=TextureCoordinate())
  n44 = MaterialFunctionCall(Float2=TextureCoordinate()) {materialFunction=&BreakOutFloat2Components}
  n45 = AppendVector(A=Multiply(A=@n44.R) {constB=2}, B=Multiply(A=@n44.G))
  n46 = TextureSample(UVs=Panner(Coordinate=Multiply(A=@n45) {constB=0.3}) {speedX=-0.2, speedY=0.1}) {samplerType=SAMPLERTYPE_Masks, texture=/Game/Global/Textures/Default/Noise/T_Generic_Perlin}
  n47 = Multiply(A=Multiply(A=Multiply(A=@n46.R, B=@n46.R), B=@n46.B)) {constB=0.2}
  n48 = TextureSample(UVs=Panner(Coordinate=Add(A=@n43, B=@n47), Time=@n40) {speedX=-0.1, speedY=0.2}) {texture=&T_FireNoise_01}
  n49 = Multiply(A=@n41, B=@n48.R)
  Multiply(A=@n49, B=@n49)
  n50 = TextureSample() {samplerType=SAMPLERTYPE_Masks, texture=/CoffeePuffGameplay/Assets/Textures/T_Gauntlet_CoffeePuff_S}
  n51 = SetMaterialAttributes!(
   MaterialAttributes
   Base Color=TextureSample() {texture=/CoffeePuffGameplay/Assets/Textures/T_Gauntlet_CoffeePuff_D}.RGB
   Metallic=NamedRerouteDeclaration(None=@n50.B) {name=Metal Mask, nodeColor=<rgba 0 0.8359 1 1>}
   Specular=Multiply(
    A=ScalarParameter() {defaultValue=1, group=PBR, parameterName=Specular, sliderMax=1}
    B=NamedRerouteDeclaration(None=@n50.R) {name=Spec Mask, nodeColor=<rgba 0.5 0 1 1>}
   )
   Roughness=LinearInterpolate(
    A=ScalarParameter() {group=PBR, parameterName=SpecRoughnessMin, sliderMax=1}
    B=ScalarParameter() {defaultValue=1, group=PBR, parameterName=SpecRoughnessMax, sliderMax=1}
    Alpha=@n50.G
   )
   Normal=NamedRerouteDeclaration(None=TextureSample() {samplerType=SAMPLERTYPE_Normal, texture=/CoffeePuffGameplay/Assets/Textures/T_Gauntlet_CoffeePuff_N}.RGB) {name=Normal, nodeColor=<rgba 0 0.8594 1 1>}
  ) {attributeSetTypes [
    #1
    #2
    #3
    #4
    #6
   ]}
  NamedRerouteDeclaration(None=@n51) {name=PBR, nodeColor=<rgba 0 0.0391 1 1>}
  n52 = NamedRerouteUsage()
  NamedRerouteDeclaration(None=Fresnel(Normal=Transform(None=@n52)) {baseReflectFraction=0, exponent=3}) {name=Base Fresnel, nodeColor=<rgba 1 0.1406 0 1>}
  n53 = ScalarParameter() {defaultValue=8, group=Heat, parameterName=Heat Intensity}
  n54 = Saturate(None=NamedRerouteUsage())
  n55 = NamedRerouteUsage()
  n56 = Add(A=Add(A=ComponentMask(None=@n55) {r=true}, B=ComponentMask(None=@n55) {g=true}), B=ComponentMask(None=@n55) {b=true})
  n57 = Multiply(A=@n56, B=@n56)
  n58 = Multiply(A=@n57, B=@n57)
  n59 = LinearInterpolate(Alpha=@n58) {constA=1, constB=0}
  n60 = Multiply(A=@n59, B=@n59)
  n61 = Saturate(None=@n60)
  n62 = SmoothStep(
   Max=ScalarParameter() {defaultValue=8, group=Heat, parameterName=Heat Contrast}
   Value=Add(A=Multiply(A=@n61) {constB=0.5}, B=@n58) {constB=0}
  ) {constMax=3}
  n63 = Multiply(A=Add(A=LinearInterpolate(Alpha=@n54) {constB=0.5}, B=Multiply(A=@n62) {constB=0.1}), B=@n62)
  n64 = LinearInterpolate(
   A=VectorParameter() {group=Heat, parameterName=Heat color 2, defaultValue=<rgba 1 0.3044 0.0133 1>}.RGB
   B=VectorParameter() {group=Heat, parameterName=Heat color 1, defaultValue=<rgba 1 0.0245 0 1>}.RGB
   Alpha=Saturate(None=@n62)
  )
  n65 = ScalarParameter() {defaultValue=0.3, group=Heat, parameterName=Desaturation}
  NamedRerouteDeclaration(None=Multiply(A=Multiply(A=@n53, B=@n63), B=Desaturation(None=@n64, Fraction=@n65))) {name=Base Heat, nodeColor=<rgba 0.4072 0.2086 0.0382 1>}
  NamedRerouteDeclaration(None=Power(Base=@n49) {constExponent=1.5}) {name=Fire Noise, nodeColor=<rgba 1 0.0464 0 1>}
  n66 = MaterialFunctionCall(Float2=Multiply(A=TextureCoordinate()) {constB=2}) {materialFunction=&BreakOutFloat2Components}
  n67 = TextureSample(UVs=AppendVector(A=Multiply(A=@n66.R) {constB=4}, B=Multiply(A=@n66.G))) {texture=&T_FireNoise_01}
  n68 = Multiply(A=@n67.R, B=@n67.R)
  NamedRerouteDeclaration(None=Multiply(A=@n68, B=@n68)) {name=Noise Static, nodeColor=<rgba 0 0.5547 1 1>}
  NamedRerouteDeclaration(None=Power(Base=OneMinus(None=@n67.R)) {constExponent=10}) {name=Noise Static Reverse, nodeColor=<rgba 0 0.5547 1 1>}
  n69 = VectorParameter() {group=Flames, parameterName=Flame Color, defaultValue=<rgba 1 0.2847 0.1116 1>}
  n70 = Add(A=NamedRerouteUsage(), B=Multiply(A=NamedRerouteUsage()) {constB=0.5})
  n71 = Multiply(
   A=Multiply(A=NamedRerouteUsage(), B=ScalarParameter() {defaultValue=2, group=Flames, parameterName=Flames Alpha})
   B=Add(A=Multiply(A=@n70, B=@n70)) {constB=0.3}
  )
  NamedRerouteDeclaration(None=Multiply(A=Multiply(A=@n69.RGB, B=@n71), B=ScalarParameter() {defaultValue=10, group=Flames, parameterName=Flame Int})) {name=Flames, nodeColor=<rgba 1 0.1234 0.0076 1>}
  n72 = Subtract(A=Multiply(A=@n58, B=@n58), B=@n60)
  NamedRerouteDeclaration(None=Saturate(None=Multiply(A=@n60, B=@n72))) {name=Detail Normal, nodeColor=<rgba 0 1 0.3906 1>}
  n73 = LinearInterpolate(
   B=Multiply(A=VectorParameter() {group=Cracks, parameterName=Cracks Color, defaultValue=<rgba 0.6719 0.2449 0.182 1>}.RGB)
   Alpha=Saturate(None=@n8)
  )
  n74 = Multiply(A=@n73, B=NamedRerouteUsage())
  NamedRerouteDeclaration(None=Multiply(A=Multiply(A=@n73, B=@n74), B=ScalarParameter() {defaultValue=5.3107, group=Cracks, parameterName=Cracks Intensity})) {name=Burn Cracks, nodeColor=<rgba 1 0.9609 0 1>}
  n75 = NamedRerouteUsage()
  n76 = ScalarParameter() {defaultValue=0.3, group=Idle, parameterName=Base Mask Contrast}
  NamedRerouteDeclaration(None=Multiply(
    A=MaterialFunctionCall(In=@n75, Contrast=@n76) {materialFunction=&CheapContrast}.Result
    B=ScalarParameter() {defaultValue=0.4, group=Idle, parameterName=Constant Fire Int}
   ) {constB=0.2}) {name=Constant Mask, nodeColor=<rgba 1 0.7734 0 1>}
  NamedRerouteDeclaration(None=@n62) {name=Heat Mask, nodeColor=<rgba 0 1 0.2031 1>}
  NamedRerouteDeclaration(None=SmoothStep(
    Min=ScalarParameter() {defaultValue=0.7687, group=Idle, parameterName=Fire Mask Contrast Min}
    Max=ScalarParameter() {defaultValue=2.5, group=Idle, parameterName=Fire Mask Contrast Max}
    Value=Add(A=@n41, B=@n48.R)
   )) {name=Fire Noise 2, nodeColor=<rgba 0 1 0.8125 1>}
  n77 = NamedRerouteUsage()
  n78 = Constant() {r=15}
  n79 = NamedRerouteUsage()
  n80 = ScalarParameter() {defaultValue=4, group=Idle, parameterName=Heat Mask Int}
  n81 = Add(A=Add(A=Multiply(A=@n77, B=@n78), B=NamedRerouteUsage()), B=Add(A=Multiply(A=@n79, B=@n80), B=NamedRerouteUsage()))
  n82 = NamedRerouteUsage()
  n83 = Add(A=Desaturation(None=Multiply(A=@n81, B=@n82)), B=NamedRerouteUsage())
  n84 = Multiply(A=@n83, B=@n83)
  n85 = Multiply(A=@n84, B=@n84)
  n86 = ObjectPositionWS() {originType=CameraRelative}
  n87 = Subtract(
   A=WorldPosition() {worldPositionShaderOffset=WPT_CameraRelative}.Z
   B=Add(A=ComponentMask(None=@n86) {b=true}, B=ScalarParameter() {defaultValue=-46.6606, parameterName=Param_2})
  )
  n88 = ScalarParameter() {defaultValue=70.9831, parameterName=Param_1}
  n89 = LinearInterpolate(Alpha=Saturate(None=Divide(A=@n87, B=@n88))) {constA=1, constB=0}
  n90 = Saturate(None=MaterialFunctionCall(In=Multiply(A=@n15, B=@n89), Contrast=ScalarParameter() {parameterName=Param_3}) {materialFunction=&CheapContrast}.Result)
  n91 = Saturate(None=Multiply(A=Add(A=@n85, B=@n90), B=ScalarParameter() {defaultValue=5, group=Idle, parameterName=Idle Mask Int}) {constB=5})
  n92 = NamedRerouteUsage()
  n93 = Add(A=@n92, B=Multiply(A=NamedRerouteUsage(), B=Constant() {r=5}) {constB=4})
  n94 = Saturate(None=Multiply(A=Multiply(A=@n92, B=@n93), B=Constant() {r=1.5}) {constB=3})
  NamedRerouteDeclaration(None=LinearInterpolate(A=@n91, B=Add(A=@n91, B=@n94), Alpha=@n32)) {name=Final Mask, nodeColor=<rgba 1 0.5678 0.0455 1>}
  n95 = ScalarParameter() {defaultValue=3.5, group=Animation - Dont change, parameterName=FireEffect_Falloff}
  NamedRerouteDeclaration(None=Saturate(None=Multiply(A=@n7, B=@n95))) {name=Bias_Mask, nodeColor=<rgba 1 0.2061 0 1>}
  n96 = LocalPosition() {localOrigin=InstancePreSkinning}
  n97 = ShaderStageSwitch(PixelShader=VertexInterpolator(VS=@n96.XYZ).PS, VertexShader=@n96.XYZ)
  NamedRerouteDeclaration(None=Step(Y=ComponentMask(None=@n97) {r=true})) {name=M_Left, nodeColor=<rgba 0.5703 1 0 1>}
  n98 = LocalPosition() {localOrigin=InstancePreSkinning}
  n99 = ShaderStageSwitch(PixelShader=VertexInterpolator(VS=@n98.XYZ).PS, VertexShader=@n98.XYZ)
  n100 = Constant3Vector() {constant=<rgba 0 -15 140 0>}
  n101 = Constant() {r=85}
  n102 = Constant() {r=0.5}
  NamedRerouteDeclaration(None=OneMinus(None=SphereMask(A=@n99, B=@n100, Radius=@n101, Hardness=@n102) {attenuationRadius=50, hardnessPercent=1})) {name=SphereMask, nodeColor=<rgba 0 1 1 1>}
  n103 = Power(Base=NamedRerouteUsage()) {constExponent=5}
  n104 = Multiply(A=NamedRerouteUsage()) {constB=0.5}
  n105 = LinearInterpolate(Alpha=Saturate(None=Subtract(A=@n103, B=@n104))) {constA=0.1}
  n106 = Multiply(A=Multiply(A=@n105, B=@n105)) {constB=40}
  NamedRerouteDeclaration(None=Multiply(A=@n106, B=@n106)) {name=Surface Mask, nodeColor=<rgba 0.4072 0.0331 0.0056 1>}
  NamedRerouteUsage()
  NamedRerouteUsage()
  NamedRerouteUsage()
  NamedRerouteUsage()
  NamedRerouteUsage()
 ]
}
