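# Sound cue for CoffeePuff's air attack in the melee weapon sound class, capped at three concurrent plays.
# One mixer layers randomized kick, swing and quieter, lower-pitched crunch waves with varied pitch.
# Branches on islocal and IsEnemy feed that shared mix: local, enemy and unset-IsEnemy play at full volume, non-local non-enemies much quieter.
# Non-local paths use close-range melee attenuation, while the local path's attenuation node assigns no asset.
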SoundCue /CoffeePuffGameplay/Assets/Sounds/IP_SC_CoffeePuff_AirAttack_01
root=Branch_0
nodes [
 {id=Attenuation_0, children [Modulator_2]}
 {
  id=Attenuation_1
  children [Modulator_3]
  properties {attenuationSettings=/Game/Sounds/Attenuation/Fort_Weapon_Melee_Attack_Close_Range_Mono_Att}
 }
 {
  id=Attenuation_2
  children [Modulator_4]
  properties {attenuationSettings=&Fort_Weapon_Melee_Attack_Close_Range_Mono_Att}
 }
 {id=Branch_0, children [Attenuation_0, Branch_3, Branch_3], properties {boolParameterName=islocal}}
 {id=Branch_3, children [Attenuation_2, Attenuation_1, Attenuation_2], properties {boolParameterName=IsEnemy}}
 {id=Mixer_0, children [Modulator_7,Modulator_8,Modulator_0], properties {inputVolume [1.3,1,0.45]}}
 {id=Modulator_0, children [Random_0], properties {pitchMax=0.85, pitchMin=0.8, volumeMax=1, volumeMin=0.8}}
 {id=Modulator_2, children [Mixer_0], properties {pitchMax=1, pitchMin=1, volumeMax=1, volumeMin=1}}
 {id=Modulator_3, children [Mixer_0], properties {pitchMax=1, pitchMin=1, volumeMax=0.2, volumeMin=0.2}}
 {id=Modulator_4, children [Mixer_0], properties {pitchMax=1, pitchMin=1, volumeMax=1, volumeMin=1}}
 {id=Modulator_7, children [Random_4], properties {pitchMax=1.1, pitchMin=0.9, volumeMax=1, volumeMin=1}}
 {id=Modulator_8, children [Random_5], properties {pitchMax=1.3, pitchMin=1.1, volumeMax=1, volumeMin=1}}
 {
  id=Random_0
  children [WavePlayer_1,WavePlayer_2,WavePlayer_3,WavePlayer_4]
  properties {bShouldExcludeFromBranchCulling=true, weights [1,1,1,1]}
 }
 {
  id=Random_4
  children [WavePlayer_20, WavePlayer_21, WavePlayer_22, WavePlayer_23]
  properties {bShouldExcludeFromBranchCulling=true, weights [1,1,1,1]}
 }
 {
  id=Random_5
  children [WavePlayer_24, WavePlayer_25, WavePlayer_26, WavePlayer_27, WavePlayer_28, WavePlayer_29]
  properties {bShouldExcludeFromBranchCulling=true, weights [1,1,1,1,1,1]}
 }
 {id=WavePlayer_1, properties {soundWaveAssetPtr=&/IP_CoffeePuff_Attack_Swing_Big_Crunch_01}}
 {id=WavePlayer_2, properties {soundWaveAssetPtr=&/IP_CoffeePuff_Attack_Swing_Big_Crunch_02}}
 {id=WavePlayer_20, properties {soundWaveAssetPtr=&/IP_CoffeePuff_Attack_Kick_01}}
 {id=WavePlayer_21, properties {soundWaveAssetPtr=&/IP_CoffeePuff_Attack_Kick_02}}
 {id=WavePlayer_22, properties {soundWaveAssetPtr=&/IP_CoffeePuff_Attack_Kick_03}}
 {id=WavePlayer_23, properties {soundWaveAssetPtr=&/IP_CoffeePuff_Attack_Kick_04}}
 {id=WavePlayer_24, properties {soundWaveAssetPtr=&/IP_CoffeePuff_Attack_Swing_01}}
 {id=WavePlayer_25, properties {soundWaveAssetPtr=&/IP_CoffeePuff_Attack_Swing_02}}
 {id=WavePlayer_26, properties {soundWaveAssetPtr=&/IP_CoffeePuff_Attack_Swing_03}}
 {id=WavePlayer_27, properties {soundWaveAssetPtr=&/IP_CoffeePuff_Attack_Swing_04}}
 {id=WavePlayer_28, properties {soundWaveAssetPtr=&/IP_CoffeePuff_Attack_Swing_05}}
 {id=WavePlayer_29, properties {soundWaveAssetPtr=&/IP_CoffeePuff_Attack_Swing_06}}
 {id=WavePlayer_3, properties {soundWaveAssetPtr=&/IP_CoffeePuff_Attack_Swing_Big_Crunch_03}}
 {id=WavePlayer_4, properties {soundWaveAssetPtr=&/IP_CoffeePuff_Attack_Swing_Big_Crunch_04}}
]
properties {
 bDebug=true
 bOverrideConcurrency=true
 maxDistance=2097152
 soundClassObject=&Fort_WeaponMelee
 volumeMultiplier=1
 attenuationOverrides {distanceAlgorithm=Logarithmic}
 concurrencyOverrides {maxCount=3}
}
