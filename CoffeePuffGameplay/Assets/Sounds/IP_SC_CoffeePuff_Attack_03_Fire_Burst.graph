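# Sound cue for CoffeePuff's third attack fire burst, played as a melee weapon sound with a three-voice concurrency cap.
# A mixer layers randomized burst and start-scream waves with a quiet, delayed, pitched-down embers layer that plays only on the first two quality levels.
# Branches on islocal and IsEnemy route that shared mix through separate attenuation and volume: local loudest, enemy or unset close-range stereo, non-enemy quietest in mono.
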SoundCue /CoffeePuffGameplay/Assets/Sounds/IP_SC_CoffeePuff_Attack_03_Fire_Burst
root=Branch_1
nodes [
 {id=Attenuation_0, children [Modulator_1]}
 {
  id=Attenuation_1
  children [Modulator_3]
  properties {attenuationSettings=/Game/Sounds/Attenuation/Fort_Weapon_Melee_Attack_Close_Range_Mono_Att}
 }
 {
  id=Attenuation_2
  children [Modulator_4]
  properties {attenuationSettings=/Game/Sounds/Attenuation/Fort_Weapon_Melee_Attack_Close_Range_Stereo_Att}
 }
 {id=Branch_1, children [Attenuation_0, Branch_3, Branch_3], properties {boolParameterName=islocal}}
 {id=Branch_3, children [Attenuation_2, Attenuation_1, Attenuation_2], properties {boolParameterName=IsEnemy}}
 {id=Delay_0, children [Modulator_6], properties {delayMax=0.65, delayMin=0.65}}
 {id=Mixer_0, children [Modulator_0, Modulator_2, QualityLevel_0], properties {inputVolume [1,1,0.1]}}
 {id=Modulator_0, children [Random_1], properties {volumeMax=1, volumeMin=0.9}}
 {id=Modulator_1, children [Mixer_0], properties {pitchMax=1, pitchMin=1, volumeMax=0.75, volumeMin=0.75}}
 {id=Modulator_2, children [Random_2], properties {pitchMax=1.2, pitchMin=0.9, volumeMax=1, volumeMin=0.7}}
 {id=Modulator_3, children [Mixer_0], properties {pitchMax=1, pitchMin=1, volumeMax=0.2, volumeMin=0.2}}
 {id=Modulator_4, children [Mixer_0], properties {pitchMax=1, pitchMin=1, volumeMax=0.5, volumeMin=0.5}}
 {id=Modulator_6, children [Random_3], properties {pitchMax=0.9, pitchMin=0.75, volumeMax=1, volumeMin=1}}
 {id=QualityLevel_0, children [Delay_0,Delay_0,null,null]}
 {id=Random_1, children [WavePlayer_2,WavePlayer_3,WavePlayer_4], properties {bShouldExcludeFromBranchCulling=true, weights [1,1,1]}}
 {id=Random_2, children [WavePlayer_5,WavePlayer_6,WavePlayer_7], properties {bShouldExcludeFromBranchCulling=true, weights [1,1,1]}}
 {id=Random_3, children [WavePlayer_8, WavePlayer_9, WavePlayer_10], properties {bShouldExcludeFromBranchCulling=true, weights [1,1,1]}}
 {id=WavePlayer_10, properties {soundWaveAssetPtr=&/IP_CoffeePuff_Attack_03_Fire_Burst_Embers_03}}
 {id=WavePlayer_2, properties {soundWaveAssetPtr=&/IP_CoffeePuff_Attack_03_Fire_Burst_01}}
 {id=WavePlayer_3, properties {soundWaveAssetPtr=&/IP_CoffeePuff_Attack_03_Fire_Burst_02}}
 {id=WavePlayer_4, properties {soundWaveAssetPtr=&/IP_CoffeePuff_Attack_03_Fire_Burst_03}}
 {id=WavePlayer_5, properties {soundWaveAssetPtr=&/IP_CoffeePuff_Attack_03_Fire_Start_Scream_01}}
 {id=WavePlayer_6, properties {soundWaveAssetPtr=&/IP_CoffeePuff_Attack_03_Fire_Start_Scream_02}}
 {id=WavePlayer_7, properties {soundWaveAssetPtr=&/IP_CoffeePuff_Attack_03_Fire_Start_Scream_03}}
 {id=WavePlayer_8, properties {soundWaveAssetPtr=&/IP_CoffeePuff_Attack_03_Fire_Burst_Embers_01}}
 {id=WavePlayer_9, properties {soundWaveAssetPtr=&/IP_CoffeePuff_Attack_03_Fire_Burst_Embers_02}}
]
properties {
 bOverrideConcurrency=true
 maxDistance=2097152
 soundClassObject=&Fort_WeaponMelee
 virtualizationMode=Disabled
 volumeMultiplier=1
 concurrencyOverrides {maxCount=3}
}
