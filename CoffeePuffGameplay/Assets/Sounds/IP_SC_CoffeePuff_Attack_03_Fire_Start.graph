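# Sound cue for the start of CoffeePuff's third attack fire sound, in the melee weapon sound class.
# A random pick of three waves, with slight modulation, feeds one shared mixer for all listener variants.
# Branches on islocal and IsEnemy choose local playback, louder stereo close-range attenuation for enemies or unset, or much quieter mono attenuation for non-enemies.
# A concurrency override caps simultaneous instances at two.
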SoundCue /CoffeePuffGameplay/Assets/Sounds/IP_SC_CoffeePuff_Attack_03_Fire_Start
root=Branch_1
nodes [
 {id=Attenuation_0, children [Modulator_1]}
 {
  id=Attenuation_1
  children [Modulator_3]
  properties {attenuationSettings=/Game/Sounds/Attenuation/Fort_Weapon_Melee_Attack_Close_Range_Mono_Att}
 }
 {
  id=Attenuation_2
  children [Modulator_4]
  properties {attenuationSettings=/Game/Sounds/Attenuation/Fort_Weapon_Melee_Attack_Close_Range_Stereo_Att}
 }
 {id=Branch_1, children [Attenuation_0, Branch_3, Branch_3], properties {boolParameterName=islocal}}
 {id=Branch_3, children [Attenuation_2, Attenuation_1, Attenuation_2], properties {boolParameterName=IsEnemy}}
 {id=Mixer_0, children [Modulator_0,null], properties {inputVolume [1,1]}}
 {id=Modulator_0, children [Random_0], properties {volumeMax=1}}
 {id=Modulator_1, children [Mixer_0], properties {pitchMax=1, pitchMin=1, volumeMax=1, volumeMin=1}}
 {id=Modulator_3, children [Mixer_0], properties {pitchMax=1, pitchMin=1, volumeMax=0.2, volumeMin=0.2}}
 {id=Modulator_4, children [Mixer_0], properties {pitchMax=1, pitchMin=1, volumeMax=1.25, volumeMin=1.25}}
 {id=Random_0, children [WavePlayer_0,WavePlayer_1,WavePlayer_2], properties {bShouldExcludeFromBranchCulling=true, weights [1,1,1]}}
 {id=WavePlayer_0, properties {soundWaveAssetPtr=&/IP_CoffeePuff_Attack_03_Fire_Start_01}}
 {id=WavePlayer_1, properties {soundWaveAssetPtr=&/IP_CoffeePuff_Attack_03_Fire_Start_02}}
 {id=WavePlayer_2, properties {soundWaveAssetPtr=&/IP_CoffeePuff_Attack_03_Fire_Start_03}}
]
properties {
 bDebug=true
 bOverrideConcurrency=true
 maxDistance=2097152
 soundClassObject=&Fort_WeaponMelee
 virtualizationMode=Disabled
 volumeMultiplier=1
 attenuationOverrides {distanceAlgorithm=Logarithmic}
 concurrencyOverrides {maxCount=2}
}
