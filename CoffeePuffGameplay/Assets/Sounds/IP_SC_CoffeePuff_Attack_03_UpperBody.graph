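# Sound cue for the CoffeePuff melee attack swing (attack 03, upper body), in the melee weapon sound class with capped concurrency.
# It mixes two randomized six-wave layers, swing and a louder-weighted swing thwap, into one shared mix.
# Branches on islocal and IsEnemy route that mix: local or unset plays at full volume, while remote sounds use close-range melee attenuation, with non-enemies much quieter.
# The debug flag is left on.
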SoundCue /CoffeePuffGameplay/Assets/Sounds/IP_SC_CoffeePuff_Attack_03_UpperBody
root=Branch_0
nodes [
 {id=Attenuation_0, children [Modulator_2]}
 {
  id=Attenuation_1
  children [Modulator_3]
  properties {attenuationSettings=/Game/Sounds/Attenuation/Fort_Weapon_Melee_Attack_Close_Range_Mono_Att}
 }
 {
  id=Attenuation_2
  children [Modulator_4]
  properties {attenuationSettings=&Fort_Weapon_Melee_Attack_Close_Range_Mono_Att}
 }
 {id=Branch_0, children [Attenuation_0, Branch_3, Attenuation_0], properties {boolParameterName=islocal}}
 {id=Branch_3, children [Attenuation_2, Attenuation_1, Attenuation_2], properties {boolParameterName=IsEnemy}}
 {id=Mixer_0, children [Modulator_0,Modulator_5], properties {inputVolume [0.6,1.5]}}
 {id=Modulator_0, children [Random_0], properties {volumeMax=1}}
 {id=Modulator_2, children [Mixer_0], properties {pitchMax=1, pitchMin=1, volumeMax=1, volumeMin=1}}
 {id=Modulator_3, children [Mixer_0], properties {pitchMax=1, pitchMin=1, volumeMax=0.2, volumeMin=0.2}}
 {id=Modulator_4, children [Mixer_0], properties {pitchMax=1, pitchMin=1, volumeMax=1, volumeMin=1}}
 {id=Modulator_5, children [Random_2], properties {volumeMax=1}}
 {
  id=Random_0
  children [WavePlayer_1,WavePlayer_2,WavePlayer_3,WavePlayer_4,WavePlayer_5,WavePlayer_6]
  properties {bShouldExcludeFromBranchCulling=true, weights [1,1,1,1,1,1]}
 }
 {
  id=Random_2
  children [WavePlayer_12, WavePlayer_13, WavePlayer_14, WavePlayer_15, WavePlayer_16, WavePlayer_17]
  properties {bShouldExcludeFromBranchCulling=true, weights [1,1,1,1,1,1]}
 }
 {id=WavePlayer_1, properties {soundWaveAssetPtr=&/IP_CoffeePuff_Attack_Swing_01}}
 {id=WavePlayer_12, properties {soundWaveAssetPtr=&/IP_CoffeePuff_Attack_Swing_Thwap_01}}
 {id=WavePlayer_13, properties {soundWaveAssetPtr=&/IP_CoffeePuff_Attack_Swing_Thwap_02}}
 {id=WavePlayer_14, properties {soundWaveAssetPtr=&/IP_CoffeePuff_Attack_Swing_Thwap_03}}
 {id=WavePlayer_15, properties {soundWaveAssetPtr=&/IP_CoffeePuff_Attack_Swing_Thwap_04}}
 {id=WavePlayer_16, properties {soundWaveAssetPtr=&/IP_CoffeePuff_Attack_Swing_Thwap_05}}
 {id=WavePlayer_17, properties {soundWaveAssetPtr=&/IP_CoffeePuff_Attack_Swing_Thwap_06}}
 {id=WavePlayer_2, properties {soundWaveAssetPtr=&/IP_CoffeePuff_Attack_Swing_02}}
 {id=WavePlayer_3, properties {soundWaveAssetPtr=&/IP_CoffeePuff_Attack_Swing_03}}
 {id=WavePlayer_4, properties {soundWaveAssetPtr=&/IP_CoffeePuff_Attack_Swing_04}}
 {id=WavePlayer_5, properties {soundWaveAssetPtr=&/IP_CoffeePuff_Attack_Swing_05}}
 {id=WavePlayer_6, properties {soundWaveAssetPtr=&/IP_CoffeePuff_Attack_Swing_06}}
]
properties {
 bDebug=true
 bOverrideConcurrency=true
 maxDistance=2097152
 soundClassObject=&Fort_WeaponMelee
 attenuationOverrides {distanceAlgorithm=Logarithmic}
 concurrencyOverrides {maxCount=3}
}
