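# Sound cue for a CoffeePuff melee attack, layering a random swing with a random pitched-up big crunch.
# Both layers get random pitch modulation and feed one mixer, with the swing boosted.
# Branches on islocal and IsEnemy set volume: local full, enemy or unset slightly lower, non-enemy much quieter.
# Non-local paths use close-range melee attenuation, and a concurrency override caps simultaneous plays at three.
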SoundCue /CoffeePuffGameplay/Assets/Sounds/IP_SC_CoffeePuff_Attack_04
root=Branch_0
nodes [
 {id=Attenuation_0, children [Modulator_2]}
 {
  id=Attenuation_1
  children [Modulator_3]
  properties {attenuationSettings=/Game/Sounds/Attenuation/Fort_Weapon_Melee_Attack_Close_Range_Mono_Att}
 }
 {
  id=Attenuation_2
  children [Modulator_4]
  properties {attenuationSettings=&Fort_Weapon_Melee_Attack_Close_Range_Mono_Att}
 }
 {id=Branch_0, children [Attenuation_0, Branch_3, Branch_3], properties {boolParameterName=islocal}}
 {id=Branch_3, children [Attenuation_2, Attenuation_1, Attenuation_2], properties {boolParameterName=IsEnemy}}
 {id=Mixer_0, children [Modulator_1,Modulator_0], properties {inputVolume [1.65,1]}}
 {id=Modulator_0, children [Random_0], properties {pitchMax=1.4, pitchMin=1.2, volumeMax=1, volumeMin=0.8}}
 {id=Modulator_1, children [Random_1], properties {pitchMax=1.1, pitchMin=1, volumeMax=1}}
 {id=Modulator_2, children [Mixer_0], properties {pitchMax=1, pitchMin=1, volumeMax=1, volumeMin=1}}
 {id=Modulator_3, children [Mixer_0], properties {pitchMax=1, pitchMin=1, volumeMax=0.2, volumeMin=0.2}}
 {id=Modulator_4, children [Mixer_0], properties {pitchMax=1, pitchMin=1, volumeMax=0.9, volumeMin=0.9}}
 {
  id=Random_0
  children [WavePlayer_1,WavePlayer_2,WavePlayer_3,WavePlayer_4]
  properties {bShouldExcludeFromBranchCulling=true, weights [1,1,1,1]}
 }
 {
  id=Random_1
  children [WavePlayer_0, WavePlayer_7, WavePlayer_8, WavePlayer_9, WavePlayer_10, WavePlayer_11]
  properties {bShouldExcludeFromBranchCulling=true, weights [1,1,1,1,1,1]}
 }
 {id=WavePlayer_0, properties {soundWaveAssetPtr=&/IP_CoffeePuff_Attack_Swing_01}}
 {id=WavePlayer_1, properties {soundWaveAssetPtr=&/IP_CoffeePuff_Attack_Swing_Big_Crunch_01}}
 {id=WavePlayer_10, properties {soundWaveAssetPtr=&/IP_CoffeePuff_Attack_Swing_05}}
 {id=WavePlayer_11, properties {soundWaveAssetPtr=&/IP_CoffeePuff_Attack_Swing_06}}
 {id=WavePlayer_2, properties {soundWaveAssetPtr=&/IP_CoffeePuff_Attack_Swing_Big_Crunch_02}}
 {id=WavePlayer_3, properties {soundWaveAssetPtr=&/IP_CoffeePuff_Attack_Swing_Big_Crunch_03}}
 {id=WavePlayer_4, properties {soundWaveAssetPtr=&/IP_CoffeePuff_Attack_Swing_Big_Crunch_04}}
 {id=WavePlayer_7, properties {soundWaveAssetPtr=&/IP_CoffeePuff_Attack_Swing_02}}
 {id=WavePlayer_8, properties {soundWaveAssetPtr=&/IP_CoffeePuff_Attack_Swing_03}}
 {id=WavePlayer_9, properties {soundWaveAssetPtr=&/IP_CoffeePuff_Attack_Swing_04}}
]
properties {
 bDebug=true
 bOverrideConcurrency=true
 maxDistance=2097152
 soundClassObject=&Fort_WeaponMelee
 volumeMultiplier=0.487
 attenuationOverrides {distanceAlgorithm=Logarithmic}
 concurrencyOverrides {maxCount=3}
}
