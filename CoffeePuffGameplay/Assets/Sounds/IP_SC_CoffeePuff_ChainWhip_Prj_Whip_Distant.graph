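# Sound cue for the distant whip sound of the CoffeePuff chain whip (hookshot) projectile.
# A Random node picks with equal weight among six far hookshot whip wave variations.
# It uses the shared bullet whip attenuation and the ranged weapon projectile effects sound class.
# A concurrency override caps it at two instances, stopping the farthest first, and virtualization is disabled.
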SoundCue /CoffeePuffGameplay/Assets/Sounds/IP_SC_CoffeePuff_ChainWhip_Prj_Whip_Distant
root=Random_1
nodes [
 {
  id=Random_1
  children [WavePlayer_5, WavePlayer_6, WavePlayer_7, WavePlayer_8, WavePlayer_9, WavePlayer_12]
  properties {bShouldExcludeFromBranchCulling=true, weights [1,1,1,1,1,1]}
 }
 {id=WavePlayer_12, properties {soundWaveAssetPtr=&/IP_CoffeePuff_Hookshot_Prj_Whip_Far_06}}
 {id=WavePlayer_5, properties {soundWaveAssetPtr=&/IP_CoffeePuff_Hookshot_Prj_Whip_Far_01}}
 {id=WavePlayer_6, properties {soundWaveAssetPtr=&/IP_CoffeePuff_Hookshot_Prj_Whip_Far_02}}
 {id=WavePlayer_7, properties {soundWaveAssetPtr=&/IP_CoffeePuff_Hookshot_Prj_Whip_Far_03}}
 {id=WavePlayer_8, properties {soundWaveAssetPtr=&/IP_CoffeePuff_Hookshot_Prj_Whip_Far_04}}
 {id=WavePlayer_9, properties {soundWaveAssetPtr=&/IP_CoffeePuff_Hookshot_Prj_Whip_Far_05}}
]
properties {
 attenuationSettings=/Game/Athena/Sounds/Weapons/BulletWhips/bullet_whip_Att.Bullet_Whip_Att
 bBypassVolumeScaleForPriority=true
 bOverrideConcurrency=true
 maxDistance=2097152
 priority=0.5
 soundClassObject=&Fort_WeaponRanged_Projectile_Effects
 virtualizationMode=Disabled
 volumeMultiplier=1
 attenuationOverrides {bAttenuate=false, stereoSpread=0}
 concurrencyOverrides {maxCount=2, resolutionRule=StopFarthestThenPreventNew}
}
