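# SoundCue that plays the equip foley for the CoffeePuff item, using the weapon foley sound class and weapon reload attenuation.
# A Random node picks one of three equally weighted equip waves and is excluded from branch culling, feeding a Modulator at the root left at default settings.
# The cue plays at reduced volume and overrides concurrency to cap it at two simultaneous instances.
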SoundCue /CoffeePuffGameplay/Assets/Sounds/IP_SC_CoffeePuff_Equip
root=Modulator_0
nodes [
 {id=Modulator_0, children [Random_0]}
 {id=Random_0, children [WavePlayer_1,WavePlayer_2,WavePlayer_3], properties {bShouldExcludeFromBranchCulling=true, weights [1,1,1]}}
 {id=WavePlayer_1, properties {soundWaveAssetPtr=&/IP_CoffeePuff_Euip_01}}
 {id=WavePlayer_2, properties {soundWaveAssetPtr=&/IP_CoffeePuff_Euip_02}}
 {id=WavePlayer_3, properties {soundWaveAssetPtr=&/IP_CoffeePuff_Euip_03}}
]
properties {
 attenuationSettings=&Fort_Weapon_Reload_Att
 bOverrideConcurrency=true
 maxDistance=2250
 soundClassObject=&Fort_WeaponFoley
 volumeMultiplier=0.55
 concurrencyOverrides {maxCount=2}
}
