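# Sound cue for the CoffeePuff hookshot fire, with separate mixes for local, enemy and friendly players.
# An islocal branch plays a tell followed by briefly delayed hand foley and release layers for the local player.
# Otherwise an IsEnemy branch uses third-person attenuation on an undelayed tell and release mix, much quieter for non-enemies, defaulting to enemy when unset.
# Layers pick random wave variants with pitch modulation, and concurrency limits apply.
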SoundCue /CoffeePuffGameplay/Assets/Sounds/IP_SC_CoffeePuff_Hookshot_Fire
root=Branch_0
nodes [
 {id=Attenuation_0, children [Modulator_2]}
 {id=Attenuation_2, children [Modulator_4], properties {attenuationSettings=&/IP_ATT_CoffeePuff_Hookshot_3P}}
 {id=Attenuation_3, children [Modulator_3], properties {attenuationSettings=&IP_ATT_CoffeePuff_Hookshot_3P}}
 {id=Branch_0, children [Attenuation_0, Branch_3, Branch_3], properties {boolParameterName=islocal}}
 {id=Branch_3, children [Attenuation_2, Attenuation_3, Attenuation_2], properties {boolParameterName=IsEnemy}}
 {id=Delay_0, children [Modulator_1], properties {delayMax=0.145, delayMin=0.145}}
 {id=Delay_3, children [Modulator_0], properties {delayMax=0.1, delayMin=0.1}}
 {id=Mixer_0, children [Delay_3,Delay_0], properties {inputVolume [1,1]}}
 {id=Mixer_2, children [Modulator_8,Mixer_0], properties {inputVolume [1,1]}}
 {id=Mixer_5, children [Modulator_11,Modulator_10], properties {inputVolume [0.8,1.15]}}
 {id=Modulator_0, children [Random_6], properties {pitchMax=1.2, pitchMin=1, volumeMax=1, volumeMin=0.9}}
 {id=Modulator_1, children [Random_0], properties {pitchMax=1.5, pitchMin=1.3}}
 {id=Modulator_10, children [Random_7], properties {pitchMax=1.5, pitchMin=1.3}}
 {id=Modulator_11, children [Random_8], properties {pitchMax=1.2, pitchMin=1, volumeMax=1, volumeMin=1}}
 {id=Modulator_2, children [Mixer_2], properties {pitchMax=1, pitchMin=1, volumeMax=0.7, volumeMin=0.7}}
 {id=Modulator_3, children [Mixer_5], properties {pitchMax=1.3, pitchMin=1.1, volumeMax=0.1, volumeMin=0.1}}
 {id=Modulator_4, children [Mixer_5], properties {pitchMax=1.2, pitchMin=1.1, volumeMax=0.5, volumeMin=0.5}}
 {id=Modulator_8, children [Random_3], properties {pitchMax=1.2, pitchMin=1, volumeMax=1, volumeMin=0.7}}
 {id=Random_0, children [WavePlayer_0,WavePlayer_1,WavePlayer_2], properties {bShouldExcludeFromBranchCulling=true, weights [1,1,1]}}
 {id=Random_3, children [WavePlayer_11, WavePlayer_12, WavePlayer_13], properties {bShouldExcludeFromBranchCulling=true, weights [1,1,1]}}
 {
  id=Random_6
  children [WavePlayer_27, WavePlayer_28, WavePlayer_29, WavePlayer_30]
  properties {bShouldExcludeFromBranchCulling=true, weights [1,1,1,1]}
 }
 {id=Random_7, children [WavePlayer_21, WavePlayer_22, WavePlayer_23], properties {bShouldExcludeFromBranchCulling=true, weights [1,1,1]}}
 {id=Random_8, children [WavePlayer_24, WavePlayer_25, WavePlayer_26], properties {bShouldExcludeFromBranchCulling=true, weights [1,1,1]}}
 {id=WavePlayer_0, properties {soundWaveAssetPtr=&/IP_CoffeePuff_Hookshot_Cast_Release_01}}
 {id=WavePlayer_1, properties {soundWaveAssetPtr=&/IP_CoffeePuff_Hookshot_Cast_Release_02}}
 {id=WavePlayer_11, properties {soundWaveAssetPtr=&/IP_CoffeePuff_Hookshot_Cast_Tell_01}}
 {id=WavePlayer_12, properties {soundWaveAssetPtr=&/IP_CoffeePuff_Hookshot_Cast_Tell_02}}
 {id=WavePlayer_13, properties {soundWaveAssetPtr=&/IP_CoffeePuff_Hookshot_Cast_Tell_03}}
 {id=WavePlayer_2, properties {soundWaveAssetPtr=&/IP_CoffeePuff_Hookshot_Cast_Release_03}}
 {id=WavePlayer_21, properties {soundWaveAssetPtr=&IP_CoffeePuff_Hookshot_Cast_Release_01}}
 {id=WavePlayer_22, properties {soundWaveAssetPtr=&IP_CoffeePuff_Hookshot_Cast_Release_02}}
 {id=WavePlayer_23, properties {soundWaveAssetPtr=&IP_CoffeePuff_Hookshot_Cast_Release_03}}
 {id=WavePlayer_24, properties {soundWaveAssetPtr=&IP_CoffeePuff_Hookshot_Cast_Tell_01}}
 {id=WavePlayer_25, properties {soundWaveAssetPtr=&IP_CoffeePuff_Hookshot_Cast_Tell_02}}
 {id=WavePlayer_26, properties {soundWaveAssetPtr=&IP_CoffeePuff_Hookshot_Cast_Tell_03}}
 {id=WavePlayer_27, properties {soundWaveAssetPtr=&/IP_CoffeePuff_Hookshot_Cast_Hand_Foley_01}}
 {id=WavePlayer_28, properties {soundWaveAssetPtr=&/IP_CoffeePuff_Hookshot_Cast_Hand_Foley_02}}
 {id=WavePlayer_29, properties {soundWaveAssetPtr=&/IP_CoffeePuff_Hookshot_Cast_Hand_Foley_03}}
 {id=WavePlayer_30, properties {soundWaveAssetPtr=&/IP_CoffeePuff_Hookshot_Cast_Hand_Foley_04}}
]
properties {
 bDebug=true
 maxDistance=2097152
 soundClassObject=&Fort_WeaponMelee
 volumeMultiplier=1
 concurrencyOverrides {maxCount=3}
 concurrencySet [
  &/IP_CONC_CoffeePuff_Hookshot_Fire_Global
  &/IP_CONC_CoffeePuff_Hookshot_Fire_Local
 ]
}
