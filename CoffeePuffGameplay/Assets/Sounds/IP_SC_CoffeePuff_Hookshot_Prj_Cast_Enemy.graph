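# Sound cue for the enemy version of the CoffeePuff hookshot projectile cast, using third-person attenuation and the melee weapon sound class.
# A mixer plays one of six random enemy cast waves at reduced volume, plus a random "Retro" layer that fires about one time in ten via empty random slots.
# A top modulator adds slight pitch variation.
# Concurrency sets and a per-owner override limit stacking, and virtualization keeps it playing when silent.
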SoundCue /CoffeePuffGameplay/Assets/Sounds/IP_SC_CoffeePuff_Hookshot_Prj_Cast_Enemy
root=Modulator_4
nodes [
 {id=Mixer_0, children [Random_3], properties {inputVolume [0.75]}}
 {id=Mixer_2, children [Random_0,Random_2], properties {inputVolume [0.35,1]}}
 {id=Modulator_4, children [Mixer_2], properties {pitchMax=1.1, pitchMin=0.9, volumeMax=1, volumeMin=1}}
 {
  id=Random_0
  children [WavePlayer_1, WavePlayer_2, WavePlayer_9, WavePlayer_13, WavePlayer_14, WavePlayer_15]
  properties {bShouldExcludeFromBranchCulling=true, weights [1,1,1,1,1,1]}
 }
 {
  id=Random_2
  children [null,null,null,null,null,null,null,null,null,Mixer_0]
  properties {bShouldExcludeFromBranchCulling=true, weights [1,1,1,1,1,1,1,1,1,1]}
 }
 {id=Random_3, children [WavePlayer_10, WavePlayer_11, WavePlayer_12], properties {bShouldExcludeFromBranchCulling=true, weights [1,1,1]}}
 {id=WavePlayer_1, properties {soundWaveAssetPtr=&/IP_CoffeePuff_Hookshot_Cast_Enemy_01}}
 {id=WavePlayer_10, properties {soundWaveAssetPtr=&/IP_CoffeePuff_Hookshot_Prj_Cast_Retro_01}}
 {id=WavePlayer_11, properties {soundWaveAssetPtr=&/IP_CoffeePuff_Hookshot_Prj_Cast_Retro_02}}
 {id=WavePlayer_12, properties {soundWaveAssetPtr=&/IP_CoffeePuff_Hookshot_Prj_Cast_Retro_03}}
 {id=WavePlayer_13, properties {soundWaveAssetPtr=&/IP_CoffeePuff_Hookshot_Cast_Enemy_04}}
 {id=WavePlayer_14, properties {soundWaveAssetPtr=&/IP_CoffeePuff_Hookshot_Cast_Enemy_05}}
 {id=WavePlayer_15, properties {soundWaveAssetPtr=&/IP_CoffeePuff_Hookshot_Cast_Enemy_06}}
 {id=WavePlayer_2, properties {soundWaveAssetPtr=&/IP_CoffeePuff_Hookshot_Cast_Enemy_02}}
 {id=WavePlayer_9, properties {soundWaveAssetPtr=&/IP_CoffeePuff_Hookshot_Cast_Enemy_03}}
]
properties {
 attenuationSettings=&/IP_ATT_CoffeePuff_Hookshot_3P
 maxDistance=16000
 soundClassObject=&Fort_WeaponMelee
 virtualizationMode=PlayWhenSilent
 volumeMultiplier=1
 concurrencyOverrides {bLimitToOwner=true, maxCount=1, resolutionRule=StopOldest}
 concurrencySet [
  &/IP_CONC_CoffeePuff_Hookshot_Cast_Global
  &/IP_CONC_CoffeePuff_Hookshot_Cast_Local
 ]
}
