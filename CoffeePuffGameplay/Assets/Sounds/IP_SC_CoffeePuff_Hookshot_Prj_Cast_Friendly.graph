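# Sound cue for the CoffeePuff hookshot projectile cast, friendly variant, playing one of six cast recordings.
# An equally weighted Random node picks the wave and feeds a Modulator that shifts pitch slightly upward at random, with volume held steady.
# It uses third-person hookshot attenuation, the melee weapon sound class, and a low overall volume.
# Global and local cast concurrency sets apply, and the cue keeps playing when silent.
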SoundCue /CoffeePuffGameplay/Assets/Sounds/IP_SC_CoffeePuff_Hookshot_Prj_Cast_Friendly
root=Modulator_3
nodes [
 {id=Modulator_3, children [Random_4], properties {pitchMax=1.2, pitchMin=1.1, volumeMax=1, volumeMin=1}}
 {
  id=Random_4
  children [WavePlayer_13, WavePlayer_14, WavePlayer_15, WavePlayer_23, WavePlayer_24, WavePlayer_25]
  properties {bShouldExcludeFromBranchCulling=true, weights [1,1,1,1,1,1]}
 }
 {id=WavePlayer_13, properties {soundWaveAssetPtr=&/IP_CoffeePuff_Hookshot_Cast_01}}
 {id=WavePlayer_14, properties {soundWaveAssetPtr=&/IP_CoffeePuff_Hookshot_Cast_02}}
 {id=WavePlayer_15, properties {soundWaveAssetPtr=&/IP_CoffeePuff_Hookshot_Cast_03}}
 {id=WavePlayer_23, properties {soundWaveAssetPtr=&/IP_CoffeePuff_Hookshot_Cast_04}}
 {id=WavePlayer_24, properties {soundWaveAssetPtr=&/IP_CoffeePuff_Hookshot_Cast_05}}
 {id=WavePlayer_25, properties {soundWaveAssetPtr=&/IP_CoffeePuff_Hookshot_Cast_06}}
]
properties {
 attenuationSettings=&/IP_ATT_CoffeePuff_Hookshot_3P
 maxDistance=16000
 soundClassObject=&Fort_WeaponMelee
 virtualizationMode=PlayWhenSilent
 volumeMultiplier=0.1
 concurrencyOverrides {bLimitToOwner=true, maxCount=1, resolutionRule=StopOldest}
 concurrencySet [
  &/IP_CONC_CoffeePuff_Hookshot_Cast_Global
  &/IP_CONC_CoffeePuff_Hookshot_Cast_Local
 ]
}
