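# Sound cue for the local, first-person cast of the CoffeePuff hookshot projectile, using 1P attenuation and the melee weapon sound class.
# A top-level Random repeats the standard branch in nine of ten equally weighted slots, making the louder-mixed retro branch a rare variant.
# The standard branch randomizes six cast waves and the retro branch three, all feeding a neutral Modulator root.
# Global and local hookshot cast concurrency sets are assigned.
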SoundCue /CoffeePuffGameplay/Assets/Sounds/IP_SC_CoffeePuff_Hookshot_Prj_Cast_Local
root=Modulator_2
nodes [
 {id=Mixer_0, children [Random_3], properties {inputVolume [0.75]}}
 {id=Mixer_1, children [Random_4], properties {inputVolume [0.5]}}
 {id=Modulator_2, children [Random_1], properties {pitchMax=1, pitchMin=1, volumeMax=1, volumeMin=1}}
 {
  id=Random_1
  children [Mixer_1,Mixer_1,Mixer_1,Mixer_1,Mixer_1,Mixer_1,Mixer_1,Mixer_1,Mixer_1,Mixer_0]
  properties {bShouldExcludeFromBranchCulling=true, weights [1,1,1,1,1,1,1,1,1,1]}
 }
 {id=Random_3, children [WavePlayer_10, WavePlayer_11, WavePlayer_12], properties {bShouldExcludeFromBranchCulling=true, weights [1,1,1]}}
 {
  id=Random_4
  children [WavePlayer_13, WavePlayer_14, WavePlayer_15, WavePlayer_23, WavePlayer_24, WavePlayer_25]
  properties {bShouldExcludeFromBranchCulling=true, weights [1,1,1,1,1,1]}
 }
 {id=WavePlayer_10, properties {soundWaveAssetPtr=&/IP_CoffeePuff_Hookshot_Prj_Cast_Retro_01}}
 {id=WavePlayer_11, properties {soundWaveAssetPtr=&/IP_CoffeePuff_Hookshot_Prj_Cast_Retro_02}}
 {id=WavePlayer_12, properties {soundWaveAssetPtr=&/IP_CoffeePuff_Hookshot_Prj_Cast_Retro_03}}
 {id=WavePlayer_13, properties {soundWaveAssetPtr=&/IP_CoffeePuff_Hookshot_Cast_01}}
 {id=WavePlayer_14, properties {soundWaveAssetPtr=&/IP_CoffeePuff_Hookshot_Cast_02}}
 {id=WavePlayer_15, properties {soundWaveAssetPtr=&/IP_CoffeePuff_Hookshot_Cast_03}}
 {id=WavePlayer_23, properties {soundWaveAssetPtr=&/IP_CoffeePuff_Hookshot_Cast_04}}
 {id=WavePlayer_24, properties {soundWaveAssetPtr=&/IP_CoffeePuff_Hookshot_Cast_05}}
 {id=WavePlayer_25, properties {soundWaveAssetPtr=&/IP_CoffeePuff_Hookshot_Cast_06}}
]
properties {
 attenuationSettings=&/IP_ATT_CoffeePuff_Hookshot_Prj_Cast_1P
 maxDistance=18000
 soundClassObject=&Fort_WeaponMelee
 volumeMultiplier=1
 concurrencyOverrides {bLimitToOwner=true, maxCount=1, resolutionRule=StopOldest}
 concurrencySet [
  &/IP_CONC_CoffeePuff_Hookshot_Cast_Global
  &/IP_CONC_CoffeePuff_Hookshot_Cast_Local
 ]
}
