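# Sound cue for the CoffeePuff hookshot projectile's damage impact on a player, as heard by the local player in first person.
# A fixed short delay feeds an equally weighted random pick among four first-person impact waves, with the random node excluded from branch culling.
# It routes to the received 2D hit-notification sound class, with boosted volume and a small concurrency cap.
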SoundCue /CoffeePuffGameplay/Assets/Sounds/IP_SC_CoffeePuff_Hookshot_Prj_Damage_Local_Player
root=Delay_0
nodes [
 {id=Delay_0, children [Random_0], properties {delayMax=0.3, delayMin=0.3}}
 {
  id=Random_0
  children [WavePlayer_4,WavePlayer_5,WavePlayer_6,WavePlayer_7]
  properties {bShouldExcludeFromBranchCulling=true, weights [1,1,1,1]}
 }
 {id=WavePlayer_4, properties {soundWaveAssetPtr=&/IP_CoffeePuff_Hookshot_Prj_Impact_Player_1P_01}}
 {id=WavePlayer_5, properties {soundWaveAssetPtr=&/IP_CoffeePuff_Hookshot_Prj_Impact_Player_1P_02}}
 {id=WavePlayer_6, properties {soundWaveAssetPtr=&/IP_CoffeePuff_Hookshot_Prj_Impact_Player_1P_03}}
 {id=WavePlayer_7, properties {soundWaveAssetPtr=&/IP_CoffeePuff_Hookshot_Prj_Impact_Player_1P_04}}
]
properties {
 bDebug=true
 maxDistance=2097152
 soundClassObject=/Game/Sounds/FortSoundClassesAndModes/Fort_2DHitNotifications_Receive_PSM
 volumeMultiplier=1.25
 attenuationOverrides {distanceAlgorithm=Logarithmic}
 concurrencyOverrides {maxCount=3}
 concurrencySet [null]
}
