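# Sound cue for the CoffeePuff hookshot projectile hitting a player, choosing randomly among three impact waves.
# An islocal branch routes local playback, mixed loudest, to the 2D received-hit sound class without an attenuation node.
# Otherwise, it uses the melee weapon class with close-range attenuation, where isenemy false gives a quieter, higher-pitched variant.
# All paths share the random node, and local and global concurrency sets limit overlapping impacts.
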SoundCue /CoffeePuffGameplay/Assets/Sounds/IP_SC_CoffeePuff_Hookshot_Prj_Impact_Player
root=Branch_0
nodes [
 {
  id=Attenuation_0
  children [Modulator_4]
  properties {attenuationSettings=/Game/Sounds/Attenuation/Fort_Weapon_Melee_Attack_Close_Range_Mono_Att}
 }
 {
  id=Attenuation_2
  children [Modulator_3]
  properties {attenuationSettings=&Fort_Weapon_Melee_Attack_Close_Range_Mono_Att}
 }
 {id=Branch_0, children [SoundClass_0,SoundClass_1,SoundClass_1], properties {boolParameterName=islocal}}
 {id=Branch_3, children [Attenuation_0, Attenuation_2, Attenuation_0], properties {boolParameterName=isenemy}}
 {id=Mixer_0, children [Random_2], properties {inputVolume [2]}}
 {id=Mixer_2, children [Random_2], properties {inputVolume [0.5]}}
 {id=Mixer_3, children [Random_2], properties {inputVolume [1]}}
 {id=Modulator_2, children [Mixer_0], properties {pitchMax=1.1, pitchMin=0.9, volumeMax=1, volumeMin=1}}
 {id=Modulator_3, children [Mixer_2], properties {pitchMax=1.3, pitchMin=1.2, volumeMax=1, volumeMin=0.8}}
 {id=Modulator_4, children [Mixer_3], properties {pitchMax=1.1, pitchMin=0.9, volumeMax=1, volumeMin=1}}
 {id=Random_2, children [WavePlayer_7, WavePlayer_8, WavePlayer_10], properties {bShouldExcludeFromBranchCulling=true, weights [1,1,1]}}
 {
  id=SoundClass_0
  children [Modulator_2]
  properties {soundClassOverride=/Game/Sounds/FortSoundClassesAndModes/Fort_2DHitNotifications_Receive_PSM}
 }
 {id=SoundClass_1, children [Branch_3], properties {soundClassOverride=&Fort_WeaponMelee}}
 {id=WavePlayer_10, properties {soundWaveAssetPtr=&/IP_CoffeePuff_Hookshot_Prj_Impact_03}}
 {id=WavePlayer_7, properties {soundWaveAssetPtr=&/IP_CoffeePuff_Hookshot_Prj_Impact_01}}
 {id=WavePlayer_8, properties {soundWaveAssetPtr=&/IP_CoffeePuff_Hookshot_Prj_Impact_02}}
]
properties {
 attenuationSettings=&Fort_Weapon_Melee_Attack_Close_Range_Mono_Att
 bDebug=true
 maxDistance=10058.75
 volumeMultiplier=0.85
 attenuationOverrides {distanceAlgorithm=Logarithmic}
 concurrencyOverrides {maxCount=1}
 concurrencySet [&/IP_CONC_CoffeePuff_Impacts_Local, &/IP_CONC_CoffeePuff_Impacts_Global]
}
