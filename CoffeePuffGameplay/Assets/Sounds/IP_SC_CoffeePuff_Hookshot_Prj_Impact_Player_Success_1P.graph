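# SoundCue for the first-person success sound when CoffeePuff's hookshot projectile hits a player, routed through a 2D hit-notification sound class.
# A mixer layers three randomized, pitch-modulated variant sets: pitched-up enemy-hit waves, a quieter pitched-up impact layer, and a louder recall-catch layer.
# The catch layer is pitched down and starts after a short fixed delay.
# Concurrency allows one instance, stopping the oldest when retriggered.
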SoundCue /CoffeePuffGameplay/Assets/Sounds/IP_SC_CoffeePuff_Hookshot_Prj_Impact_Player_Success_1P
root=Mixer_0
nodes [
 {id=Delay_1, children [Modulator_3], properties {delayMax=0.19, delayMin=0.19}}
 {id=Mixer_0, children [Modulator_0,Modulator_2,Delay_1], properties {inputVolume [1,0.25,1.5]}}
 {id=Modulator_0, children [Random_1], properties {pitchMax=1.3, pitchMin=1.2, volumeMax=1, volumeMin=1}}
 {id=Modulator_2, children [Random_3], properties {pitchMax=1.3, pitchMin=1.2, volumeMax=1, volumeMin=1}}
 {id=Modulator_3, children [Random_5], properties {pitchMax=0.9, pitchMin=0.8, volumeMax=1, volumeMin=1}}
 {
  id=Random_1
  children [WavePlayer_0,WavePlayer_3,WavePlayer_4,WavePlayer_5]
  properties {bShouldExcludeFromBranchCulling=true, weights [1,1,1,1]}
 }
 {id=Random_3, children [WavePlayer_6,WavePlayer_7,WavePlayer_9], properties {bShouldExcludeFromBranchCulling=true, weights [1,1,1]}}
 {
  id=Random_5
  children [WavePlayer_19, WavePlayer_20, WavePlayer_21, WavePlayer_22]
  properties {bShouldExcludeFromBranchCulling=true, weights [1,1,1,1]}
 }
 {id=WavePlayer_0, properties {soundWaveAssetPtr=&/IP_CoffeePuff_Hookshot_Prj_Hit_Enemy_2D_01}}
 {id=WavePlayer_19, properties {soundWaveAssetPtr=&/IP_CoffeePuff_Hookshot_Prj_Recall_Catch_04}}
 {id=WavePlayer_20, properties {soundWaveAssetPtr=&/IP_CoffeePuff_Hookshot_Prj_Recall_Catch_03}}
 {id=WavePlayer_21, properties {soundWaveAssetPtr=&/IP_CoffeePuff_Hookshot_Prj_Recall_Catch_02}}
 {id=WavePlayer_22, properties {soundWaveAssetPtr=&/IP_CoffeePuff_Hookshot_Prj_Recall_Catch_01}}
 {id=WavePlayer_3, properties {soundWaveAssetPtr=&/IP_CoffeePuff_Hookshot_Prj_Hit_Enemy_2D_02}}
 {id=WavePlayer_4, properties {soundWaveAssetPtr=&/IP_CoffeePuff_Hookshot_Prj_Hit_Enemy_2D_03}}
 {id=WavePlayer_5, properties {soundWaveAssetPtr=&/IP_CoffeePuff_Hookshot_Prj_Hit_Enemy_2D_04}}
 {id=WavePlayer_6, properties {soundWaveAssetPtr=&/IP_CoffeePuff_Hookshot_Prj_Impact_03}}
 {id=WavePlayer_7, properties {soundWaveAssetPtr=&/IP_CoffeePuff_Hookshot_Prj_Impact_02}}
 {id=WavePlayer_9, properties {soundWaveAssetPtr=&/IP_CoffeePuff_Hookshot_Prj_Impact_01}}
]
properties {
 bDebug=true
 bOverrideConcurrency=true
 maxDistance=2097152
 soundClassObject=&Fort_2DHitNotifications_Instigate_PSM
 volumeMultiplier=0.9
 attenuationOverrides {distanceAlgorithm=Logarithmic}
 concurrencyOverrides {maxCount=1, resolutionRule=StopOldest}
 concurrencySet [null]
}
