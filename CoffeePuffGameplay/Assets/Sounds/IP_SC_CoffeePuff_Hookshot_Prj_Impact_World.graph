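# Sound cue for the CoffeePuff hookshot projectile hitting the world, played as an attenuated melee-weapon-class sound.
# A random node picks one of four equally weighted impact waves, and a shared modulator varies its pitch slightly.
# Branches on islocal and IsEnemy keep local, enemy and unset-allegiance impacts at full volume, while non-local friendly impacts drop sharply.
# It references separate local and global concurrency assets to limit simultaneous impacts.
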SoundCue /CoffeePuffGameplay/Assets/Sounds/IP_SC_CoffeePuff_Hookshot_Prj_Impact_World
root=Branch_0
nodes [
 {id=Branch_0, children [Modulator_2,Branch_3,Branch_3], properties {boolParameterName=islocal}}
 {id=Branch_3, children [Modulator_4,Modulator_3,Modulator_4], properties {boolParameterName=IsEnemy}}
 {id=Modulator_0, children [Random_0], properties {pitchMax=1.1, pitchMin=0.9, volumeMax=1, volumeMin=1}}
 {id=Modulator_2, children [Modulator_0], properties {pitchMax=1, pitchMin=1, volumeMax=1, volumeMin=1}}
 {id=Modulator_3, children [Modulator_0], properties {pitchMax=1, pitchMin=1, volumeMax=0.2, volumeMin=0.2}}
 {id=Modulator_4, children [Modulator_0], properties {pitchMax=1, pitchMin=1, volumeMax=1, volumeMin=1}}
 {
  id=Random_0
  children [WavePlayer_0,WavePlayer_1,WavePlayer_2,WavePlayer_3]
  properties {bShouldExcludeFromBranchCulling=true, weights [1,1,1,1]}
 }
 {id=WavePlayer_0, properties {soundWaveAssetPtr=&/IP_CoffeePuff_Hookshot_Prj_Impact_World_01}}
 {id=WavePlayer_1, properties {soundWaveAssetPtr=&/IP_CoffeePuff_Hookshot_Prj_Impact_World_02}}
 {id=WavePlayer_2, properties {soundWaveAssetPtr=&/IP_CoffeePuff_Hookshot_Prj_Impact_World_03}}
 {id=WavePlayer_3, properties {soundWaveAssetPtr=&/IP_CoffeePuff_Hookshot_Prj_Impact_World_04}}
]
properties {
 attenuationSettings=/Game/Sounds/Attenuation/Fort_Weapon_Melee_Attack_Close_Range_Mono_Att
 bDebug=true
 maxDistance=10058.75
 soundClassObject=&Fort_WeaponMelee
 volumeMultiplier=0.8
 attenuationOverrides {distanceAlgorithm=Logarithmic}
 concurrencyOverrides {maxCount=3}
 concurrencySet [
  &/IP_CONC_CoffeePuff_Hookshot_Prj_Impacts_Local
  &/IP_CONC_CoffeePuff_Hookshot_Prj_Impacts_Global
 ]
}
