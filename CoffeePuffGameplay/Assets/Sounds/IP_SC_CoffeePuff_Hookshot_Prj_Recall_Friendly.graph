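# Sound cue for the CoffeePuff hookshot projectile recall, friendly variant, played as a positional sound on the melee weapon sound class.
# A Random node picks one of four equally weighted recall waves, excluded from branch culling.
# A Modulator slightly raises pitch and varies volume, feeding a neutral root Modulator at reduced cue volume.
# Third-person hookshot attenuation shapes falloff, and global and local recall concurrency sets limit overlapping voices.
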SoundCue /CoffeePuffGameplay/Assets/Sounds/IP_SC_CoffeePuff_Hookshot_Prj_Recall_Friendly
root=Modulator_3
nodes [
 {id=Modulator_3, children [Modulator_5], properties {pitchMax=1, pitchMin=1, volumeMax=1, volumeMin=1}}
 {id=Modulator_5, children [Random_1], properties {pitchMax=1.2, pitchMin=1.1, volumeMax=1, volumeMin=0.8}}
 {
  id=Random_1
  children [WavePlayer_4, WavePlayer_5, WavePlayer_6, WavePlayer_11]
  properties {bShouldExcludeFromBranchCulling=true, weights [1,1,1,1]}
 }
 {id=WavePlayer_11, properties {soundWaveAssetPtr=&/IP_CoffeePuff_Hookshot_Prj_Recall_04}}
 {id=WavePlayer_4, properties {soundWaveAssetPtr=&/IP_CoffeePuff_Hookshot_Prj_Recall_01}}
 {id=WavePlayer_5, properties {soundWaveAssetPtr=&/IP_CoffeePuff_Hookshot_Prj_Recall_02}}
 {id=WavePlayer_6, properties {soundWaveAssetPtr=&/IP_CoffeePuff_Hookshot_Prj_Recall_03}}
]
properties {
 attenuationSettings=&/IP_ATT_CoffeePuff_Hookshot_3P
 maxDistance=16000
 soundClassObject=&Fort_WeaponMelee
 volumeMultiplier=0.3
 concurrencyOverrides {maxCount=2}
 concurrencySet [
  &/IP_CONC_CoffeePuff_Hookshot_Recall_Global
  &/IP_CONC_CoffeePuff_Hookshot_Recall_Local
 ]
}
