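# Sound cue for the CoffeePuff hookshot projectile recall, the local variant using first-person (1P) attenuation.
# A Random node picks one of four equally weighted recall waves and is excluded from branch culling.
# A Modulator at the root varies each play with slight upward pitch and reduced volume ranges.
# It routes to the melee weapon sound class and references both global and local recall concurrency assets.
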SoundCue /CoffeePuffGameplay/Assets/Sounds/IP_SC_CoffeePuff_Hookshot_Prj_Recall_Local
root=Modulator_2
nodes [
 {id=Modulator_2, children [Random_1], properties {pitchMax=1.15, pitchMin=1, volumeMax=1, volumeMin=0.8}}
 {
  id=Random_1
  children [WavePlayer_4, WavePlayer_5, WavePlayer_6, WavePlayer_11]
  properties {bShouldExcludeFromBranchCulling=true, weights [1,1,1,1]}
 }
 {id=WavePlayer_11, properties {soundWaveAssetPtr=&/IP_CoffeePuff_Hookshot_Prj_Recall_04}}
 {id=WavePlayer_4, properties {soundWaveAssetPtr=&/IP_CoffeePuff_Hookshot_Prj_Recall_01}}
 {id=WavePlayer_5, properties {soundWaveAssetPtr=&/IP_CoffeePuff_Hookshot_Prj_Recall_02}}
 {id=WavePlayer_6, properties {soundWaveAssetPtr=&/IP_CoffeePuff_Hookshot_Prj_Recall_03}}
]
properties {
 attenuationSettings=&/IP_ATT_CoffeePuff_Hookshot_Prj_Recall_1P
 maxDistance=16000
 soundClassObject=&Fort_WeaponMelee
 volumeMultiplier=1
 concurrencyOverrides {maxCount=3}
 concurrencySet [
  &/IP_CONC_CoffeePuff_Hookshot_Recall_Global
  &/IP_CONC_CoffeePuff_Hookshot_Recall_Local
 ]
}
