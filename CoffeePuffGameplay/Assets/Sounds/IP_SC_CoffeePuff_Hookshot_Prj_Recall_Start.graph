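# Sound cue named for the start of the CoffeePuff hookshot projectile recall, currently with no sound nodes and a null root, so it has nothing to play.
# Playback settings use a third-person hookshot attenuation asset and the Fort melee weapon sound class.
# It references global and local recall concurrency assets and also carries a concurrency override that caps simultaneous instances.
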SoundCue /CoffeePuffGameplay/Assets/Sounds/IP_SC_CoffeePuff_Hookshot_Prj_Recall_Start
root=null
nodes []
properties {
 attenuationSettings=&/IP_ATT_CoffeePuff_Hookshot_3P
 maxDistance=3000
 soundClassObject=&Fort_WeaponMelee
 concurrencyOverrides {maxCount=3}
 concurrencySet [
  &/IP_CONC_CoffeePuff_Hookshot_Recall_Global
  &/IP_CONC_CoffeePuff_Hookshot_Recall_Local
 ]
}
