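# Sound cue for the CoffeePuff hookshot projectile recall voice line, playing a single reel-in wave.
# An islocal branch sends local playback through an attenuation node with no assigned settings at slightly reduced volume.
# Otherwise an IsEnemy branch uses shared NPC voice attenuation, at full volume for enemies or unset, halved for non-enemies.
# A concurrency set is assigned and virtualization is disabled.
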SoundCue /CoffeePuffGameplay/Assets/Sounds/IP_SC_CoffeePuff_Hookshot_Prj_Recall_VO_B
root=Branch_0
nodes [
 {id=Attenuation_0, children [WavePlayer_0]}
 {
  id=Attenuation_1
  children [WavePlayer_0]
  properties {attenuationSettings=/Game/Sounds/Fort_Human_VO_BR/NPCs/Shared/NPC_VO_Alerted_Ethereal_Stereo_Att}
 }
 {id=Branch_0, children [Modulator_2,Branch_3,Branch_3], properties {boolParameterName=islocal}}
 {id=Branch_3, children [Modulator_4,Modulator_3,Modulator_4], properties {boolParameterName=IsEnemy}}
 {id=Modulator_2, children [Attenuation_0], properties {pitchMax=1, pitchMin=1, volumeMax=0.85, volumeMin=0.85}}
 {id=Modulator_3, children [Attenuation_1], properties {pitchMax=1, pitchMin=1, volumeMax=0.5, volumeMin=0.5}}
 {id=Modulator_4, children [Attenuation_1], properties {pitchMax=1, pitchMin=1, volumeMax=1, volumeMin=1}}
 {id=WavePlayer_0, properties {soundWaveAssetPtr=&/IP_CoffeePuff_Hookshot_VO_Reel_In_01}}
]
properties {
 maxDistance=2097152
 soundClassObject=&Fort_WeaponMelee
 virtualizationMode=Disabled
 volumeMultiplier=1.1
 concurrencyOverrides {maxCount=3}
 concurrencySet [&/IP_CONC_CoffeePuff_Hookshot_VO]
}
