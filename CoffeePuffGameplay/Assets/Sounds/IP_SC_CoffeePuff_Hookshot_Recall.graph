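# Sound cue for the CoffeePuff hookshot recall, reusing four hookshot cast hand-foley waves.
# A random pick feeds a pitched-up modulator and attenuating mixer shared by every listener path.
# Branches on islocal and isenemy give the local player the loudest level, enemies or unset a medium one, and non-enemies a quiet one, with close third-person attenuation on both remote paths.
# It uses the melee weapon sound class and overrides concurrency to cap simultaneous instances.
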SoundCue /CoffeePuffGameplay/Assets/Sounds/IP_SC_CoffeePuff_Hookshot_Recall
root=Branch_0
nodes [
 {
  id=Attenuation_0
  children [Modulator_4]
  properties {attenuationSettings=&/IP_ATT_CoffeePuff_Hookshot_3P_Close}
 }
 {id=Attenuation_1, children [Modulator_2]}
 {
  id=Attenuation_2
  children [Modulator_3]
  properties {attenuationSettings=&IP_ATT_CoffeePuff_Hookshot_3P_Close}
 }
 {id=Branch_0, children [Attenuation_1, Branch_3, Branch_3], properties {boolParameterName=islocal}}
 {id=Branch_3, children [Attenuation_0, Attenuation_2, Attenuation_0], properties {boolParameterName=isenemy}}
 {id=Mixer_0, children [Modulator_0], properties {inputVolume [0.3]}}
 {id=Modulator_0, children [Random_6], properties {pitchMax=1.3, pitchMin=1.2, volumeMax=1}}
 {id=Modulator_2, children [Mixer_0], properties {pitchMax=1.1, pitchMin=0.9, volumeMax=1.3, volumeMin=1.25}}
 {id=Modulator_3, children [Mixer_0], properties {pitchMax=1.1, volumeMax=0.2, volumeMin=0.2}}
 {id=Modulator_4, children [Mixer_0], properties {pitchMax=1.1, pitchMin=0.9, volumeMax=0.8, volumeMin=0.8}}
 {
  id=Random_6
  children [WavePlayer_27, WavePlayer_28, WavePlayer_29, WavePlayer_30]
  properties {bShouldExcludeFromBranchCulling=true, weights [1,1,1,1]}
 }
 {id=WavePlayer_27, properties {soundWaveAssetPtr=&/IP_CoffeePuff_Hookshot_Cast_Hand_Foley_01}}
 {id=WavePlayer_28, properties {soundWaveAssetPtr=&/IP_CoffeePuff_Hookshot_Cast_Hand_Foley_02}}
 {id=WavePlayer_29, properties {soundWaveAssetPtr=&/IP_CoffeePuff_Hookshot_Cast_Hand_Foley_03}}
 {id=WavePlayer_30, properties {soundWaveAssetPtr=&/IP_CoffeePuff_Hookshot_Cast_Hand_Foley_04}}
]
properties {
 bOverrideConcurrency=true
 maxDistance=2097152
 soundClassObject=&Fort_WeaponMelee
 volumeMultiplier=0.5
 concurrencyOverrides {maxCount=3}
}
