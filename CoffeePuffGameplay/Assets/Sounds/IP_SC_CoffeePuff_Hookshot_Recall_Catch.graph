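# Sound cue for catching the CoffeePuff hookshot projectile on recall, playing one of three equally weighted catch waves.
# An islocal branch routes the local player to a louder, pitch-varied close-attenuation mix, and others to an isenemy branch.
# Both third-person paths use 3P close attenuation, with non-enemies much quieter than enemies and an unset isenemy treated as enemy.
# All paths share one random wave chain, and a concurrency override limits simultaneous instances.
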SoundCue /CoffeePuffGameplay/Assets/Sounds/IP_SC_CoffeePuff_Hookshot_Recall_Catch
root=Branch_0
nodes [
 {
  id=Attenuation_0
  children [Modulator_3]
  properties {attenuationSettings=&/IP_ATT_CoffeePuff_Hookshot_3P_Close}
 }
 {
  id=Attenuation_1
  children [Modulator_2]
  properties {attenuationSettings=&/IP_ATT_CoffeePuff_Hookshot_Close}
 }
 {
  id=Attenuation_2
  children [Modulator_4]
  properties {attenuationSettings=&IP_ATT_CoffeePuff_Hookshot_3P_Close}
 }
 {id=Branch_0, children [Attenuation_1, Branch_3, Branch_3], properties {boolParameterName=islocal}}
 {id=Branch_3, children [Attenuation_2, Attenuation_0, Attenuation_2], properties {boolParameterName=isenemy}}
 {id=Modulator_1, children [Random_0], properties {volumeMax=1}}
 {id=Modulator_2, children [Modulator_1], properties {pitchMax=1.3, pitchMin=1, volumeMax=1.3, volumeMin=1.25}}
 {id=Modulator_3, children [Modulator_1], properties {pitchMax=1, pitchMin=1, volumeMax=0.2, volumeMin=0.2}}
 {id=Modulator_4, children [Modulator_1], properties {pitchMax=1.1, pitchMin=0.9, volumeMax=0.8, volumeMin=0.8}}
 {id=Random_0, children [WavePlayer_1,WavePlayer_2,WavePlayer_3], properties {bShouldExcludeFromBranchCulling=true, weights [1,1,1]}}
 {id=WavePlayer_1, properties {soundWaveAssetPtr=&/IP_CoffeePuff_Hookshot_Prj_Recall_Catch_03}}
 {id=WavePlayer_2, properties {soundWaveAssetPtr=&/IP_CoffeePuff_Hookshot_Prj_Recall_Catch_02}}
 {id=WavePlayer_3, properties {soundWaveAssetPtr=&/IP_CoffeePuff_Hookshot_Prj_Recall_Catch_01}}
]
properties {
 bDebug=true
 bOverrideConcurrency=true
 maxDistance=8000
 soundClassObject=&Fort_WeaponMelee
 volumeMultiplier=0.4
 attenuationOverrides {distanceAlgorithm=Logarithmic}
 concurrencyOverrides {maxCount=2}
}
