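# Sound cue for the CoffeePuff first-person impact on an enemy, routed to the instigator 2D hit-notification sound class.
# A Random node picks one of four equally weighted player-impact waves and feeds a Modulator that adds slight upward pitch variation.
# The cue primes on load, disables virtualization and uses the shared local CoffeePuff impacts concurrency set.
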SoundCue /CoffeePuffGameplay/Assets/Sounds/IP_SC_CoffeePuff_Impact_1P_Hit_Enemy_Default
root=Modulator_6
nodes [
 {id=Modulator_6, children [Random_0], properties {pitchMax=1.1, pitchMin=1, volumeMax=1, volumeMin=1}}
 {
  id=Random_0
  children [WavePlayer_6,WavePlayer_7,WavePlayer_8,WavePlayer_9]
  properties {bShouldExcludeFromBranchCulling=true, weights [1,1,1,1]}
 }
 {id=WavePlayer_6, properties {soundWaveAssetPtr=&/IP_CoffeePuff_impact_Player_01}}
 {id=WavePlayer_7, properties {soundWaveAssetPtr=&/IP_CoffeePuff_impact_Player_02}}
 {id=WavePlayer_8, properties {soundWaveAssetPtr=&/IP_CoffeePuff_impact_Player_03}}
 {id=WavePlayer_9, properties {soundWaveAssetPtr=&/IP_CoffeePuff_impact_Player_04}}
]
properties {
 bPrimeOnLoad=true
 maxDistance=2097152
 soundClassObject=&Fort_2DHitNotifications_Instigate_PSM
 virtualizationMode=Disabled
 volumeMultiplier=0.85
 attenuationOverrides {
  bAttenuateWithLPF=true
  distanceAlgorithm=Logarithmic
  falloffDistance=800
  lPFRadiusMax=800
  lPFRadiusMin=400
  attenuationShapeExtents {x=300}
 }
 concurrencyOverrides {bLimitToOwner=true, maxCount=1, resolutionRule=StopOldest}
 concurrencySet [&/IP_CONC_CoffeePuff_Impacts_Local]
}
