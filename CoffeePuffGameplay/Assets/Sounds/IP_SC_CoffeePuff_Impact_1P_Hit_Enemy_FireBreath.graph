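# CoffeePuff sound cue for the first-person impact heard when a fire breath hits an enemy, routed through a 2D hit notification sound class.
# It picks randomly among six equally weighted Uppercut fire impact waves, then a modulator raises the pitch by a varying amount.
# Concurrency allows one instance per owner and blocks new ones while it plays.
# The cue primes on load and has virtualization disabled.
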SoundCue /CoffeePuffGameplay/Assets/Sounds/IP_SC_CoffeePuff_Impact_1P_Hit_Enemy_FireBreath
root=Modulator_1
nodes [
 {id=Modulator_1, children [Random_3], properties {pitchMax=1.5, pitchMin=1.3}}
 {
  id=Random_3
  children [WavePlayer_0, WavePlayer_4, WavePlayer_5, WavePlayer_12, WavePlayer_13, WavePlayer_14]
  properties {bShouldExcludeFromBranchCulling=true, weights [1,1,1,1,1,1]}
 }
 {id=WavePlayer_0, properties {soundWaveAssetPtr=&/IP_CoffeePuff_Uppercut_Impact_Fire_01}}
 {id=WavePlayer_12, properties {soundWaveAssetPtr=&/IP_CoffeePuff_Uppercut_Impact_Fire_04}}
 {id=WavePlayer_13, properties {soundWaveAssetPtr=&/IP_CoffeePuff_Uppercut_Impact_Fire_05}}
 {id=WavePlayer_14, properties {soundWaveAssetPtr=&/IP_CoffeePuff_Uppercut_Impact_Fire_06}}
 {id=WavePlayer_4, properties {soundWaveAssetPtr=&/IP_CoffeePuff_Uppercut_Impact_Fire_02}}
 {id=WavePlayer_5, properties {soundWaveAssetPtr=&/IP_CoffeePuff_Uppercut_Impact_Fire_03}}
]
properties {
 bOverrideConcurrency=true
 bPrimeOnLoad=true
 maxDistance=2097152
 soundClassObject=&Fort_2DHitNotifications_Instigate_PSM
 virtualizationMode=Disabled
 volumeMultiplier=0.8
 attenuationOverrides {
  bAttenuateWithLPF=true
  distanceAlgorithm=Logarithmic
  falloffDistance=800
  lPFRadiusMax=800
  lPFRadiusMin=400
  attenuationShapeExtents {x=300}
 }
 concurrencyOverrides {bLimitToOwner=true, maxCount=1, resolutionRule=PreventNew}
}
