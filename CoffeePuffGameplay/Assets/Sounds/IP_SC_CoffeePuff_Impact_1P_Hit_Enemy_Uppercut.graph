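# Sound cue for the CoffeePuff uppercut hitting an enemy, named for first person and routed to the 2D instigator hit-notification sound class.
# A mixer layers a random crack, a quieter random fire sound after a short fixed delay, and a louder fixed stinger.
# A root modulator adds slight random pitch variation to the mix.
# Overridden concurrency allows one instance per owner, stopping the oldest, and the cue primes on load without virtualization.
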SoundCue /CoffeePuffGameplay/Assets/Sounds/IP_SC_CoffeePuff_Impact_1P_Hit_Enemy_Uppercut
root=Modulator_6
nodes [
 {id=Delay_0, children [Random_3], properties {delayMax=0.1, delayMin=0.1}}
 {id=Mixer_0, children [Random_2,Delay_0,WavePlayer_6], properties {inputVolume [1,0.5,1.5]}}
 {id=Modulator_6, children [Mixer_0], properties {pitchMax=1.1, pitchMin=0.9, volumeMax=1, volumeMin=1}}
 {id=Random_2, children [WavePlayer_0,WavePlayer_1,WavePlayer_2], properties {bShouldExcludeFromBranchCulling=true, weights [1,1,1]}}
 {
  id=Random_3
  children [WavePlayer_3, WavePlayer_4, WavePlayer_5, WavePlayer_12, WavePlayer_13, WavePlayer_14]
  properties {bShouldExcludeFromBranchCulling=true, weights [1,1,1,1,1,1]}
 }
 {id=WavePlayer_0, properties {soundWaveAssetPtr=&/IP_CoffeePuff_Uppercut_Impact_Crack_01}}
 {id=WavePlayer_1, properties {soundWaveAssetPtr=&/IP_CoffeePuff_Uppercut_Impact_Crack_02}}
 {id=WavePlayer_12, properties {soundWaveAssetPtr=&/IP_CoffeePuff_Uppercut_Impact_Fire_04}}
 {id=WavePlayer_13, properties {soundWaveAssetPtr=&/IP_CoffeePuff_Uppercut_Impact_Fire_05}}
 {id=WavePlayer_14, properties {soundWaveAssetPtr=&/IP_CoffeePuff_Uppercut_Impact_Fire_06}}
 {id=WavePlayer_2, properties {soundWaveAssetPtr=&/IP_CoffeePuff_Uppercut_Impact_Crack_03}}
 {id=WavePlayer_3, properties {soundWaveAssetPtr=&/IP_CoffeePuff_Uppercut_Impact_Fire_01}}
 {id=WavePlayer_4, properties {soundWaveAssetPtr=&/IP_CoffeePuff_Uppercut_Impact_Fire_02}}
 {id=WavePlayer_5, properties {soundWaveAssetPtr=&/IP_CoffeePuff_Uppercut_Impact_Fire_03}}
 {id=WavePlayer_6, properties {soundWaveAssetPtr=&/IP_CoffeePuff_Uppercut_Impact_Stinger_01}}
]
properties {
 bOverrideConcurrency=true
 bPrimeOnLoad=true
 maxDistance=2097152
 soundClassObject=&Fort_2DHitNotifications_Instigate_PSM
 virtualizationMode=Disabled
 volumeMultiplier=0.7
 attenuationOverrides {
  bAttenuateWithLPF=true
  distanceAlgorithm=Logarithmic
  falloffDistance=800
  lPFRadiusMax=800
  lPFRadiusMin=400
  attenuationShapeExtents {x=300}
 }
 concurrencyOverrides {bLimitToOwner=true, maxCount=1, resolutionRule=StopOldest}
}
