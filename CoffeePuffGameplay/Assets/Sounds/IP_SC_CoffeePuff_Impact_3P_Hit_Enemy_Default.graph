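# Sound cue for CoffeePuff impacts on players, randomly picking one of four waves with slight pitch variation.
# An islocal branch sends local hits to the received-hit 2D notification class at half volume, without an attenuation node.
# Other hits use the player impact class and pickaxe attenuation, where an isenemy branch lowers non-enemy volume and keeps enemy or unset at full.
# Concurrency override caps simultaneous instances per owner.
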SoundCue /CoffeePuffGameplay/Assets/Sounds/IP_SC_CoffeePuff_Impact_3P_Hit_Enemy_Default
root=Branch_0
nodes [
 {id=Attenuation_0, children [Modulator_4], properties {attenuationSettings=/Game/Sounds/Attenuation/Fort_Weapon_Pickaxe_Player_Impact_Att}}
 {id=Attenuation_2, children [Modulator_3], properties {attenuationSettings=&Fort_Weapon_Pickaxe_Player_Impact_Att}}
 {id=Branch_0, children [SoundClass_0,SoundClass_1,SoundClass_1], properties {boolParameterName=islocal}}
 {id=Branch_3, children [Attenuation_0, Attenuation_2, Attenuation_0], properties {boolParameterName=isenemy}}
 {id=Modulator_2, children [Modulator_6], properties {pitchMax=1.2, volumeMax=0.5, volumeMin=0.5}}
 {id=Modulator_3, children [Modulator_6], properties {pitchMax=1, pitchMin=1, volumeMax=0.75, volumeMin=0.75}}
 {id=Modulator_4, children [Modulator_6], properties {pitchMax=1, pitchMin=1, volumeMax=1, volumeMin=1}}
 {id=Modulator_6, children [Random_0], properties {pitchMax=1.1, pitchMin=1, volumeMax=1, volumeMin=1}}
 {
  id=Random_0
  children [WavePlayer_6,WavePlayer_7,WavePlayer_8,WavePlayer_9]
  properties {bShouldExcludeFromBranchCulling=true, weights [1,1,1,1]}
 }
 {
  id=SoundClass_0
  children [Modulator_2]
  properties {soundClassOverride=/Game/Sounds/FortSoundClassesAndModes/Fort_2DHitNotifications_Receive_PSM}
 }
 {id=SoundClass_1, children [Branch_3], properties {soundClassOverride=/Game/Sounds/FortSoundClassesAndModes/Fort_Impact_Player_PSM}}
 {id=WavePlayer_6, properties {soundWaveAssetPtr=&/IP_CoffeePuff_impact_Player_01}}
 {id=WavePlayer_7, properties {soundWaveAssetPtr=&/IP_CoffeePuff_impact_Player_02}}
 {id=WavePlayer_8, properties {soundWaveAssetPtr=&/IP_CoffeePuff_impact_Player_03}}
 {id=WavePlayer_9, properties {soundWaveAssetPtr=&/IP_CoffeePuff_impact_Player_04}}
]
properties {
 bOverrideConcurrency=true
 bPrimeOnLoad=true
 maxDistance=3512
 virtualizationMode=Disabled
 volumeMultiplier=1
 attenuationOverrides {
  bAttenuateWithLPF=true
  distanceAlgorithm=Logarithmic
  falloffDistance=800
  lPFRadiusMax=800
  lPFRadiusMin=400
  attenuationShapeExtents {x=300}
 }
 concurrencyOverrides {bLimitToOwner=true, maxCount=2}
 concurrencySet [null]
}
