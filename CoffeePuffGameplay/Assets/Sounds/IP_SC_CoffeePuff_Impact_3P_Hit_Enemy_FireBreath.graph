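# Sound cue for CoffeePuff fire-breath hit impacts, randomly picking one of six uppercut fire impact waves.
# An `islocal` branch plays the local case without an attenuation node, while other cases branch on `isenemy` into attenuation nodes sharing the pickaxe player impact asset.
# All paths converge on the same random pool, so they currently sound alike apart from attenuation.
# Concurrency allows one instance per owner and blocks new ones.
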SoundCue /CoffeePuffGameplay/Assets/Sounds/IP_SC_CoffeePuff_Impact_3P_Hit_Enemy_FireBreath
root=Branch_0
nodes [
 {id=Attenuation_0, children [Modulator_4], properties {attenuationSettings=/Game/Sounds/Attenuation/Fort_Weapon_Pickaxe_Player_Impact_Att}}
 {id=Attenuation_2, children [Modulator_3], properties {attenuationSettings=&Fort_Weapon_Pickaxe_Player_Impact_Att}}
 {id=Branch_0, children [Modulator_2,Branch_3,Branch_3], properties {boolParameterName=islocal}}
 {id=Branch_3, children [Attenuation_0, Attenuation_2, Attenuation_0], properties {boolParameterName=isenemy}}
 {id=Modulator_1, children [Random_3]}
 {id=Modulator_2, children [Modulator_1], properties {pitchMax=1, pitchMin=1, volumeMax=1, volumeMin=1}}
 {id=Modulator_3, children [Modulator_1], properties {pitchMax=1, pitchMin=1, volumeMax=1, volumeMin=1}}
 {id=Modulator_4, children [Modulator_1], properties {pitchMax=1, pitchMin=1, volumeMax=1, volumeMin=1}}
 {
  id=Random_3
  children [WavePlayer_0, WavePlayer_4, WavePlayer_5, WavePlayer_12, WavePlayer_13, WavePlayer_14]
  properties {bShouldExcludeFromBranchCulling=true, weights [1,1,1,1,1,1]}
 }
 {id=WavePlayer_0, properties {soundWaveAssetPtr=&/IP_CoffeePuff_Uppercut_Impact_Fire_01}}
 {id=WavePlayer_12, properties {soundWaveAssetPtr=&/IP_CoffeePuff_Uppercut_Impact_Fire_04}}
 {id=WavePlayer_13, properties {soundWaveAssetPtr=&/IP_CoffeePuff_Uppercut_Impact_Fire_05}}
 {id=WavePlayer_14, properties {soundWaveAssetPtr=&/IP_CoffeePuff_Uppercut_Impact_Fire_06}}
 {id=WavePlayer_4, properties {soundWaveAssetPtr=&/IP_CoffeePuff_Uppercut_Impact_Fire_02}}
 {id=WavePlayer_5, properties {soundWaveAssetPtr=&/IP_CoffeePuff_Uppercut_Impact_Fire_03}}
]
properties {
 bOverrideConcurrency=true
 bPrimeOnLoad=true
 maxDistance=3512
 soundClassObject=/Game/Sounds/FortSoundClassesAndModes/Fort_Impact_Player_PSM
 volumeMultiplier=1
 attenuationOverrides {
  bAttenuateWithLPF=true
  distanceAlgorithm=Logarithmic
  falloffDistance=800
  lPFRadiusMax=800
  lPFRadiusMin=400
  attenuationShapeExtents {x=300}
 }
 concurrencyOverrides {bLimitToOwner=true, maxCount=1, resolutionRule=PreventNew}
}
