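# Sound cue for CoffeePuff's uppercut impact, named as a third-person hit on an enemy, limited per owner by concurrency overrides.
# All paths share a mixer layering a random crack at reduced volume with a random player-impact sound, under slight random pitch variation.
# An islocal branch plays it boosted without an attenuation node.
# Otherwise, an IsEnemy branch applies uppercut attenuation, slightly louder when true or unset than when false.
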SoundCue /CoffeePuffGameplay/Assets/Sounds/IP_SC_CoffeePuff_Impact_3P_Hit_Enemy_Uppercut
root=Branch_0
nodes [
 {
  id=Attenuation_4
  children [Modulator_1]
  properties {attenuationSettings=&/IP_ATT_CoffePuff_Uppercut_Impact}
 }
 {
  id=Attenuation_5
  children [Modulator_5]
  properties {attenuationSettings=&IP_ATT_CoffePuff_Uppercut_Impact}
 }
 {id=Branch_0, children [Modulator_2,Branch_2,Branch_2], properties {boolParameterName=islocal}}
 {id=Branch_2, children [Attenuation_5, Attenuation_4, Attenuation_5], properties {boolParameterName=IsEnemy}}
 {id=Delay_0, children [Random_0]}
 {id=Mixer_0, children [Random_2,Delay_0], properties {inputVolume [0.5,1]}}
 {id=Modulator_1, children [Modulator_6], properties {pitchMax=1, pitchMin=1, volumeMax=1, volumeMin=1}}
 {id=Modulator_2, children [Modulator_6], properties {pitchMax=1, pitchMin=1, volumeMax=2.65, volumeMin=2.65}}
 {id=Modulator_5, children [Modulator_6], properties {pitchMax=1, pitchMin=1, volumeMax=1.2, volumeMin=1.2}}
 {id=Modulator_6, children [Mixer_0], properties {pitchMax=1.1, pitchMin=1, volumeMax=1, volumeMin=1}}
 {
  id=Random_0
  children [WavePlayer_6,WavePlayer_7,WavePlayer_8,WavePlayer_9]
  properties {bShouldExcludeFromBranchCulling=true, weights [1,1,1,1]}
 }
 {id=Random_2, children [WavePlayer_0,WavePlayer_1,WavePlayer_2], properties {bShouldExcludeFromBranchCulling=true, weights [1,1,1]}}
 {id=WavePlayer_0, properties {soundWaveAssetPtr=&/IP_CoffeePuff_Uppercut_Impact_Crack_01}}
 {id=WavePlayer_1, properties {soundWaveAssetPtr=&/IP_CoffeePuff_Uppercut_Impact_Crack_02}}
 {id=WavePlayer_2, properties {soundWaveAssetPtr=&/IP_CoffeePuff_Uppercut_Impact_Crack_03}}
 {id=WavePlayer_6, properties {soundWaveAssetPtr=&/IP_CoffeePuff_impact_Player_01}}
 {id=WavePlayer_7, properties {soundWaveAssetPtr=&/IP_CoffeePuff_impact_Player_02}}
 {id=WavePlayer_8, properties {soundWaveAssetPtr=&/IP_CoffeePuff_impact_Player_03}}
 {id=WavePlayer_9, properties {soundWaveAssetPtr=&/IP_CoffeePuff_impact_Player_04}}
]
properties {
 bOverrideConcurrency=true
 bPrimeOnLoad=true
 maxDistance=6512
 soundClassObject=/Game/Sounds/FortSoundClassesAndModes/Fort_Impact_Player_PSM
 virtualizationMode=Disabled
 volumeMultiplier=1
 attenuationOverrides {
  bAttenuateWithLPF=true
  distanceAlgorithm=Logarithmic
  falloffDistance=800
  lPFRadiusMax=800
  lPFRadiusMin=400
  attenuationShapeExtents {x=300}
 }
 concurrencyOverrides {bLimitToOwner=true, maxCount=3}
}
