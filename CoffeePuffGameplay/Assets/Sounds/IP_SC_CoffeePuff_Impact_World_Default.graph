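# Sound cue for the CoffeePuff pickaxe's default world impact, built from randomized impact and fire wave layers.
# A mixer blends one of four player-impact waves with a half-volume, higher-pitched uppercut fire layer, then adds slight pitch variation.
# An islocal branch plays local hits directly; otherwise isenemy selects attenuated paths, with non-enemy hits much quieter.
# Concurrency is overridden to limit simultaneous instances per owner.
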SoundCue /CoffeePuffGameplay/Assets/Sounds/IP_SC_CoffeePuff_Impact_World_Default
root=Branch_0
nodes [
 {id=Attenuation_0, children [Modulator_4], properties {attenuationSettings=/Game/Sounds/Attenuation/Fort_Weapon_Pickaxe_Player_Impact_Att}}
 {id=Attenuation_2, children [Modulator_3], properties {attenuationSettings=&Fort_Weapon_Pickaxe_Player_Impact_Att}}
 {id=Branch_0, children [Modulator_2,Branch_3,Branch_3], properties {boolParameterName=islocal}}
 {id=Branch_3, children [Attenuation_0, Attenuation_2, Attenuation_0], properties {boolParameterName=isenemy}}
 {id=Mixer_0, children [Random_0,Modulator_0], properties {inputVolume [1,0.5]}}
 {id=Modulator_0, children [Random_1], properties {pitchMax=1.4, pitchMin=1.2, volumeMax=1, volumeMin=0.5}}
 {id=Modulator_2, children [Modulator_6], properties {pitchMax=1, pitchMin=1, volumeMax=1, volumeMin=1}}
 {id=Modulator_3, children [Modulator_6], properties {pitchMax=1, pitchMin=1, volumeMax=0.2, volumeMin=0.2}}
 {id=Modulator_4, children [Modulator_6], properties {pitchMax=1, pitchMin=1, volumeMax=1, volumeMin=1}}
 {id=Modulator_6, children [Mixer_0], properties {pitchMax=1.1, pitchMin=1, volumeMax=1, volumeMin=1}}
 {
  id=Random_0
  children [WavePlayer_6,WavePlayer_0,WavePlayer_1,WavePlayer_2]
  properties {bShouldExcludeFromBranchCulling=true, weights [1,1,1,1]}
 }
 {
  id=Random_1
  children [WavePlayer_3, WavePlayer_4, WavePlayer_5, WavePlayer_13, WavePlayer_14, WavePlayer_15]
  properties {bShouldExcludeFromBranchCulling=true, weights [1,1,1,1,1,1]}
 }
 {id=WavePlayer_0, properties {soundWaveAssetPtr=&/IP_CoffeePuff_impact_Player_02}}
 {id=WavePlayer_1, properties {soundWaveAssetPtr=&/IP_CoffeePuff_impact_Player_03}}
 {id=WavePlayer_13, properties {soundWaveAssetPtr=&/IP_CoffeePuff_Uppercut_Impact_Fire_04}}
 {id=WavePlayer_14, properties {soundWaveAssetPtr=&/IP_CoffeePuff_Uppercut_Impact_Fire_05}}
 {id=WavePlayer_15, properties {soundWaveAssetPtr=&/IP_CoffeePuff_Uppercut_Impact_Fire_06}}
 {id=WavePlayer_2, properties {soundWaveAssetPtr=&/IP_CoffeePuff_impact_Player_04}}
 {id=WavePlayer_3, properties {soundWaveAssetPtr=&/IP_CoffeePuff_Uppercut_Impact_Fire_01}}
 {id=WavePlayer_4, properties {soundWaveAssetPtr=&/IP_CoffeePuff_Uppercut_Impact_Fire_02}}
 {id=WavePlayer_5, properties {soundWaveAssetPtr=&/IP_CoffeePuff_Uppercut_Impact_Fire_03}}
 {id=WavePlayer_6, properties {soundWaveAssetPtr=&/IP_CoffeePuff_impact_Player_01}}
]
properties {
 attenuationSettings=/Game/Sounds/Attenuation/Fort_Pickaxe_Impact_Att
 bOverrideConcurrency=true
 bPrimeOnLoad=true
 maxDistance=4300
 soundClassObject=/Game/Sounds/FortSoundClassesAndModes/Fort_Impact_Player_PSM
 volumeMultiplier=1
 attenuationOverrides {
  bAttenuateWithLPF=true
  distanceAlgorithm=Logarithmic
  falloffDistance=800
  lPFRadiusMax=800
  lPFRadiusMin=400
  attenuationShapeExtents {x=300}
 }
 concurrencyOverrides {bLimitToOwner=true, maxCount=3}
 concurrencySet [&Player_Melee_Weapon_Conc]
}
