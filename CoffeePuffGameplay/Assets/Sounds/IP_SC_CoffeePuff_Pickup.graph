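# Sound cue for the pickup sound of the CoffeePuff item.
# A single wave player feeds a modulator at the root, which applies pitch and volume modulation to the one pickup wave.
# The cue uses the weapon pickup attenuation and weapon foley sound class at reduced volume, with a concurrency override that caps simultaneous plays.
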SoundCue /CoffeePuffGameplay/Assets/Sounds/IP_SC_CoffeePuff_Pickup
root=Modulator_0
nodes [
 {id=Modulator_0, children [WavePlayer_0]}
 {id=WavePlayer_0, properties {soundWaveAssetPtr=&/IP_CoffeePuff_Pickup_01}}
]
properties {
 attenuationSettings=&Fort_Weapon_Pickup_Att
 bOverrideConcurrency=true
 maxDistance=3500
 soundClassObject=&Fort_WeaponFoley
 volumeMultiplier=0.3
 concurrencyOverrides {maxCount=2}
}
