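# Sound cue for the looping ambient idle sound of the CoffeePuff pickup.
# A looping wave player feeds an Enveloper whose volume curve fades the loop in from silence over a few seconds.
# The cue plays quieter and lower-pitched, using shared item ambient attenuation, the loot ambient loop sound class and a loot ambient concurrency set.
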SoundCue /CoffeePuffGameplay/Assets/Sounds/IP_SC_CoffeePuff_Pickup_Idle_Loop
root=Enveloper_0
nodes [
 {
  id=Enveloper_0
  children [WavePlayer_0]
  properties {pitchCurve {editorCurveData {keys [{time=0}]}}, volumeCurve {editorCurveData {keys [<at 0 0>,<at 2.5 1>]}}}
 }
 {id=WavePlayer_0, properties {bLooping=true, soundWaveAssetPtr=&/IP_CoffeePuff_Pickup_Idle_Loop}}
]
properties {
 attenuationSettings=/Game/Sounds/Attenuation/Fort_Item_Ambient_Loop_Mono_Default_Att
 maxDistance=1250
 pitchMultiplier=0.8
 soundClassObject=/Game/Sounds/FortSoundClassesAndModes/Fort_Loot_Ambient_Loop
 volumeMultiplier=0.3
 concurrencyOverrides {maxCount=2}
 concurrencySet [/Game/Sounds/Concurrency_Rules/Loot_Ambient_Conc]
}
