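; Character part modifier for CoffeePuff that manages a Niagara equip effect on the player pawn's left forearm.
; A tick after res-in, it attaches the effect to the lowerarm_l bone and stores it as the left-hand idle FX.
; Another tick later, it resets the relative rotation of the component returned by GetIdleVFXComponent.
; Switching to a non-CoffeePuff weapon deactivates the stored effect, with no reactivation here.
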(event EventOnPostResIn
  (|CallParentFunction)
  (Utilities|FlowControl|DelayUntilNextTick)
  (bind _returnvalue (PartModifier|SpawnSystem|SpawnSystemAttachedPartModifier "/CoffeePuffGameplay/Assets/VFX/NS_CoffeePuff_Equip.NS_CoffeePuff_Equip" (CustomCharacter|GetSkeletalMeshForPartOnPlayerPawn) "lowerarm_l" "0, 0, 0" "90.000000,0.000000,0.000000" "SnapToTarget")
    (:Success
      (Variables|Default|SetLeftHandIdleFX _returnvalue)
      (Utilities|FlowControl|DelayUntilNextTick)
      (Transformation|SetRelativeRotation (CallFunction|GetIdleVFXComponent) "-90.000000,0.000000,0.000000" false false))))

(event EventOnWeaponChanged
  (params (NewWeapon :type Fort%20Weapon%20Object%20Reference) (OldWeapon :type Fort%20Weapon%20Object%20Reference))
  (bind (_asb_coffee_puff_weapon _bsuccess) (Utilities|Casting|CastToB_CoffeePuff_Weapon NewWeapon))
  (|CallParentFunction NewWeapon OldWeapon)
  (if (not _bsuccess)
    (Components|Activation|Deactivate (Variables|Default|GetLeftHandIdleFX))))

(fn ConstructionScript
  (|CallParentFunction))
