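# Niagara system for the CoffeePuff combo fire breath, with flame sprites, GPU embers, an energy wave and optional head and eye fire.
# User.FireBreathLengthLerp scales jet and ember speeds, while System.BreathDelayTime delays the main breath layers.
# At the user-supplied touch point, flames spread along the surface normal, a flickering light flashes, and terrain hits can add a fading decal.
# A player-pawn hit flag shrinks flames, and time-of-day scalars adjust brightness.
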NiagaraSystem /CoffeePuffGameplay/Assets/VFX/NS_CoffeePuff_Combo_FireBreath
dependencies {
 usedDataInterfaces [
  &NiagaraDataInterfaceCurve
  &NiagaraDataInterfaceColorCurve
  &NiagaraDataInterfaceVectorCurve
  &NiagaraDataInterfaceVector2DCurve
  &NiagaraDataInterfaceArrayFloat3
  &NiagaraDataInterfaceArrayFloat
  &NiagaraDataInterfaceSkeletalMesh
 ]
 usedDynamicInputs [&FN_LOD_Distance_Remap, &VectorFromCurve]
 usedModules [
  &self:SystemSpawnScript.SystemScriptSource.SystemScriptGraph.NiagaraNodeAssignment_0.SetVariables_#1
  &self:SystemSpawnScript.SystemScriptSource.SystemScriptGraph.NiagaraNodeAssignment_2.SetVariables_#2
  &self:SystemSpawnScript.SystemScriptSource.SystemScriptGraph.NiagaraNodeAssignment_3.SetVariables_#3
  &SystemState
  &self:SystemSpawnScript.SystemScriptSource.SystemScriptGraph.NiagaraNodeAssignment_4.SetVariables_#4
  &self:SystemSpawnScript.SystemScriptSource.SystemScriptGraph.NiagaraNodeAssignment_1.SetVariables_#5
 ]
}
emitters {
 Embers_Burst {
  simTarget=GPUComputeSim
  emitterUpdateScript [
   &EmitterState(
    Life Cycle Mode=::NewEnumerator1
    Loop Behavior=::NewEnumerator1
    Loop Delay=$System.BreathDelayTime
    Loop Duration Mode=::NewEnumerator1
    UseLoopDelay=true
   )
   &SpawnBurst_Instantaneous(Spawn Count=50, Use Spawn Probability=false, UseLoopCountLimit=false)
   &SpawnRate(SpawnRate=1000, Use Spawn Probability=false)
   &SpawnRate(SpawnRate=5000, Use Spawn Probability=false) {enabled=false, module=SpawnRate001}
  ]
  particleSpawnScript [
   &InitializeParticle(
    Color=<rgba 1 1 1 2>
    Lifetime=&UniformRangedFloat(Minimum=0.3, Maximum=1.2)
    Mass Mode=::NewEnumerator1
    Sprite Size=&Vector2DFromFloat(Value=&FloatFromCurve(
      FloatCurve=DataInterface() {propertyValues {Curve {keys [<at 0 1.2>,<at 0.1257 0.9847>,<at 0.2996 0.5>,<at 1 0.3>]}}}
      CurveIndex=&UniformRangedFloat(Minimum=0, Maximum=1)
      Scale Curve=2.5
     ))
    Sprite Size Mode=::NewEnumerator4
    UsePositionOffset=false
    Write Material Random=true
   ) {module=InitializeParticle001}
   &ShapeLocation(MaskSpawnGroup=false, Override Seed=false, Sphere Radius=10, UseOwnerScale=true)
   &AddVelocity(
    Module.Use Falloff Distance=false
    Origin Offset=<-150 0 0>
    Override Seed=false
    Velocity Mode=::NewEnumerator1
    Velocity Speed=&Multiply_Float(
     A=&RandomRangeFloat(Minimum=0.2, Maximum=1)
     B=&Lerp_Float(A=0, B=800, Alpha=$User.FireBreathLengthLerp)
    )
   )
   &AddVelocity(
    Cone Angle=&RandomRangeFloat(Minimum=25, Maximum=90)
    Module.UseConeFalloff=false
    Module.UseVelDistribution=false
    Override Seed=false
    Velocity Mode=::NewEnumerator2
    Velocity Speed=&Multiply_Float(
     A=&RandomRangeFloat(Minimum=0.5, Maximum=1)
     B=&Lerp_Float(A=300, B=1600, Alpha=$User.FireBreathLengthLerp)
    )
   ) {module=AddVelocity001}
  ]
  particleUpdateScript [
   &ParticleState()
   &ScaleSpriteSize(
    Scale Sprite Size Mode=::NewEnumerator0
    Uniform Scale Factor=&Lerp_Float(A=1, B=2.5, Alpha=$System.DistanceScalar)
   )
   &ScaleSpriteSize(Uniform Curve Sprite Scale=DataInterface() {propertyValues {Curve {keys [{arriveTangent=2.4391, interpMode=RCIM_Cubic, leaveTangent=2.4391, tangentMode=RCTM_User, time=0, value=0.325}, <at 1 1 RCIM_Cubic>]}}}) {module=ScaleSpriteSize001}
   &AccelerationForce(Acceleration=&UniformRangedVector(Minimum=<0 0 800>, Maximum=<0 0 200>))
   &Drag(
    Drag=&Multiply_Float(
     A=&FloatFromCurve(
      FloatCurve=DataInterface() {propertyValues {Curve {keys [<at 0 0.5>, {arriveTangent=0.5, interpMode=RCIM_Cubic, leaveTangent=0.5, time=0.3, value=0.5}, <at 1 1 RCIM_Cubic>]}}}
      Scale Curve=10
     )
     B=1
    )
    Use Linear Drag=true
    Use Rotational Drag=false
   )
   /Niagara/Modules/Update/Forces/CurlNoiseForce(
    Noise Frequency=10
    Noise Strength=&FloatFromCurve(FloatCurve=DataInterface() {propertyValues {Curve {keys [<at 0 1>,<at 1 1>]}}}, Scale Curve=3000)
    Pan Noise=true
    Pan Noise Field=<0.1 0.1 0.5>
   )
   &GravityForce(Gravity=&RandomRangeVector(Minimum=<0 0 900>, Maximum=<0 0 -500>))
   &SolveForcesAndVelocity(Clamp Velocity=false, Limit Acceleration=false)
   &ScaleSpriteSizeBySpeed(
    Max Scale Factor=<0.8 4>
    Min Scale Factor=<0.8 0.5>
    Sample Scale Factor By Curve=true
    Scale Factor Curve=DataInterface() {propertyValues {Curve {keys [
        <at -0.0077 0.0765 RCIM_Cubic>
        {arriveTangent=1.2463, interpMode=RCIM_Cubic, leaveTangent=1.2463, tangentMode=RCTM_User, time=0.9985, value=0.9549}
       ]}}}
    Velocity Threshold=300
   )
   &ScaleColor(
    Scale Alpha=&FloatFromCurve(FloatCurve=DataInterface() {propertyValues {Curve {keys [<at 0.25 1>,<at 1 0>]}}}, Scale Curve=$System.TOD)
    Scale RGB=&MakeVectorFromLinearColorRGB(Color=&ColorFromCurve(ColorCurve=DataInterface() {propertyValues {
        AlphaCurve {keys [<at 0 1 RCIM_Cubic>,<at 0.7107 1>]}
        BlueCurve {keys [<at 0 14.6701 RCIM_Cubic>,<at 0.368 3.776>,<at 0.855 0.1753 RCIM_Cubic>]}
        GreenCurve {keys [<at 0 36.7481 RCIM_Cubic>,<at 0.368 13.2325>,<at 0.855 0.3243 RCIM_Cubic>]}
        RedCurve {keys [<at 0 50 RCIM_Cubic>,<at 0.368 25>,<at 0.855 1 RCIM_Cubic>]}
       }}))
    ScaleA=true
    ScaleRGB=true
   )
  ]
  rendererClasses [&NiagaraSpriteRendererProperties]
  renderers [{rendererClass=&NiagaraSpriteRendererProperties}]
 }
 Embers_Burst001 {
  bEnabled=false
  simTarget=GPUComputeSim
  emitterUpdateScript [
   &EmitterState(Life Cycle Mode=::NewEnumerator1, Loop Behavior=::NewEnumerator1, Loop Duration=0.3, UseLoopDelay=false)
   &SpawnBurst_Instantaneous(Spawn Count=60, Use Spawn Probability=false, UseLoopCountLimit=false)
   &SpawnBurst_Instantaneous(
    Spawn Count=50
    Spawn Group=1
    Spawn Probability=&MakeCustomFloatFromBool(Bool=$User.IsHittingPlayer, True Float=1, False Float=0)
    Spawn Time=$System.SubSpawnDelay
    Use Spawn Probability=true
    UseLoopCountLimit=false
   ) {enabled=false, module=SpawnBurst_Instantaneous001}
   &SpawnBurst_Instantaneous(Spawn Count=50, Spawn Group=2, Use Spawn Probability=true, UseLoopCountLimit=false) {enabled=false, module=SpawnBurst_Instantaneous002}
   &SpawnRate(SpawnGroup=3, SpawnRate=50, Use Spawn Probability=true)
  ]
  particleSpawnScript [
   &InitializeParticle(
    Color=<rgba 1 1 1 2>
    Lifetime=&UniformRangedFloat(Minimum=0.3, Maximum=0.8)
    Mass Mode=::NewEnumerator1
    Sprite Size=&Vector2DFromFloat(Value=&FloatFromCurve(
      FloatCurve=DataInterface() {propertyValues {Curve {keys [<at 0 1.2>,<at 0.1257 0.9847>,<at 0.2996 0.5>,<at 1 0.3>]}}}
      CurveIndex=&UniformRangedFloat(Minimum=0, Maximum=1)
      Scale Curve=2
     ))
    Sprite Size Mode=::NewEnumerator4
    UsePositionOffset=false
    Write Material Random=true
   ) {module=InitializeParticle001}
   &ShapeLocation(MaskSpawnGroup=false, Override Seed=false, Sphere Radius=5, UseOwnerScale=true)
   &AddVelocity(
    Module.Use Falloff Distance=false
    Override Seed=false
    Velocity Mode=::NewEnumerator1
    Velocity Speed=&Multiply_Float(A=&RandomRangeFloat(Minimum=0.5, Maximum=1), B=1000)
   ) {enabled=false}
   &AddVelocity(
    Cone Angle=&SelectFloatFromArray(Direct Array Index=$Emitter.SpawnGroup, Float Selection Array=DataInterface() {propertyValues {FloatData [45,45,360,360]}})
    Cone Axis=&MakeVector(
     X=&SelectFloatFromArray(Direct Array Index=$Emitter.SpawnGroup, Float Selection Array=DataInterface() {propertyValues {FloatData [1,-1,1,1]}})
     Y=0
     Z=0
    )
    Module.UseConeFalloff=false
    Module.UseVelDistribution=false
    Override Seed=false
    Velocity Mode=::NewEnumerator2
    Velocity Speed=&Multiply_Float(
     A=&RandomRangeFloat(Minimum=0.5, Maximum=1)
     B=&Multiply_Float(
      A=1000
      B=&SelectFloatFromArray(Direct Array Index=$Emitter.SpawnGroup, Float Selection Array=DataInterface() {propertyValues {FloatData [1,0.5,1,0.5]}})
     )
    )
   ) {module=AddVelocity002}
  ]
  particleUpdateScript [
   &ParticleState()
   &InheritSourceMovement(
    Applied Position Scale=&SelectVectorFromArray(
     Direct Array Index=&SpawnGroup()
     Vector Selection Array=DataInterface() {propertyValues {FloatData [<0 0 0>, <0 0 0>, <0 0 0>, <1 1 1>]}}
    )
    ApplyForce=false
    ApplyPosition=true
    ApplyPositionFalloff=false
   )
   &ScaleSpriteSize(
    Scale Sprite Size Mode=::NewEnumerator0
    Uniform Scale Factor=&Lerp_Float(A=1, B=2.5, Alpha=$System.DistanceScalar)
   )
   &ScaleSpriteSize(
    Scale Sprite Size Mode=::NewEnumerator0
    Uniform Scale Factor=&Lerp_Float(A=0.4, B=1, Alpha=$System.DistanceScalarShortDis)
   ) {module=ScaleSpriteSize001}
   &Drag(
    Drag=&Multiply_Float(
     A=&FloatFromCurve(
      FloatCurve=DataInterface() {propertyValues {Curve {keys [<at 0 0.5>, {arriveTangent=0.5, interpMode=RCIM_Cubic, leaveTangent=0.5, time=0.3, value=0.5}, <at 1 1 RCIM_Cubic>]}}}
      Scale Curve=12
     )
     B=1
    )
    Use Linear Drag=true
    Use Rotational Drag=false
   )
   &CurlNoiseForce(
    Noise Frequency=10
    Noise Strength=&FloatFromCurve(FloatCurve=DataInterface() {propertyValues {Curve {keys [<at 0 1>,<at 1 0>]}}}, Scale Curve=3000)
    Pan Noise=true
    Pan Noise Field=<0.1 0.1 0.5>
   )
   &GravityForce(Gravity=&RandomRangeVector(Minimum=<0 0 900>, Maximum=<0 0 -500>))
   &SolveForcesAndVelocity(Clamp Velocity=false, Limit Acceleration=false)
   &ScaleSpriteSizeBySpeed(
    Max Scale Factor=<0.8 5>
    Min Scale Factor=<0.8 0.8>
    Sample Scale Factor By Curve=true
    Scale Factor Curve=DataInterface() {propertyValues {Curve {keys [
        <at -0.0077 0.0765 RCIM_Cubic>
        {arriveTangent=1.2463, interpMode=RCIM_Cubic, leaveTangent=1.2463, tangentMode=RCTM_User, time=0.9985, value=0.9549}
       ]}}}
    Velocity Threshold=300
   )
   &ScaleColor(
    Scale Alpha=&FloatFromCurve(FloatCurve=DataInterface() {propertyValues {Curve {keys [<at 0.25 1>,<at 1 0>]}}})
    Scale RGB=&MakeVectorFromLinearColorRGB(Color=&ColorFromCurve(ColorCurve=DataInterface() {propertyValues {
        AlphaCurve {keys [<at 0 1 RCIM_Cubic>,<at 0.7107 1>]}
        BlueCurve {keys [<at 0 14.6701 RCIM_Cubic>,<at 0.368 3.776>,<at 0.855 0.1753 RCIM_Cubic>]}
        GreenCurve {keys [<at 0 36.7481 RCIM_Cubic>,<at 0.368 13.2325>,<at 0.855 0.3243 RCIM_Cubic>]}
        RedCurve {keys [<at 0 50 RCIM_Cubic>,<at 0.368 25>,<at 0.855 1 RCIM_Cubic>]}
       }}))
    ScaleA=true
    ScaleRGB=true
   )
  ]
  rendererClasses [&NiagaraSpriteRendererProperties]
  renderers [{rendererClass=&NiagaraSpriteRendererProperties}]
 }
 EnergyWave {
  emitterUpdateScript [
   &EmitterState(Life Cycle Mode=::NewEnumerator1, Loop Behavior=::NewEnumerator1, Loop Duration=1, UseLoopDelay=false)
   &SpawnBurst_Instantaneous(Spawn Count=2, Spawn Time=0.02, Use Spawn Probability=false)
  ]
  particleSpawnScript [
   &PartitionParticles()
   &InitializeParticle(
    Color Mode=::NewEnumerator0
    Lifetime=0.3
    Mesh Scale=&SelectVectorFromArray(
     Direct Array Index=$Particles.PartitionParticles.Partition
     Vector Selection Array=DataInterface() {propertyValues {FloatData [<-0.05 0.1 0.1>, <-0.12 0.08 0.08>]}}
    )
    Mesh Scale Mode=::NewEnumerator4
    UsePositionOffset=false
    Write Material Random=true
   )
   &InitialMeshOrientation(
    Orientation Axis=<0 0 1>
    Orientation Vector=<0 0 1>
    Rotation=&RandomRangeVector(Minimum=<0 0 0>, Maximum=<1 0 0>)
    Use Orientation Vector=true
    Use Rotation Vector=true
   )
   &CameraOffset()
   &SystemLocation(MaskSpawnGroup=false, Offset=<-10 0 0>)
  ]
  particleUpdateScript [
   &ParticleState()
   &ScaleColor(
    Linear Color Curve=DataInterface() {propertyValues {
      AlphaCurve {keys [<at 0 1>,<at 0.3576 1>,<at 1 0>]}
      BlueCurve {keys [<at 0 0>,<at 0.1684 0.002>,<at 0.3654 0.0199>,<at 1 0>]}
      GreenCurve {keys [<at 0 0.4051>,<at 0.1684 0.2138>,<at 0.3654 0.1104>,<at 1 0.0063>]}
      RedCurve {keys [<at 0 1>,<at 0.1684 1>,<at 0.3654 0.4479>,<at 1 0.3438>]}
     }}
    Scale Mode=::NewEnumerator2
   )
   &DynamicMaterialParameters(
    Index 0 Param 1=&FloatFromCurve(FloatCurve=DataInterface() {propertyValues {Curve {keys [{arriveTangent=-2.1015, interpMode=RCIM_Cubic, leaveTangent=-2.1015, tangentMode=RCTM_User, time=0, value=0.5}, <at 1 -0.15>]}}})
    Index 0 Param 2=&FloatFromCurve(FloatCurve=DataInterface() {propertyValues {Curve {keys [{arriveTangent=-0.3464, interpMode=RCIM_Cubic, leaveTangent=-0.3464, tangentMode=RCTM_User, time=0, value=1}, <at 1 0.9>]}}})
    Index 0 Param 3=&FloatFromCurve(FloatCurve=DataInterface() {propertyValues {Curve {keys [
         {interpMode=RCIM_Cubic, tangentMode=RCTM_Break, time=0, value=0}
         {arriveTangent=0.4478, interpMode=RCIM_Cubic, leaveTangent=0.4478, tangentMode=RCTM_User, time=0.15, value=1}
         {interpMode=RCIM_Cubic, tangentMode=RCTM_Break, time=1, value=1.2}
        ]}}})
    Module.Param0WriteEnabled=true
    Module.Param1WriteEnabled=true
    Module.Param2WriteEnabled=true
    Module.Param3WriteEnabled=false
   )
   &ScaleColor(
    Scale RGB=&VectorFromFloat(Value=&FloatFromCurve(
      FloatCurve=DataInterface() {propertyValues {Curve {keys [
          {interpMode=RCIM_Cubic, tangentMode=RCTM_Break, time=0, value=10}
          {arriveTangent=-6.6072, interpMode=RCIM_Cubic, leaveTangent=-6.6072, tangentMode=RCTM_User, time=0.4, value=2}
          {interpMode=RCIM_Cubic, tangentMode=RCTM_Break, time=1, value=1}
         ]}}}
      Scale Curve=2
     ))
    ScaleA=true
    ScaleRGB=true
   ) {module=ScaleColor001}
   &ScaleMeshSize(Scale Factor=&VectorFromFloat(Value=&FloatFromCurve(
      FloatCurve=DataInterface() {propertyValues {Curve {keys [
          {interpMode=RCIM_Cubic, tangentMode=RCTM_Break, time=0, value=0.2}
          {arriveTangent=2.9365, interpMode=RCIM_Cubic, leaveTangent=2.9365, tangentMode=RCTM_User, time=0.2, value=0.875}
          {interpMode=RCIM_Cubic, tangentMode=RCTM_Break, time=1, value=2}
         ]}}}
      Scale Curve=1.6
     )))
  ]
  rendererClasses [&NiagaraMeshRendererProperties]
  renderers [{rendererClass=&NiagaraMeshRendererProperties, rendererIndex=0}]
 }
 FireFast_V_aligned {
  emitterUpdateScript [
   &EmitterState(Delay First Loop Only=true, Life Cycle Mode=::NewEnumerator1, Loop Delay=$System.BreathDelayTime, Loop Duration=1, UseLoopDelay=true)
   &SpawnRate(SpawnRate=25, Use Spawn Probability=false)
  ]
  particleSpawnScript [
   &InitializeParticle(
    Lifetime Max=0.45
    Lifetime Min=0.3
    Lifetime Mode=::NewEnumerator1
    Sprite Size=<50 200>
    Sprite Size Mode=::NewEnumerator4
    Sprite UV Mode=::NewEnumerator2
    UsePositionOffset=false
    Write Material Random=true
   )
   &DynamicMaterialParameters(
    Index 0 Param 2=&RandomRangeFloat(Minimum=0, Maximum=4)
    Index 0 Param 4=2.5
    Module.Param0WriteEnabled=false
    Module.Param1WriteEnabled=true
    Module.Param2WriteEnabled=false
    Module.Param3WriteEnabled=true
   ) {module=DynamicMaterialParameters001}
   &AddVelocity(Velocity=&MakeVector(X=&Lerp_Float(A=50, B=1200, Alpha=$User.FireBreathLengthLerp), Y=0, Z=0)) {module=AddVelocity001}
  ]
  particleUpdateScript [
   &ParticleState()
   &Drag(Drag=3, Use Linear Drag=true, Use Rotational Drag=false)
   &ScaleColor(
    Linear Color Curve=DataInterface() {propertyValues {
      AlphaCurve {keys [<at 0 1>,<at 1 0>]}
      BlueCurve {keys [<at 0.0419 11.2101>,<at 0.2881 2.242>,<at 0.6965 0.565>,<at 0.9695 2.1713>]}
      GreenCurve {keys [<at 0.0419 14.4749>,<at 0.2881 2.895>,<at 0.6965 0.6614>,<at 0.9695 3.074>]}
      RedCurve {keys [<at 0.0419 20>,<at 0.2881 4>,<at 0.6965 1>,<at 0.9695 5>]}
     }}
    Scale Mode=::NewEnumerator2
   ) {module=ScaleColor001}
   &ScaleColor(
    Scale Alpha=&FloatFromCurve(
     FloatCurve=DataInterface() {propertyValues {Curve {keys [
         {arriveTangent=6.7724, interpMode=RCIM_Cubic, leaveTangent=6.7724, tangentMode=RCTM_User, time=0, value=0}
         {interpMode=RCIM_Cubic, leaveTangent=-5.6421, tangentMode=RCTM_Break, time=0.5, value=1}
         {arriveTangent=-2, interpMode=RCIM_Cubic, leaveTangent=-2, time=0.7, value=0.2}
         {arriveTangent=-0.1613, interpMode=RCIM_Cubic, leaveTangent=-0.1613, tangentMode=RCTM_User, time=1, value=0}
        ]}}}
     Scale Curve=&Multiply_Float(A=1, B=$System.TOD)
    )
    ScaleA=true
    ScaleRGB=true
   )
   &SolveForcesAndVelocity(Clamp Velocity=false, Limit Acceleration=false)
   &ScaleSpriteSize(Scale Factor=&Multiply_Vector2D_ByFloat(
     Float=&FloatFromCurve(
      FloatCurve=DataInterface() {propertyValues {Curve {keys [
          {arriveTangent=1.4708, interpMode=RCIM_Cubic, leaveTangent=1.4708, tangentMode=RCTM_User, time=0, value=0.525}
          {arriveTangent=0.625, interpMode=RCIM_Cubic, leaveTangent=0.625, time=0.7, value=1.05}
          {interpMode=RCIM_Cubic, tangentMode=RCTM_User, time=1, value=1.15}
         ]}}}
      Scale Curve=&MakeCustomFloatFromBool(Bool=$User.IsHitPlayerPawn, True Float=1.1, False Float=1.5)
     )
     Vector2D=<0.8 1>
    )) {module=ScaleSpriteSize001}
   &ScaleSpriteSize(
    Non-Uniform Curve Scale=<3 1>
    Non-Uniform Curve Sprite Scale=DataInterface() {propertyValues {
      XCurve {keys [
        {arriveTangent=2.6926, interpMode=RCIM_Cubic, leaveTangent=2.6926, tangentMode=RCTM_User, time=0, value=0}
        {arriveTangent=0.1227, interpMode=RCIM_Cubic, leaveTangent=0.1227, tangentMode=RCTM_User, time=1, value=1}
       ]}
      YCurve {keys [
        {arriveTangent=0.2869, interpMode=RCIM_Cubic, leaveTangent=0.2869, tangentMode=RCTM_User, time=0, value=0.85}
        {tangentMode=RCTM_User, time=1, value=1}
       ]}
     }}
    Scale Sprite Size Mode=::NewEnumerator3
   ) {module=ScaleSpriteSize002}
  ]
  rendererClasses [&NiagaraSpriteRendererProperties]
  renderers [{rendererClass=&NiagaraSpriteRendererProperties}]
 }
 FireOnEyes {
  emitterUpdateScript [
   &EmitterState(Life Cycle Mode=::NewEnumerator1, Loop Duration=0.5, UseLoopDelay=false)
   &SpawnBurst_Instantaneous(
    Spawn Count=2
    Spawn Probability=&MakeCustomFloatFromBool(Bool=$User.HideHeadFX, True Float=0, False Float=1)
    Use Spawn Probability=true
    UseLoopCountLimit=false
   )
   &SpawnRate(
    Spawn Probability=&MakeCustomFloatFromBool(Bool=$User.HideHeadFX, True Float=0, False Float=1)
    SpawnRate=20
    Use Spawn Probability=true
   )
  ]
  particleSpawnScript [
   &PartitionParticles()
   &InitializeParticle(
    Lifetime Max=0.45
    Lifetime Min=0.3
    Lifetime Mode=::NewEnumerator1
    Sprite Size=<5 20>
    Sprite Size Mode=::NewEnumerator4
    Sprite UV Mode=::NewEnumerator2
    UsePositionOffset=false
    Write Material Random=true
   )
   &DynamicMaterialParameters(
    Index 0 Param 2=&RandomRangeFloat(Minimum=0, Maximum=4)
    Index 0 Param 4=2.5
    Module.Param0WriteEnabled=false
    Module.Param1WriteEnabled=true
    Module.Param2WriteEnabled=false
    Module.Param3WriteEnabled=true
   ) {module=DynamicMaterialParameters001}
   &AddVelocity(
    Velocity=&SelectVectorFromArray(
     Direct Array Index=$Particles.PartitionParticles.Partition
     Vector Selection Array=DataInterface() {propertyValues {FloatData [<-20 -20 50>, <-20 -20 -50>]}}
    )
    Velocity Speed Scale=0.7
   ) {module=AddVelocity001}
   &SystemLocation(
    MaskSpawnGroup=false
    Offset=&SelectVectorFromArray(
     Direct Array Index=$Particles.PartitionParticles.Partition
     Vector Selection Array=DataInterface() {propertyValues {FloatData [<-10 -6 2>, <-10 -6 -2>]}}
    )
   )
   &CameraOffset(Camera Offset Amount=1)
  ]
  particleUpdateScript [
   &ParticleState()
   &Drag(Drag=3, Use Linear Drag=true, Use Rotational Drag=false)
   &ScaleColor(
    Linear Color Curve=DataInterface() {propertyValues {
      AlphaCurve {keys [<at 0 1>,<at 1 0>]}
      BlueCurve {keys [<at 0.0419 11.2101>,<at 0.2881 2.242>,<at 0.6965 0.565>,<at 0.9695 2.1713>]}
      GreenCurve {keys [<at 0.0419 14.4749>,<at 0.2881 2.895>,<at 0.6965 0.6614>,<at 0.9695 3.074>]}
      RedCurve {keys [<at 0.0419 20>,<at 0.2881 4>,<at 0.6965 1>,<at 0.9695 5>]}
     }}
    Scale Mode=::NewEnumerator2
   ) {module=ScaleColor001}
   &ScaleColor(
    Scale Alpha=&FloatFromCurve(FloatCurve=DataInterface() {propertyValues {Curve {keys [
         {arriveTangent=6.7724, interpMode=RCIM_Cubic, leaveTangent=6.7724, tangentMode=RCTM_User, time=0, value=0}
         {interpMode=RCIM_Cubic, leaveTangent=-5.6421, tangentMode=RCTM_Break, time=0.4, value=0.95}
         {arriveTangent=-1.5833, interpMode=RCIM_Cubic, leaveTangent=-1.5833, time=0.6, value=0.2}
         {arriveTangent=-0.1613, interpMode=RCIM_Cubic, leaveTangent=-0.1613, tangentMode=RCTM_User, time=1, value=0}
        ]}}})
    ScaleA=true
    ScaleRGB=true
   )
   &SolveForcesAndVelocity(Clamp Velocity=false, Limit Acceleration=false)
   &ScaleSpriteSize(Scale Factor=&Multiply_Vector2D_ByFloat(
     Float=&FloatFromCurve(
      FloatCurve=DataInterface() {propertyValues {Curve {keys [
          {arriveTangent=1.4708, interpMode=RCIM_Cubic, leaveTangent=1.4708, tangentMode=RCTM_User, time=0, value=0.525}
          {arriveTangent=0.625, interpMode=RCIM_Cubic, leaveTangent=0.625, time=0.7, value=1.05}
          {interpMode=RCIM_Cubic, tangentMode=RCTM_User, time=1, value=1.15}
         ]}}}
      Scale Curve=0.7
     )
     Vector2D=<0.8 1>
    )) {module=ScaleSpriteSize001}
   &ScaleSpriteSize(
    Non-Uniform Curve Scale=<3 1>
    Non-Uniform Curve Sprite Scale=DataInterface() {propertyValues {
      XCurve {keys [
        {arriveTangent=1.9284, interpMode=RCIM_Cubic, leaveTangent=1.9284, tangentMode=RCTM_User, time=0, value=0.2}
        {arriveTangent=0.1227, interpMode=RCIM_Cubic, leaveTangent=0.1227, tangentMode=RCTM_User, time=1, value=1}
       ]}
      YCurve {keys [
        {arriveTangent=0.2869, interpMode=RCIM_Cubic, leaveTangent=0.2869, tangentMode=RCTM_User, time=0, value=0.85}
        {tangentMode=RCTM_User, time=1, value=1}
       ]}
     }}
    Scale Sprite Size Mode=::NewEnumerator3
   ) {module=ScaleSpriteSize002}
  ]
  rendererClasses [&NiagaraSpriteRendererProperties]
  renderers [{rendererClass=&NiagaraSpriteRendererProperties}]
 }
 Fire_OnHead {
  emitterUpdateScript [
   &EmitterState(Life Cycle Mode=::NewEnumerator1, Loop Behavior=::NewEnumerator1, Loop Duration=0.7, UseLoopDelay=false)
   &SpawnBurst_Instantaneous(
    Spawn Count=3
    Spawn Probability=&MakeCustomFloatFromBool(Bool=$User.HideHeadFX, True Float=0, False Float=1)
    Use Spawn Probability=true
    UseLoopCountLimit=false
   )
   &SpawnRate(
    Spawn Probability=&MakeCustomFloatFromBool(Bool=$User.HideHeadFX, True Float=0, False Float=1)
    SpawnRate=12
    Use Spawn Probability=true
   )
  ]
  particleSpawnScript [
   &InitializeParticle(
    Lifetime Max=0.6
    Lifetime Min=0.5
    Lifetime Mode=::NewEnumerator1
    Sprite Rotation Mode=::NewEnumerator2
    Sprite Size Mode=::NewEnumerator1
    Uniform Sprite Size Max=200
    Uniform Sprite Size Min=100
    UsePositionOffset=false
    Write Material Random=true
   )
   &DynamicMaterialParameters(
    Index 0 Param 2=&RandomRangeFloat(Minimum=0, Maximum=4)
    Index 0 Param 4=2.5
    Module.Param0WriteEnabled=false
    Module.Param1WriteEnabled=true
    Module.Param2WriteEnabled=false
    Module.Param3WriteEnabled=true
   ) {module=DynamicMaterialParameters001}
   &SystemLocation(MaskSpawnGroup=false, Offset=<-16 0 0>)
   &InheritVelocity(Inherited Velocity Amount Scale=<-0.05 -0.05 -0.05>, Speed Limit=true, UseSpeedThreshold=false)
   &CameraOffset(Camera Offset Amount=10)
  ]
  particleUpdateScript [
   &ParticleState()
   &ScaleColor(
    Linear Color Curve=DataInterface() {propertyValues {
      AlphaCurve {keys [{arriveTangent=-0.1596, interpMode=RCIM_Cubic, leaveTangent=-0.1596, tangentMode=RCTM_User, time=0.8987, value=1}, <at 0.982 0 RCIM_Cubic>]}
      BlueCurve {keys [
        <at 0.0505 1.2598 RCIM_Cubic>
        {arriveTangent=-0.8397, interpMode=RCIM_Cubic, leaveTangent=-0.8397, time=0.53, value=0.8348}
        <at 0.9823 0.4774>
       ]}
      GreenCurve {keys [
        <at 0.0505 1.4658 RCIM_Cubic>
        {arriveTangent=-0.9429, interpMode=RCIM_Cubic, leaveTangent=-0.9429, time=0.53, value=0.8753}
        <at 0.9823 0.5872>
       ]}
      RedCurve {keys [<at 0.0505 1.8 RCIM_Cubic>, {arriveTangent=-0.8586, interpMode=RCIM_Cubic, leaveTangent=-0.8586, time=0.53, value=1.2}, <at 0.9823 1>]}
     }}
    Scale Mode=::NewEnumerator2
   ) {module=ScaleColor001}
   &ScaleColor(
    Scale Alpha=&FloatFromCurve(FloatCurve=DataInterface() {propertyValues {Curve {keys [
         {arriveTangent=6.7724, interpMode=RCIM_Cubic, leaveTangent=6.7724, tangentMode=RCTM_User, time=0, value=0}
         {interpMode=RCIM_Cubic, tangentMode=RCTM_User, time=0.3949, value=1}
         {arriveTangent=-1.6527, interpMode=RCIM_Cubic, leaveTangent=-1.6527, time=0.7199, value=0.2109}
         {arriveTangent=-0.174, interpMode=RCIM_Cubic, leaveTangent=-0.174, tangentMode=RCTM_User, time=1, value=0}
        ]}}})
    ScaleA=true
    ScaleRGB=true
   )
   &AccelerationForce(Acceleration=<0 0 150>)
   &InheritSourceMovement(ApplyForce=false, ApplyPosition=true, ApplyPositionFalloff=false)
   &SolveForcesAndVelocity(Clamp Velocity=false, Limit Acceleration=false)
   &ScaleSpriteSize(Scale Factor=&Multiply_Vector2D_ByFloat(
     Float=&FloatFromCurve(
      FloatCurve=DataInterface() {propertyValues {Curve {keys [
          {arriveTangent=1.4708, interpMode=RCIM_Cubic, leaveTangent=1.4708, tangentMode=RCTM_User, time=0, value=1}
          {arriveTangent=1, interpMode=RCIM_Cubic, leaveTangent=1, time=0.4773, value=1.6606}
          {interpMode=RCIM_Cubic, tangentMode=RCTM_User, time=1, value=2}
         ]}}}
      Scale Curve=0.15
     )
     Vector2D=<1 1>
    )) {module=ScaleSpriteSize001}
  ]
  rendererClasses [&NiagaraSpriteRendererProperties]
  renderers [{rendererClass=&NiagaraSpriteRendererProperties}]
 }
 Fire_Sprite {
  emitterUpdateScript [
   &EmitterState(Delay First Loop Only=true, Life Cycle Mode=::NewEnumerator1, Loop Delay=$System.BreathDelayTime, Loop Duration=1, UseLoopDelay=true)
   &SpawnRate(SpawnRate=25, Use Spawn Probability=false)
  ]
  particleSpawnScript [
   &InitializeParticle(
    Lifetime Max=0.6
    Lifetime Min=0.5
    Lifetime Mode=::NewEnumerator1
    Sprite Rotation Mode=::NewEnumerator2
    Sprite Size Mode=::NewEnumerator1
    Uniform Sprite Size Max=200
    Uniform Sprite Size Min=100
    UsePositionOffset=false
    Write Material Random=true
   )
   &DynamicMaterialParameters(
    Index 0 Param 2=&RandomRangeFloat(Minimum=0, Maximum=4)
    Index 0 Param 4=2.5
    Module.Param0WriteEnabled=false
    Module.Param1WriteEnabled=true
    Module.Param2WriteEnabled=false
    Module.Param3WriteEnabled=true
   ) {module=DynamicMaterialParameters001}
   &AddVelocity(
    Cone Angle=8
    Module.UseConeFalloff=false
    Module.UseVelDistribution=false
    Override Seed=false
    Velocity Mode=::NewEnumerator2
    Velocity Speed=&Multiply_Float(
     A=&RandomRangeFloat(Minimum=0.8, Maximum=1)
     B=&Lerp_Float(A=200, B=1000, Alpha=$User.FireBreathLengthLerp)
    )
   ) {module=AddVelocity001}
  ]
  particleUpdateScript [
   &ParticleState()
   &AccelerationForce(Acceleration=&UniformRangedVector(Minimum=<0 0 1000>, Maximum=<0 0 500>))
   &Drag(
    Drag=&FloatFromCurve(FloatCurve=DataInterface() {propertyValues {Curve {keys [<at 0 0.075 RCIM_Cubic>,<at 1 1.075 RCIM_Cubic>]}}}, Scale Curve=4)
    Use Linear Drag=true
    Use Rotational Drag=false
   )
   &ScaleColor(
    Linear Color Curve=DataInterface() {propertyValues {
      AlphaCurve {keys [{arriveTangent=-0.1596, interpMode=RCIM_Cubic, leaveTangent=-0.1596, tangentMode=RCTM_User, time=0.8987, value=1}, <at 0.982 0 RCIM_Cubic>]}
      BlueCurve {keys [
        <at 0.0505 3.4995 RCIM_Cubic>
        {arriveTangent=-3.8023, interpMode=RCIM_Cubic, leaveTangent=-3.8023, time=0.3055, value=1.1959}
        {arriveTangent=-1.0615, interpMode=RCIM_Cubic, leaveTangent=-1.0615, time=0.7879, value=0.6957}
        <at 0.9823 0.4774>
       ]}
      GreenCurve {keys [
        <at 0.0505 4.0716 RCIM_Cubic>
        {arriveTangent=-4.5324, interpMode=RCIM_Cubic, leaveTangent=-4.5324, time=0.3055, value=1.3562}
        {arriveTangent=-1.1361, interpMode=RCIM_Cubic, leaveTangent=-1.1361, time=0.7879, value=0.7294}
        <at 0.9823 0.5872>
       ]}
      RedCurve {keys [
        <at 0.0505 5 RCIM_Cubic>
        {arriveTangent=-5.4244, interpMode=RCIM_Cubic, leaveTangent=-5.4244, time=0.3055, value=2}
        {arriveTangent=-1.4775, interpMode=RCIM_Cubic, leaveTangent=-1.4775, time=0.7879, value=1}
        <at 0.9823 1>
       ]}
     }}
    Scale Mode=::NewEnumerator2
   ) {module=ScaleColor001}
   &ScaleColor(
    Scale Alpha=&FloatFromCurve(
     FloatCurve=DataInterface() {propertyValues {Curve {keys [
         {arriveTangent=6.7724, interpMode=RCIM_Cubic, leaveTangent=6.7724, tangentMode=RCTM_User, time=0, value=0}
         {interpMode=RCIM_Cubic, tangentMode=RCTM_User, time=0.3949, value=1}
         {arriveTangent=-1.6527, interpMode=RCIM_Cubic, leaveTangent=-1.6527, time=0.7199, value=0.2109}
         {arriveTangent=-0.174, interpMode=RCIM_Cubic, leaveTangent=-0.174, tangentMode=RCTM_User, time=1, value=0}
        ]}}}
     Scale Curve=&Multiply_Float(A=0.8, B=$System.TOD)
    )
    ScaleA=true
    ScaleRGB=true
   )
   &SolveForcesAndVelocity(Clamp Velocity=false, Limit Acceleration=false)
   &ScaleSpriteSize(Scale Factor=&Multiply_Vector2D_ByFloat(
     Float=&FloatFromCurve(
      FloatCurve=DataInterface() {propertyValues {Curve {keys [
          {arriveTangent=1.4708, interpMode=RCIM_Cubic, leaveTangent=1.4708, tangentMode=RCTM_User, time=0, value=0.1}
          {arriveTangent=1.9, interpMode=RCIM_Cubic, leaveTangent=1.9, time=0.6, value=1.625}
          {interpMode=RCIM_Cubic, tangentMode=RCTM_User, time=1, value=2}
         ]}}}
      Scale Curve=&MakeCustomFloatFromBool(Bool=$User.IsHitPlayerPawn, True Float=0.6, False Float=1)
     )
     Vector2D=<1 1>
    )) {module=ScaleSpriteSize001}
  ]
  rendererClasses [&NiagaraSpriteRendererProperties]
  renderers [{rendererClass=&NiagaraSpriteRendererProperties}]
 }
 Fire_TouchSpread {
  emitterUpdateScript [
   &EmitterState(Delay First Loop Only=true, Life Cycle Mode=::NewEnumerator1, Loop Delay=$System.BreathDelayTime, Loop Duration=1, UseLoopDelay=true)
   &SpawnRate(
    Spawn Probability=&FloatFromCurve(
     FloatCurve=DataInterface() {propertyValues {Curve {keys [<at 0 1 RCIM_Constant>,<at 0.9 1 RCIM_Constant>,<at 0.98 0 RCIM_Constant>]}}}
     CurveIndex=$User.FireBreathLengthLerp
    )
    SpawnRate=&MakeCustomFloatFromBool(Bool=$User.IsHitPlayerPawn, True Float=25, False Float=50)
    Use Spawn Probability=true
   )
  ]
  particleSpawnScript [
   &InitializeParticle(
    Lifetime Max=0.7
    Lifetime Min=0.6
    Lifetime Mode=::NewEnumerator1
    Sprite Rotation Mode=::NewEnumerator2
    Sprite Size Mode=::NewEnumerator1
    Uniform Sprite Size Max=200
    Uniform Sprite Size Min=100
    UsePositionOffset=false
    Write Material Random=true
   )
   &DynamicMaterialParameters(
    Index 0 Param 2=&RandomRangeFloat(Minimum=0, Maximum=4)
    Index 0 Param 4=2.5
    Module.Param0WriteEnabled=false
    Module.Param1WriteEnabled=true
    Module.Param2WriteEnabled=false
    Module.Param3WriteEnabled=true
   ) {module=DynamicMaterialParameters001}
   &ShapeLocation(
    MaskSpawnGroup=false
    Offset Coordinate Space=::World
    Offset Mode=::NewEnumerator0
    Override Seed=false
    Ring Radius=15
    Rotation Axis=$User.FireTouchPointNormal
    Rotation Coordinate Space=::World
    Rotation Mode=::NewEnumerator1
    Shape Origin=$User.FireTouchPointLocation
    Shape Primitive=::NewEnumerator4
    UseOwnerScale=true
   )
   &AddVelocity(
    Module.Use Falloff Distance=false
    Origin Offset Coordinate Space=::World
    Override Seed=false
    Velocity Mode=::NewEnumerator1
    Velocity Origin=$User.FireTouchPointLocation
    Velocity Speed=&Multiply_Float(
     A=&Multiply_Float(
      A=&RandomRangeFloat(Minimum=0.8, Maximum=1)
      B=&Lerp_Float(A=400, B=100, Alpha=$User.FireBreathLengthLerp)
     )
     B=&MakeCustomFloatFromBool(Bool=$User.IsHitPlayerPawn, True Float=0.5, False Float=1)
    )
   ) {module=AddVelocity001}
   &SystemLocation(MaskSpawnGroup=false, Offset=<500 0 0>) {enabled=false}
  ]
  particleUpdateScript [
   &ParticleState()
   &AccelerationForce(Acceleration=&UniformRangedVector(Minimum=<0 0 500>, Maximum=<0 0 250>))
   &Drag(
    Drag=&FloatFromCurve(
     FloatCurve=DataInterface() {propertyValues {Curve {keys [{arriveTangent=1.7, interpMode=RCIM_Cubic, leaveTangent=1.7, tangentMode=RCTM_User, time=0, value=0.075}, <at 1 1.075 RCIM_Cubic>]}}}
     Scale Curve=4
    )
    Use Linear Drag=true
    Use Rotational Drag=false
   )
   &ScaleColor(
    Linear Color Curve=DataInterface() {propertyValues {
      AlphaCurve {keys [{arriveTangent=-0.1596, interpMode=RCIM_Cubic, leaveTangent=-0.1596, tangentMode=RCTM_User, time=0.8987, value=1}, <at 0.982 0 RCIM_Cubic>]}
      BlueCurve {keys [
        <at 0.0505 1.0498 RCIM_Cubic>
        {arriveTangent=-0.6143, interpMode=RCIM_Cubic, leaveTangent=-0.6143, time=0.53, value=0.6957}
        <at 0.9823 0.4774>
       ]}
      GreenCurve {keys [
        <at 0.0505 1.2215 RCIM_Cubic>
        {arriveTangent=-0.6807, interpMode=RCIM_Cubic, leaveTangent=-0.6807, time=0.53, value=0.7294}
        <at 0.9823 0.5872>
       ]}
      RedCurve {keys [<at 0.0505 1.5 RCIM_Cubic>, {arriveTangent=-0.5366, interpMode=RCIM_Cubic, leaveTangent=-0.5366, time=0.53, value=1}, <at 0.9823 1>]}
     }}
    Scale Mode=::NewEnumerator2
   ) {module=ScaleColor001}
   &ScaleColor(
    Scale Alpha=&FloatFromCurve(
     FloatCurve=DataInterface() {propertyValues {Curve {keys [
         {arriveTangent=6.7724, interpMode=RCIM_Cubic, leaveTangent=6.7724, tangentMode=RCTM_User, time=0, value=0}
         {interpMode=RCIM_Cubic, tangentMode=RCTM_User, time=0.3949, value=1}
         {arriveTangent=-1.6527, interpMode=RCIM_Cubic, leaveTangent=-1.6527, time=0.7199, value=0.2109}
         {arriveTangent=-0.174, interpMode=RCIM_Cubic, leaveTangent=-0.174, tangentMode=RCTM_User, time=1, value=0}
        ]}}}
     Scale Curve=&Multiply_Float(A=0.8, B=$System.TOD)
    )
    ScaleA=true
    ScaleRGB=true
   )
   &SolveForcesAndVelocity(Clamp Velocity=false, Limit Acceleration=false)
   &ScaleSpriteSize(Scale Factor=&Multiply_Vector2D_ByFloat(
     Float=&FloatFromCurve(
      FloatCurve=DataInterface() {propertyValues {Curve {keys [
          {arriveTangent=1.4708, interpMode=RCIM_Cubic, leaveTangent=1.4708, tangentMode=RCTM_User, time=0, value=1}
          {arriveTangent=1, interpMode=RCIM_Cubic, leaveTangent=1, time=0.4773, value=1.6606}
          {interpMode=RCIM_Cubic, tangentMode=RCTM_User, time=1, value=2}
         ]}}}
      Scale Curve=&MakeCustomFloatFromBool(Bool=$User.IsHitPlayerPawn, True Float=0.4, False Float=0.6)
     )
     Vector2D=<1 1>
    )) {module=ScaleSpriteSize001}
  ]
  rendererClasses [&NiagaraSpriteRendererProperties]
  renderers [{rendererClass=&NiagaraSpriteRendererProperties}]
 }
 Flash_Flare {
  emitterUpdateScript [
   &EmitterState(Life Cycle Mode=::NewEnumerator1, Loop Behavior=::NewEnumerator1, Loop Duration=1, UseLoopDelay=false)
   &SpawnBurst_Instantaneous(Spawn Count=1, Spawn Time=$System.BreathDelayTime, Use Spawn Probability=false)
  ]
  particleSpawnScript [
   &InitializeParticle(
    Color=<rgba 0.4896 0.1043 0 1>
    Lifetime=0.6
    Sprite Size=<200 600>
    Sprite Size Mode=::NewEnumerator4
    UsePositionOffset=false
    Write Material Random=true
   )
   &DynamicMaterialParameters(
    Index 0 Param 1=0.2
    Index 0 Param 2=1
    Module.Param0WriteEnabled=true
    Module.Param1WriteEnabled=true
    Module.Param2WriteEnabled=true
    Module.Param3WriteEnabled=true
   )
   /SetVariables(Particles.ParticleAlignmentVector:Vector3f=<1 0 0>) {isSetParameters=true}
   &SystemLocation(MaskSpawnGroup=false) {enabled=false}
  ]
  particleUpdateScript [
   &ParticleState()
   &ScaleColor(
    Scale Alpha=&FloatFromCurve(
     FloatCurve=DataInterface() {propertyValues {Curve {keys [
         {arriveTangent=-0.0383, interpMode=RCIM_Cubic, leaveTangent=-0.0383, tangentMode=RCTM_User, time=0, value=0}
         {arriveTangent=-0.0405, interpMode=RCIM_Cubic, leaveTangent=-0.0405, tangentMode=RCTM_User, time=0.7009, value=1}
         {arriveTangent=-8.0089, interpMode=RCIM_Cubic, leaveTangent=-8.0089, tangentMode=RCTM_User, time=1, value=0}
        ]}}}
     Scale Curve=&Lerp_Float(A=0, B=1, Alpha=$User.FireBreathLengthLerp)
    )
    Scale RGB=&VectorFromFloat(Value=1)
    ScaleA=true
    ScaleRGB=true
   )
   &FN_Light_Attributes(
    Color=$Particles.Color
    Create Color Attribute=true
    Create Position Attribute=true
    Create Radius Attribute=true
    Position=&TransformPosition(Position=$User.FireTouchPointLocation, Source Space=::World, Destination Space=::Local)
    Radius=800
    Scale Color=&FloatFromCurve(
     FloatCurve=DataInterface() {propertyValues {Curve {keys [
         {tangentMode=RCTM_User, time=0, value=1}
         {arriveTangent=-0.0405, interpMode=RCIM_Cubic, leaveTangent=-0.0405, tangentMode=RCTM_User, time=0.7009, value=1}
         {arriveTangent=-8.0089, interpMode=RCIM_Cubic, leaveTangent=-8.0089, tangentMode=RCTM_User, time=1, value=0}
        ]}}}
     Scale Curve=&Multiply_Float(
      A=&Multiply_Float(
       A=&RandomRangeFloat(Minimum=12, Maximum=30, Evaluation Type=::NewEnumerator1)
       B=&FNTimeOfDayScalar(Night Scale=0.5, Morning Scale=0.5, Evening Scale=0.5)
      )
      B=&FloatFromCurve(
       FloatCurve=DataInterface() {propertyValues {Curve {keys [
           {arriveTangent=1.5243, interpMode=RCIM_Cubic, leaveTangent=1.5243, tangentMode=RCTM_User, time=0, value=0.1}
           {arriveTangent=0.0515, interpMode=RCIM_Cubic, leaveTangent=0.0515, tangentMode=RCTM_User, time=1, value=1}
          ]}}}
       CurveIndex=$System.NearCameraFade
      )
     )
    )
   )
  ]
  rendererClasses [&NiagaraSpriteRendererProperties, &NiagaraLightRendererProperties]
  renderers [
   {rendererClass=&NiagaraSpriteRendererProperties}
   {rendererClass=&NiagaraLightRendererProperties, rendererIndex=1}
  ]
 }
 Flash_Flare001 {
  emitterUpdateScript [
   &EmitterState(Life Cycle Mode=::NewEnumerator1, Loop Behavior=::NewEnumerator1, Loop Duration=1, UseLoopDelay=false)
   &SpawnBurst_Instantaneous(
    Spawn Count=1
    Spawn Probability=&MakeCustomFloatFromBool(Bool=$User.IsHitTerrain, True Float=1, False Float=0)
    Spawn Time=0.3
    Use Spawn Probability=true
   )
  ]
  particleSpawnScript [
   &InitializeParticle(Lifetime=2, Sprite Size Mode=::NewEnumerator3, Uniform Sprite Size=600, UsePositionOffset=false, Write Material Random=true)
   /SetVariables(
    Particles.DecalFade=1
    Particles.DecalOrientation:Quat4f=&OrientMeshToVector(Look At Direction=$User.FireTouchPointNormal, Reference Axis Vector=<1 0 0>)
    Particles.DecalSize:Vector3f=<300 500 500>
   ) {isSetParameters=true}
   /SetVariables(Particles.Position=$User.FireTouchPointLocation) {isSetParameters=true}
  ]
  particleUpdateScript [
   &ParticleState()
   /SetVariables(Particles.DecalFade:NiagaraFloat=&Lerp_Float(
     A=0
     B=1
     Alpha=&FloatFromCurve(FloatCurve=DataInterface() {propertyValues {Curve {keys [
          {arriveTangent=12.3499, interpMode=RCIM_Cubic, leaveTangent=12.3499, tangentMode=RCTM_User, time=0, value=0}
          <at 0.2 1 RCIM_Cubic>
          {arriveTangent=-0.0037, interpMode=RCIM_Cubic, leaveTangent=-0.0037, tangentMode=RCTM_User, time=1, value=0}
         ]}}})
    )) {isSetParameters=true}
  ]
  rendererClasses [&NiagaraDecalRendererProperties, &NiagaraSpriteRendererProperties]
  renderers [
   {rendererClass=&NiagaraDecalRendererProperties, rendererIndex=0}
   {rendererClass=&NiagaraSpriteRendererProperties, rendererIndex=1}
  ]
 }
 GroundFire {
  emitterUpdateScript [
   &EmitterState(Delay First Loop Only=true, Life Cycle Mode=::NewEnumerator1, Loop Delay=$System.BreathDelayTime, Loop Duration=1, UseLoopDelay=true)
   &SpawnRate(
    Spawn Probability=&FloatFromCurve(
     FloatCurve=DataInterface() {propertyValues {Curve {keys [<at 0 1 RCIM_Constant>,<at 0.9 1 RCIM_Constant>,<at 0.98 0 RCIM_Constant>]}}}
     CurveIndex=$User.FireBreathLengthLerp
     Scale Curve=&MakeCustomFloatFromBool(Bool=$User.IsHitTerrain, True Float=1, False Float=0)
    )
    SpawnRate=50
    Use Spawn Probability=true
   )
  ]
  particleSpawnScript [
   &InitializeParticle(
    Color=<rgba 1 0.7498 0.4718 1>
    Lifetime Max=1.2
    Lifetime Min=0.6
    Lifetime Mode=::NewEnumerator1
    Sprite Size Mode=::NewEnumerator1
    Sprite UV Mode=::NewEnumerator1
    Uniform Sprite Size Max=150
    Uniform Sprite Size Min=80
    UsePositionOffset=false
    Write Material Random=true
   )
   &DynamicMaterialParameters(
    Index 0 Param 2=&RandomRangeFloat(Minimum=0, Maximum=4)
    Index 0 Param 4=2.5
    Module.Param0WriteEnabled=false
    Module.Param1WriteEnabled=true
    Module.Param2WriteEnabled=false
    Module.Param3WriteEnabled=true
   ) {module=DynamicMaterialParameters001}
   &ShapeLocation(
    MaskSpawnGroup=false
    Offset Coordinate Space=::World
    Offset Mode=::NewEnumerator0
    Override Seed=false
    Ring Radius=15
    Rotation Axis=$User.FireTouchPointNormal
    Rotation Coordinate Space=::World
    Rotation Mode=::NewEnumerator1
    Shape Origin=$User.FireTouchPointLocation
    Shape Primitive=::NewEnumerator4
    UseOwnerScale=true
   )
   &AddVelocity(
    Module.Use Falloff Distance=false
    Origin Offset Coordinate Space=::World
    Override Seed=false
    Velocity Mode=::NewEnumerator1
    Velocity Origin=$User.FireTouchPointLocation
    Velocity Speed=&Multiply_Float(
     A=&RandomRangeFloat(Minimum=0.4, Maximum=0.5)
     B=&Lerp_Float(A=400, B=100, Alpha=$User.FireBreathLengthLerp)
    )
   ) {module=AddVelocity001}
   &CameraOffset(Camera Offset Amount=50)
  ]
  particleUpdateScript [
   &ParticleState()
   &AccelerationForce(Acceleration=&UniformRangedVector(Minimum=<0 0 1000>, Maximum=<0 0 500>)) {enabled=false}
   &Drag(
    Drag=&FloatFromCurve(FloatCurve=DataInterface() {propertyValues {Curve {keys [<at 0 0.075 RCIM_Cubic>,<at 1 1.075 RCIM_Cubic>]}}}, Scale Curve=4)
    Use Linear Drag=true
    Use Rotational Drag=false
   )
   &ScaleColor(
    Linear Color Curve=DataInterface() {propertyValues {
      AlphaCurve {keys [{arriveTangent=-0.1596, interpMode=RCIM_Cubic, leaveTangent=-0.1596, tangentMode=RCTM_User, time=0.8987, value=1}, <at 0.982 0 RCIM_Cubic>]}
      BlueCurve {keys [
        <at 0.0505 3.4995 RCIM_Cubic>
        {arriveTangent=-3.6626, interpMode=RCIM_Cubic, leaveTangent=-3.6626, time=0.3055, value=1.1959}
        {arriveTangent=-1.0615, interpMode=RCIM_Cubic, leaveTangent=-1.0615, time=0.7879, value=0.7986}
        <at 0.9823 0.4774>
       ]}
      GreenCurve {keys [
        <at 0.0505 4.0716 RCIM_Cubic>
        {arriveTangent=-4.4082, interpMode=RCIM_Cubic, leaveTangent=-4.4082, time=0.3055, value=1.3562}
        {arriveTangent=-1.1361, interpMode=RCIM_Cubic, leaveTangent=-1.1361, time=0.7879, value=0.821}
        <at 0.9823 0.5872>
       ]}
      RedCurve {keys [
        <at 0.0505 5 RCIM_Cubic>
        {arriveTangent=-5.4244, interpMode=RCIM_Cubic, leaveTangent=-5.4244, time=0.3055, value=2}
        {arriveTangent=-1.4775, interpMode=RCIM_Cubic, leaveTangent=-1.4775, time=0.7879, value=1}
        <at 0.9823 1>
       ]}
     }}
    Scale Mode=::NewEnumerator2
   ) {module=ScaleColor001}
   &ScaleColor(
    Scale Alpha=&FloatFromCurve(FloatCurve=DataInterface() {propertyValues {Curve {keys [
         {arriveTangent=6.7724, interpMode=RCIM_Cubic, leaveTangent=6.7724, tangentMode=RCTM_User, time=0, value=0}
         {interpMode=RCIM_Cubic, tangentMode=RCTM_User, time=0.2, value=0.95}
         {arriveTangent=-3.0347, interpMode=RCIM_Cubic, leaveTangent=-3.0347, tangentMode=RCTM_User, time=1, value=0}
        ]}}})
    Scale RGB=&VectorFromFloat(Value=10)
    ScaleA=true
    ScaleRGB=true
   )
   &SolveForcesAndVelocity(Clamp Velocity=false, Limit Acceleration=false)
   &ScaleSpriteSize(Scale Factor=&Multiply_Vector2D_ByFloat(
     Float=&FloatFromCurve(
      FloatCurve=DataInterface() {propertyValues {Curve {keys [
          {arriveTangent=1.4708, interpMode=RCIM_Cubic, leaveTangent=1.4708, tangentMode=RCTM_User, time=0, value=0.1}
          {arriveTangent=1.9, interpMode=RCIM_Cubic, leaveTangent=1.9, time=0.6, value=1.625}
          {interpMode=RCIM_Cubic, tangentMode=RCTM_User, time=1, value=2}
         ]}}}
      Scale Curve=0.6
     )
     Vector2D=<0.8 1>
    )) {module=ScaleSpriteSize001}
  ]
  rendererClasses [&NiagaraSpriteRendererProperties]
  renderers [{rendererClass=&NiagaraSpriteRendererProperties}]
 }
 ParticleDecal {
  emitterUpdateScript [
   &EmitterState(Life Cycle Mode=::NewEnumerator1, Loop Behavior=::NewEnumerator1, Loop Duration=1, UseLoopDelay=false)
   &SpawnBurst_Instantaneous(
    Spawn Probability=&MakeCustomFloatFromBool(Bool=$User.HideHeadFX, True Float=0, False Float=1)
    Use Spawn Probability=true
    UseLoopCountLimit=false
   )
   &SpawnBurst_Instantaneous(
    Spawn Probability=&MakeCustomFloatFromBool(Bool=$User.HideHeadFX, True Float=0, False Float=1)
    Spawn Time=0.62
    Use Spawn Probability=true
    UseLoopCountLimit=false
   ) {module=SpawnBurst_Instantaneous001}
  ]
  particleSpawnScript [
   &PartitionParticles(Number Of Partitions=26) {enabled=false}
   &InitializeParticle(
    Color=<rgba 1 0.2549 0.0234 1>
    Lifetime=&Lerp_Float(
     A=0.5
     B=0.8
     Alpha=&FloatFromCurve(FloatCurve=DataInterface() {propertyValues {Curve {keys [<at 0 1 RCIM_Constant>,<at 0.5 0 RCIM_Constant>]}}}, CurveIndex=$System.Age)
    )
    Sprite Rotation Mode=::NewEnumerator2
    Sprite Size Mode=::NewEnumerator3
    Uniform Sprite Size=&Multiply_Float(
     A=45
     B=&Lerp_Float(
      A=1
      B=1.5
      Alpha=&FloatFromCurve(FloatCurve=DataInterface() {propertyValues {Curve {keys [<at 0 1 RCIM_Constant>,<at 0.5 0 RCIM_Constant>]}}}, CurveIndex=$System.Age)
     )
    )
    UsePositionOffset=false
    Write Material Random=true
   )
   &CameraOffset(Camera Offset Amount=20) {module=CameraOffset001}
   &DynamicMaterialParameters(
    Index 0 Param 1=1
    Index 0 Param 2=1
    Module.Index0Param0Write=true
    Module.Index0Param1Write=true
    Module.Index0Param2Write=true
    Module.Index0Param3Write=false
   )
   /SetVariables(Particles.EnergyDP:Vector4f=<2 1 8 1>) {isSetParameters=true}
   /SetVariables(Particles.EnergySize:Vector2f=&Multiply_Vector2D_ByFloat(Float=0.4, Vector2D=$Particles.SpriteSize)) {isSetParameters=true}
   &SystemLocation(MaskSpawnGroup=false, Offset=<-12 -2 0>)
  ]
  particleUpdateScript [
   &ParticleState()
   &ScaleSpriteSize(
    Uniform Curve Index=$Emitter.Age
    Uniform Curve Sprite Scale=DataInterface() {propertyValues {Curve {keys [
        {arriveTangent=-3.3333, interpMode=RCIM_Cubic, time=0, value=0.3711}
        {arriveTangent=-7.2071, interpMode=RCIM_Cubic, leaveTangent=-7.2071, tangentMode=RCTM_User, time=0.1611, value=2.018}
        <at 0.4445 1.018 RCIM_Cubic>
       ]}}}
   )
   &ScaleColor(
    Scale Mode=::NewEnumerator1
    Scale RGBA=&MakeVector4_FromLinearColor(LinearColor=<rgba 1 0.3998 0.1409 1>)
    ScaleRGBA=true
   ) {module=ScaleColor001}
   &ScaleColor(
    Scale Alpha=&FloatFromCurve(FloatCurve=DataInterface() {propertyValues {Curve {keys [
         {arriveTangent=0.3051, interpMode=RCIM_Cubic, leaveTangent=0.3051, tangentMode=RCTM_User, time=0, value=1}
         {arriveTangent=-1.4301, interpMode=RCIM_Cubic, leaveTangent=-1.4301, tangentMode=RCTM_User, time=0.5, value=0.95}
         {arriveTangent=-0.0405, interpMode=RCIM_Cubic, leaveTangent=-0.0405, tangentMode=RCTM_User, time=1, value=0}
        ]}}})
    Scale RGB=&VectorFromFloat(Value=&FloatFromCurve(
      FloatCurve=DataInterface() {propertyValues {Curve {keys [
          <at 0 1 RCIM_Cubic>
          {arriveTangent=-0.2857, leaveTangent=-0.2857, time=0.5, value=0.925}
          {arriveTangent=-1.85, interpMode=RCIM_Cubic, leaveTangent=-1.85, time=0.9, value=8}
          <at 1 0 RCIM_Cubic>
         ]}}}
      Scale Curve=&Lerp_Float(
       A=1
       B=5
       Alpha=&FloatFromCurve(FloatCurve=DataInterface() {propertyValues {Curve {keys [<at 0 1 RCIM_Constant>,<at 0.5 0 RCIM_Constant>]}}}, CurveIndex=$System.Age)
      )
     ))
    ScaleA=true
    ScaleRGB=true
   )
   /SetVariables(Particles.EnergyColor:LinearColor=&MultiplyLinearColor(Linear Color=$Particles.Color, Scale Factor=&MakeVector4_FromFloat(Float=50))) {isSetParameters=true}
   /SetVariables(Particles.GlowColor=&Multiply_LinearColorByFloat(Linear Color=$Particles.Color, Float=0.2)) {isSetParameters=true}
  ]
  rendererClasses [&NiagaraSpriteRendererProperties]
  renderers [
   {rendererClass=&NiagaraSpriteRendererProperties}
   {rendererClass=&NiagaraSpriteRendererProperties, rendererIndex=1}
   {rendererClass=&NiagaraSpriteRendererProperties, rendererIndex=2}
  ]
 }
}
summary {
 systemName=NS_CoffeePuff_Combo_FireBreath
 emitters [
  {emitterName=Flash_Flare, rendererClasses [&NiagaraSpriteRendererProperties, &NiagaraLightRendererProperties]}
  {emitterName=Embers_Burst, simTarget=GPUComputeSim, rendererClasses [&NiagaraSpriteRendererProperties]}
  {emitterName=Fire_Sprite, rendererClasses [&NiagaraSpriteRendererProperties]}
  {emitterName=FireFast_V_aligned, rendererClasses [&NiagaraSpriteRendererProperties]}
  {emitterName=ParticleDecal, rendererClasses [&NiagaraSpriteRendererProperties]}
  {emitterName=FireOnEyes, rendererClasses [&NiagaraSpriteRendererProperties]}
  {emitterName=EnergyWave, rendererClasses [&NiagaraMeshRendererProperties]}
  {bEnabled=false, emitterName=Embers_Burst001, simTarget=GPUComputeSim, rendererClasses [&NiagaraSpriteRendererProperties]}
  {emitterName=Fire_TouchSpread, rendererClasses [&NiagaraSpriteRendererProperties]}
  {emitterName=GroundFire, rendererClasses [&NiagaraSpriteRendererProperties]}
  {
   emitterName=Flash_Flare001
   rendererClasses [&NiagaraDecalRendererProperties, &NiagaraSpriteRendererProperties]
  }
  {emitterName=Fire_OnHead, rendererClasses [&NiagaraSpriteRendererProperties]}
 ]
 userVariables [
  User.FireBreathLengthLerp {type=NiagaraFloat, defaultValue=1, description=""}
  User.FireTouchPointLocation {type=NiagaraPosition, description="", defaultValue {}}
  User.FireTouchPointNormal {type=Vector3f, description="", defaultValue {struct=Vector3f, value=<1 0 0>}}
  User.HideHeadFX {type=NiagaraBool}
  User.IsHitPlayerPawn {type=NiagaraBool}
  User.IsHitTerrain {type=NiagaraBool}
 ]
}
