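# Niagara system for CoffeePuff's fire breath hitting a player, showing short-lived fire on the victim's body.
# Two sprite burst emitters, flame puffs and a triple-renderer glow layer named ParticleDecal, use partition indices to follow skeletal mesh bones every update.
# User.IsHeavyImpact lengthens particle lifetimes in both emitters.
# Curves drive color, fade and flame growth, with alpha scaled by time of day, plus near-camera fade on the ParticleDecal layer.
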NiagaraSystem /CoffeePuffGameplay/Assets/VFX/NS_CoffeePuff_FireBreath_ImpactOnPlayer
dependencies {
 usedDataInterfaces [
  &NiagaraDataInterfaceCurve
  &NiagaraDataInterfaceColorCurve
  &NiagaraDataInterfaceVectorCurve
  &NiagaraDataInterfaceSkeletalMesh
 ]
 usedDynamicInputs [
  &VectorFromCurve
  &FNTimeOfDayScalar
  &FN_LOD_Distance_Remap
 ]
 usedModules [
  &self:SystemSpawnScript.SystemScriptSource.SystemScriptGraph.NiagaraNodeAssignment_0.SetVariables_#1
  &self:SystemSpawnScript.SystemScriptSource.SystemScriptGraph.NiagaraNodeAssignment_2.SetVariables_#2
  &self:SystemSpawnScript.SystemScriptSource.SystemScriptGraph.NiagaraNodeAssignment_3.SetVariables_#3
  &SystemState
  &self:SystemSpawnScript.SystemScriptSource.SystemScriptGraph.NiagaraNodeAssignment_4.SetVariables_#4
  &self:SystemSpawnScript.SystemScriptSource.SystemScriptGraph.NiagaraNodeAssignment_1.SetVariables_#5
 ]
}
emitters {
 Fire_Sprite001 {
  emitterUpdateScript [
   &EmitterState()
   &SpawnBurst_Instantaneous(Spawn Count=26, Use Spawn Probability=true, UseLoopCountLimit=false)
  ]
  particleSpawnScript [
   &PartitionParticles(Number Of Partitions=26)
   &InitializeParticle(
    Lifetime=&Multiply_Float(
     A=&MakeCustomFloatFromBool(Bool=$User.IsHeavyImpact, True Float=1, False Float=0.7)
     B=&RandomRangeFloat(Minimum=0.4, Maximum=0.5)
    )
    Sprite Rotation Mode=::NewEnumerator2
    Sprite Size Mode=::NewEnumerator1
    Uniform Sprite Size Max=200
    Uniform Sprite Size Min=100
    UsePositionOffset=false
    Write Material Random=true
   )
   &DynamicMaterialParameters(
    Index 0 Param 2=&RandomRangeFloat(Minimum=0, Maximum=4)
    Index 0 Param 4=2.5
    Module.Index0Float0Write=true
    Module.Index0Float1Write=true
    Module.Index0Float2Write=true
    Module.Index0Float3Write=true
   ) {module=DynamicMaterialParameters001}
  ]
  particleUpdateScript [
   &ParticleState()
   &SkeletalMeshLocation(
    Bone / Socket Index=$Particles.PartitionParticles.Partition
    Bone Sampling Mode=::NewEnumerator4
    MaskSpawnGroup=false
    Skeletal Mesh=$System.SK_Mesh
   )
   &ScaleColor(
    Linear Color Curve=DataInterface() {propertyValues {
      AlphaCurve {keys [{arriveTangent=-0.1596, interpMode=RCIM_Cubic, leaveTangent=-0.1596, tangentMode=RCTM_User, time=0.8987, value=1}, <at 0.982 0 RCIM_Cubic>]}
      BlueCurve {keys [
        <at 0.0505 3.4995 RCIM_Cubic>
        {arriveTangent=-3.8848, interpMode=RCIM_Cubic, leaveTangent=-3.8848, time=0.3055, value=1.1959}
        {arriveTangent=-1.0615, interpMode=RCIM_Cubic, leaveTangent=-1.0615, time=0.7879, value=0.6348}
        <at 0.9823 0.4774>
       ]}
      GreenCurve {keys [
        <at 0.0505 4.0716 RCIM_Cubic>
        {arriveTangent=-4.6058, interpMode=RCIM_Cubic, leaveTangent=-4.6058, time=0.3055, value=1.3562}
        {arriveTangent=-1.1361, interpMode=RCIM_Cubic, leaveTangent=-1.1361, time=0.7879, value=0.6753}
        <at 0.9823 0.5872>
       ]}
      RedCurve {keys [
        <at 0.0505 5 RCIM_Cubic>
        {arriveTangent=-5.4244, interpMode=RCIM_Cubic, leaveTangent=-5.4244, time=0.3055, value=2}
        {arriveTangent=-1.4775, interpMode=RCIM_Cubic, leaveTangent=-1.4775, time=0.7879, value=1}
        <at 0.9823 1>
       ]}
     }}
    Scale Mode=::NewEnumerator2
   ) {module=ScaleColor001}
   &ScaleColor(
    Scale Alpha=&FloatFromCurve(
     FloatCurve=DataInterface() {propertyValues {Curve {keys [
         {arriveTangent=6.7724, interpMode=RCIM_Cubic, leaveTangent=6.7724, tangentMode=RCTM_User, time=0, value=0}
         {interpMode=RCIM_Cubic, tangentMode=RCTM_User, time=0.5, value=1}
         {arriveTangent=-3.0347, interpMode=RCIM_Cubic, leaveTangent=-3.0347, tangentMode=RCTM_User, time=1, value=0}
        ]}}}
     Scale Curve=&Multiply_Float(A=0.5, B=$System.TOD)
    )
    ScaleA=true
    ScaleRGB=true
   )
   &SolveForcesAndVelocity(Clamp Velocity=false, Limit Acceleration=false)
   &ScaleSpriteSize(Scale Factor=&Multiply_Vector2D_ByFloat(
     Float=&FloatFromCurve(
      FloatCurve=DataInterface() {propertyValues {Curve {keys [
          {arriveTangent=1.4708, interpMode=RCIM_Cubic, leaveTangent=1.4708, tangentMode=RCTM_User, time=0, value=0.875}
          {arriveTangent=1.125, interpMode=RCIM_Cubic, leaveTangent=1.125, time=0.6, value=1.625}
          {interpMode=RCIM_Cubic, tangentMode=RCTM_User, time=1, value=2}
         ]}}}
      Scale Curve=0.25
     )
     Vector2D=<1 1>
    )) {module=ScaleSpriteSize001}
  ]
  rendererClasses [&NiagaraSpriteRendererProperties]
  renderers [{rendererClass=&NiagaraSpriteRendererProperties}]
 }
 ParticleDecal {
  emitterUpdateScript [
   &EmitterState()
   &SpawnBurst_Instantaneous(Spawn Count=15, Use Spawn Probability=true, UseLoopCountLimit=false)
  ]
  particleSpawnScript [
   &PartitionParticles(Number Of Partitions=26)
   &InitializeParticle(
    Color=<rgba 1 0.2549 0.0234 1>
    Lifetime=&Multiply_Float(
     A=&MakeCustomFloatFromBool(Bool=$User.IsHeavyImpact, True Float=0.7, False Float=0.6)
     B=&RandomRangeFloat(Minimum=0.4, Maximum=0.6)
    )
    Sprite Rotation Mode=::NewEnumerator2
    Sprite Size Mode=::NewEnumerator3
    Uniform Sprite Size=&Multiply_Float(A=45, B=1)
    UsePositionOffset=false
    Write Material Random=true
   )
   &CameraOffset(Camera Offset Amount=20) {module=CameraOffset001}
   &DynamicMaterialParameters(
    Index 0 Param 1=1
    Index 0 Param 2=1
    Module.Index0Float0Write=true
    Module.Index0Float1Write=true
    Module.Index0Float2Write=true
    Module.Index0Float3Write=true
   )
   /SetVariables(Particles.EnergyDP:Vector4f=<2 1 8 1>) {isSetParameters=true}
  ]
  particleUpdateScript [
   &ParticleState()
   &SkeletalMeshLocation(
    Bone / Socket Index=$Particles.PartitionParticles.Partition
    Bone Sampling Mode=::NewEnumerator4
    MaskSpawnGroup=false
    Skeletal Mesh=$System.SK_Mesh
   )
   &ScaleColor(
    Scale Mode=::NewEnumerator1
    Scale RGBA=&MakeVector4_FromLinearColor(LinearColor=<rgba 1 0.3998 0.1409 1>)
    ScaleRGBA=true
   ) {module=ScaleColor001}
   &ScaleColor(
    Scale Alpha=&FloatFromCurve(
     FloatCurve=DataInterface() {propertyValues {Curve {keys [
         {arriveTangent=9.0725, interpMode=RCIM_Cubic, leaveTangent=9.0725, tangentMode=RCTM_User, time=0, value=1}
         {arriveTangent=0.3995, interpMode=RCIM_Cubic, leaveTangent=0.3995, tangentMode=RCTM_User, time=0.1408, value=0.9847}
         {arriveTangent=-0.0405, interpMode=RCIM_Cubic, leaveTangent=-0.0405, tangentMode=RCTM_User, time=1, value=0}
        ]}}}
     Scale Curve=&Multiply_Float(A=$System.NearCameraFade, B=$System.TOD)
    )
    Scale RGB=&VectorFromFloat(Value=&FloatFromCurve(
      FloatCurve=DataInterface() {propertyValues {Curve {keys [
          <at 0 1 RCIM_Cubic>
          {arriveTangent=-1.75, interpMode=RCIM_Cubic, leaveTangent=-1.75, time=0.2, value=0.9}
          {arriveTangent=-1.125, interpMode=RCIM_Cubic, leaveTangent=-1.125, time=0.4, value=0.3}
          <at 1 0 RCIM_Cubic>
         ]}}}
      Scale Curve=10
     ))
    ScaleA=true
    ScaleRGB=true
   )
   /SetVariables(Particles.EnergyColor:LinearColor=&MultiplyLinearColor(Linear Color=$Particles.Color, Scale Factor=&MakeVector4_FromFloat(Float=30))) {isSetParameters=true}
  ]
  rendererClasses [&NiagaraSpriteRendererProperties]
  renderers [
   {rendererClass=&NiagaraSpriteRendererProperties}
   {rendererClass=&NiagaraSpriteRendererProperties, rendererIndex=1}
   {rendererClass=&NiagaraSpriteRendererProperties, rendererIndex=2}
  ]
 }
}
summary {
 systemName=NS_CoffeePuff_FireBreath_ImpactOnPlayer
 emitters [
  {emitterName=ParticleDecal, rendererClasses [&NiagaraSpriteRendererProperties]}
  {emitterName=Fire_Sprite001, rendererClasses [&NiagaraSpriteRendererProperties]}
 ]
 userVariables [
  User.IsHeavyImpact {type=NiagaraBool}
  User.IsLocalPlayer {type=NiagaraBool, defaultValue=true}
 ]
}
