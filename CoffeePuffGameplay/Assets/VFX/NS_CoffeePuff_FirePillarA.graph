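# Niagara one-shot fire pillar effect: a short burst of flame, embers, flash and swirls.
# Two CPU flame sprite layers pair with GPU embers driven by curl noise, drag and randomized forces.
# A flare drives a light, a decal fades over the burst, and four partitioned ribbons are repositioned each update along cylinder-based swirls.
# System time-of-day scalar dims most layers, while a distance scalar enlarges embers.
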NiagaraSystem /CoffeePuffGameplay/Assets/VFX/NS_CoffeePuff_FirePillarA
dependencies {
 usedDataInterfaces [
  &NiagaraDataInterfaceCurve
  &NiagaraDataInterfaceColorCurve
  &NiagaraDataInterfaceVectorCurve
  &NiagaraDataInterfaceVector2DCurve
  &NiagaraDataInterfaceArrayFloat
  &NiagaraDataInterfaceSkeletalMesh
 ]
 usedDynamicInputs [
  &FN_LOD_Distance_Remap
  &VectorFromCurve
  &FNTimeOfDayScalar
 ]
 usedModules [
  &self:SystemSpawnScript.SystemScriptSource.SystemScriptGraph.NiagaraNodeAssignment_0.SetVariables_#1
  &self:SystemSpawnScript.SystemScriptSource.SystemScriptGraph.NiagaraNodeAssignment_2.SetVariables_#2
  &SystemState
  &self:SystemSpawnScript.SystemScriptSource.SystemScriptGraph.NiagaraNodeAssignment_1.SetVariables_#3
 ]
}
emitters {
 Embers_Burst {
  simTarget=GPUComputeSim
  emitterUpdateScript [
   &EmitterState(Life Cycle Mode=::NewEnumerator1, Loop Behavior=::NewEnumerator1, Loop Duration=0.8, UseLoopDelay=false)
   &SpawnRate(SpawnRate=1500, Use Spawn Probability=false)
   &SpawnBurst_Instantaneous(Spawn Count=500, Spawn Time=0.1, Use Spawn Probability=false, UseLoopCountLimit=false)
  ]
  particleSpawnScript [
   &InitializeParticle(
    Color=<rgba 1 1 1 2>
    Lifetime=&UniformRangedFloat(Minimum=0.3, Maximum=0.9)
    Mass Mode=::NewEnumerator1
    Sprite Size=&Vector2DFromFloat(Value=&FloatFromCurve(
      FloatCurve=DataInterface() {propertyValues {Curve {keys [<at 0 1.2>,<at 0.1257 0.9847>,<at 0.2996 0.5>,<at 1 0.3>]}}}
      CurveIndex=&UniformRangedFloat(Minimum=0, Maximum=1)
      Scale Curve=2
     ))
    Sprite Size Mode=::NewEnumerator4
    UsePositionOffset=false
    Write Material Random=true
   ) {module=InitializeParticle001}
   &ShapeLocation(
    Handle Radius=60
    Large Radius=&Lerp_Float(
     A=20
     B=120
     Alpha=&FloatFromCurve(FloatCurve=DataInterface() {propertyValues {Curve {keys [<at 0 1 RCIM_Cubic>,<at 1 0 RCIM_Cubic>]}}}, CurveIndex=$Emitter.Age)
    )
    MaskSpawnGroup=false
    Offset Mode=::NewEnumerator0
    Override Seed=false
    Rotation Mode=::NewEnumerator1
    Shape Primitive=::NewEnumerator3
    UseOwnerScale=true
   )
   &AddVelocity(
    Module.Use Falloff Distance=false
    Origin Offset=<-150 0 0>
    Override Seed=false
    Velocity Mode=::NewEnumerator1
    Velocity Speed=&Multiply_Float(A=&RandomRangeFloat(Minimum=0.2, Maximum=1), B=300)
   )
   &AddVelocity(
    Cone Angle=&RandomRangeFloat(Minimum=35, Maximum=60)
    Module.UseConeFalloff=false
    Module.UseVelDistribution=false
    Override Seed=false
    Velocity Mode=::NewEnumerator2
    Velocity Speed=&Multiply_Float(
     A=&RandomRangeFloat(Minimum=0.8, Maximum=1)
     B=&Multiply_Float(
      A=2000
      B=&FloatFromCurve(FloatCurve=DataInterface() {propertyValues {Curve {keys [<at 0 1 RCIM_Cubic>,<at 1 0 RCIM_Cubic>]}}}, CurveIndex=$Emitter.Age)
     )
    )
   ) {module=AddVelocity001}
  ]
  particleUpdateScript [
   &ParticleState()
   &ScaleSpriteSize(
    Scale Sprite Size Mode=::NewEnumerator0
    Uniform Scale Factor=&Lerp_Float(A=1.5, B=3, Alpha=$System.DistanceScalar)
   )
   &ScaleSpriteSize(Uniform Curve Sprite Scale=DataInterface() {propertyValues {Curve {keys [{arriveTangent=2.4391, interpMode=RCIM_Cubic, leaveTangent=2.4391, tangentMode=RCTM_User, time=0, value=0.325}, <at 1 1 RCIM_Cubic>]}}}) {module=ScaleSpriteSize001}
   &AccelerationForce(Acceleration=&UniformRangedVector(Minimum=<0 0 800>, Maximum=<0 0 200>))
   &Drag(
    Drag=&Multiply_Float(
     A=&FloatFromCurve(
      FloatCurve=DataInterface() {propertyValues {Curve {keys [<at 0 0.5>, {arriveTangent=0.5, interpMode=RCIM_Cubic, leaveTangent=0.5, time=0.3, value=0.5}, <at 1 1 RCIM_Cubic>]}}}
      Scale Curve=10
     )
     B=1
    )
    Use Linear Drag=true
    Use Rotational Drag=false
   )
   /Niagara/Modules/Update/Forces/CurlNoiseForce(
    Noise Frequency=10
    Noise Strength=&FloatFromCurve(FloatCurve=DataInterface() {propertyValues {Curve {keys [<at 0 1>,<at 1 1>]}}}, Scale Curve=3000)
    Pan Noise=true
    Pan Noise Field=<0.1 0.1 0.5>
   )
   &GravityForce(Gravity=&RandomRangeVector(Minimum=<0 0 900>, Maximum=<0 0 -500>))
   &SolveForcesAndVelocity(Clamp Velocity=false, Limit Acceleration=false)
   &ScaleSpriteSizeBySpeed(
    Max Scale Factor=<0.8 4>
    Min Scale Factor=<0.8 0.5>
    Sample Scale Factor By Curve=true
    Scale Factor Curve=DataInterface() {propertyValues {Curve {keys [
        <at -0.0077 0.0765 RCIM_Cubic>
        {arriveTangent=1.2463, interpMode=RCIM_Cubic, leaveTangent=1.2463, tangentMode=RCTM_User, time=0.9985, value=0.9549}
       ]}}}
    Velocity Threshold=300
   )
   &ScaleColor(
    Scale Alpha=&FloatFromCurve(FloatCurve=DataInterface() {propertyValues {Curve {keys [<at 0.25 1>,<at 1 0>]}}})
    Scale RGB=&MakeVectorFromLinearColorRGB(Color=&ColorFromCurve(ColorCurve=DataInterface() {propertyValues {
        AlphaCurve {keys [<at 0 1 RCIM_Cubic>,<at 0.7107 1>]}
        BlueCurve {keys [<at 0 14.6701 RCIM_Cubic>,<at 0.368 3.776>,<at 0.855 0.1753 RCIM_Cubic>]}
        GreenCurve {keys [<at 0 36.7481 RCIM_Cubic>,<at 0.368 13.2325>,<at 0.855 0.3243 RCIM_Cubic>]}
        RedCurve {keys [<at 0 50 RCIM_Cubic>,<at 0.368 25>,<at 0.855 1 RCIM_Cubic>]}
       }}))
    ScaleA=true
    ScaleRGB=true
   )
   &ScaleColor(
    Scale Alpha=$System.TimeOfDay
    Scale RGB=&VectorFromFloat(Value=$System.TimeOfDay)
    ScaleA=true
    ScaleRGB=true
   ) {module=ScaleColor001}
  ]
  rendererClasses [&NiagaraSpriteRendererProperties]
  renderers [{rendererClass=&NiagaraSpriteRendererProperties}]
 }
 FireFast_V_aligned {
  emitterUpdateScript [
   &EmitterState(Life Cycle Mode=::NewEnumerator1, Loop Behavior=::NewEnumerator1, Loop Duration=0.1, UseLoopDelay=false)
   &SpawnBurst_Instantaneous(Spawn Count=40, Spawn Time=0.1, Use Spawn Probability=false, UseLoopCountLimit=false)
   &SpawnRate(SpawnRate=150, Use Spawn Probability=false)
  ]
  particleSpawnScript [
   &InitializeParticle(
    Lifetime Max=0.45
    Lifetime Min=0.2
    Lifetime Mode=::NewEnumerator1
    Sprite Size=<120 300>
    Sprite Size Mode=::NewEnumerator4
    Sprite UV Mode=::NewEnumerator3
    UsePositionOffset=false
    Write Material Random=true
   )
   &DynamicMaterialParameters(
    Index 0 Param 2=&RandomRangeFloat(Minimum=0, Maximum=4)
    Index 0 Param 4=2.5
    Module.Param0WriteEnabled=false
    Module.Param1WriteEnabled=true
    Module.Param2WriteEnabled=false
    Module.Param3WriteEnabled=true
   ) {module=DynamicMaterialParameters001}
   &ShapeLocation(
    MaskSpawnGroup=false
    Offset Mode=::NewEnumerator0
    Override Seed=false
    Rotation Mode=::NewEnumerator1
    Shape Primitive=::NewEnumerator3
    UseOwnerScale=true
   ) {module=ShapeLocation001}
   &AddVelocity(
    Module.Use Falloff Distance=false
    Origin Offset=<0 0 -200>
    Origin Offset Coordinate Space=::World
    Override Seed=false
    Velocity Mode=::NewEnumerator1
    Velocity Speed=500
   ) {module=AddVelocity001}
  ]
  particleUpdateScript [
   &ParticleState()
   &Drag(Drag=3, Use Linear Drag=true, Use Rotational Drag=false)
   &ScaleColor(
    Linear Color Curve=DataInterface() {propertyValues {
      AlphaCurve {keys [<at 0 1>,<at 1 0>]}
      BlueCurve {keys [<at 0.0419 11.2101>,<at 0.2881 2.242>,<at 0.6965 0.565>,<at 0.9695 2.1713>]}
      GreenCurve {keys [<at 0.0419 14.4749>,<at 0.2881 2.895>,<at 0.6965 0.6614>,<at 0.9695 3.074>]}
      RedCurve {keys [<at 0.0419 20>,<at 0.2881 4>,<at 0.6965 1>,<at 0.9695 5>]}
     }}
    Scale Mode=::NewEnumerator2
   ) {module=ScaleColor001}
   &ScaleColor(
    Scale Alpha=&FloatFromCurve(FloatCurve=DataInterface() {propertyValues {Curve {keys [
         {arriveTangent=6.7724, interpMode=RCIM_Cubic, leaveTangent=6.7724, tangentMode=RCTM_User, time=0, value=0}
         {interpMode=RCIM_Cubic, leaveTangent=-5.6421, tangentMode=RCTM_Break, time=0.5, value=1}
         {arriveTangent=-2, interpMode=RCIM_Cubic, leaveTangent=-2, time=0.7, value=0.2}
         {arriveTangent=-0.1613, interpMode=RCIM_Cubic, leaveTangent=-0.1613, tangentMode=RCTM_User, time=1, value=0}
        ]}}})
    ScaleA=true
    ScaleRGB=true
   )
   &SolveForcesAndVelocity(Clamp Velocity=false, Limit Acceleration=false)
   &ScaleSpriteSize(Scale Factor=&Multiply_Vector2D_ByFloat(
     Float=&FloatFromCurve(FloatCurve=DataInterface() {propertyValues {Curve {keys [
          {arriveTangent=1.4708, interpMode=RCIM_Cubic, leaveTangent=1.4708, tangentMode=RCTM_User, time=0, value=0.525}
          {arriveTangent=0.625, interpMode=RCIM_Cubic, leaveTangent=0.625, time=0.7, value=1.05}
          {interpMode=RCIM_Cubic, tangentMode=RCTM_User, time=1, value=1.15}
         ]}}})
     Vector2D=<0.8 1.2>
    )) {module=ScaleSpriteSize001}
   &ScaleSpriteSize(
    Non-Uniform Curve Scale=<3 1>
    Non-Uniform Curve Sprite Scale=DataInterface() {propertyValues {
      XCurve {keys [
        {arriveTangent=2.6926, interpMode=RCIM_Cubic, leaveTangent=2.6926, tangentMode=RCTM_User, time=0, value=0}
        {arriveTangent=0.1227, interpMode=RCIM_Cubic, leaveTangent=0.1227, tangentMode=RCTM_User, time=1, value=1}
       ]}
      YCurve {keys [{tangentMode=RCTM_User, time=0, value=1}, {tangentMode=RCTM_User, time=1, value=1}]}
     }}
    Scale Sprite Size Mode=::NewEnumerator3
   ) {module=ScaleSpriteSize002}
   &ScaleColor(Scale Alpha=$System.TimeOfDay, Scale RGB=&VectorFromFloat(Value=1), ScaleA=true, ScaleRGB=true) {module=ScaleColor002}
  ]
  rendererClasses [&NiagaraSpriteRendererProperties]
  renderers [{rendererClass=&NiagaraSpriteRendererProperties}]
 }
 Fire_Sprite {
  emitterUpdateScript [
   &EmitterState(Life Cycle Mode=::NewEnumerator1, Loop Behavior=::NewEnumerator1, Loop Delay=0.2, Loop Duration=0.4, UseLoopDelay=true)
   &SpawnBurst_Instantaneous(Spawn Count=15, Use Spawn Probability=false, UseLoopCountLimit=false)
   &SpawnRate(SpawnRate=150, Use Spawn Probability=false)
  ]
  particleSpawnScript [
   &InitializeParticle(
    Lifetime Max=0.8
    Lifetime Min=0.5
    Lifetime Mode=::NewEnumerator1
    Sprite Rotation Mode=::NewEnumerator2
    Sprite Size Mode=::NewEnumerator1
    Uniform Sprite Size Max=150
    Uniform Sprite Size Min=100
    UsePositionOffset=false
    Write Material Random=true
   )
   &DynamicMaterialParameters(
    Index 0 Param 2=&RandomRangeFloat(Minimum=0, Maximum=4)
    Index 0 Param 4=2.5
    Module.Param0WriteEnabled=false
    Module.Param1WriteEnabled=true
    Module.Param2WriteEnabled=false
    Module.Param3WriteEnabled=true
   ) {module=DynamicMaterialParameters001}
   &AddVelocity(
    Cone Angle=40
    Module.UseConeFalloff=false
    Module.UseVelDistribution=false
    Override Seed=false
    Velocity Mode=::NewEnumerator2
    Velocity Speed=&Multiply_Float(
     A=&RandomRangeFloat(Minimum=0.4, Maximum=1)
     B=&Multiply_Float(
      A=1200
      B=&FloatFromCurve(FloatCurve=DataInterface() {propertyValues {Curve {keys [<at 0 1 RCIM_Cubic>,<at 0.6 0 RCIM_Cubic>]}}}, CurveIndex=$Emitter.Age)
     )
    )
   ) {module=AddVelocity001}
   &ShapeLocation(
    MaskSpawnGroup=false
    Offset=<0 0 -25>
    Offset Coordinate Space=::World
    Offset Mode=::NewEnumerator0
    Override Seed=false
    Rotation Mode=::NewEnumerator1
    Shape Primitive=::NewEnumerator3
    UseOwnerScale=true
   )
  ]
  particleUpdateScript [
   &ParticleState()
   &AccelerationForce(Acceleration=&UniformRangedVector(Minimum=<0 0 500>, Maximum=<0 0 200>))
   &VortexForce(UseOriginOffset=false, Vortex Force Amount=1500)
   &Drag(Drag=5, Use Linear Drag=true, Use Rotational Drag=false)
   &ScaleColor(
    Linear Color Curve=DataInterface() {propertyValues {
      AlphaCurve {keys [{arriveTangent=-0.1596, interpMode=RCIM_Cubic, leaveTangent=-0.1596, tangentMode=RCTM_User, time=0.8987, value=1}, <at 0.982 0 RCIM_Cubic>]}
      BlueCurve {keys [
        <at 0.0505 3.4995 RCIM_Cubic>
        {arriveTangent=-3.8023, interpMode=RCIM_Cubic, leaveTangent=-3.8023, time=0.3055, value=1.1959}
        {arriveTangent=-1.0615, interpMode=RCIM_Cubic, leaveTangent=-1.0615, time=0.7879, value=0.6957}
        <at 0.9823 0.4774>
       ]}
      GreenCurve {keys [
        <at 0.0505 4.0716 RCIM_Cubic>
        {arriveTangent=-4.5324, interpMode=RCIM_Cubic, leaveTangent=-4.5324, time=0.3055, value=1.3562}
        {arriveTangent=-1.1361, interpMode=RCIM_Cubic, leaveTangent=-1.1361, time=0.7879, value=0.7294}
        <at 0.9823 0.5872>
       ]}
      RedCurve {keys [
        <at 0.0505 5 RCIM_Cubic>
        {arriveTangent=-5.4244, interpMode=RCIM_Cubic, leaveTangent=-5.4244, time=0.3055, value=2}
        {arriveTangent=-1.4775, interpMode=RCIM_Cubic, leaveTangent=-1.4775, time=0.7879, value=1}
        <at 0.9823 1>
       ]}
     }}
    Scale Mode=::NewEnumerator2
   ) {module=ScaleColor001}
   &ScaleColor(
    Scale Alpha=&FloatFromCurve(
     FloatCurve=DataInterface() {propertyValues {Curve {keys [
         {arriveTangent=6.7724, interpMode=RCIM_Cubic, leaveTangent=6.7724, tangentMode=RCTM_User, time=0, value=0}
         {interpMode=RCIM_Cubic, tangentMode=RCTM_User, time=0.1, value=0.95}
         {arriveTangent=-1.0556, interpMode=RCIM_Cubic, leaveTangent=-1.0556, time=0.495, value=0.3596}
         {arriveTangent=-0.0646, interpMode=RCIM_Cubic, leaveTangent=-0.0646, tangentMode=RCTM_User, time=1, value=0}
        ]}}}
     Scale Curve=0.8
    )
    Scale RGB=&VectorFromFloat(Value=&FloatFromCurve(
      FloatCurve=DataInterface() {propertyValues {Curve {keys [
          <at 0.0113 0.5266 RCIM_Cubic>
          {arriveTangent=-0.4315, interpMode=RCIM_Cubic, leaveTangent=-0.4315, time=0.1245, value=0.1785}
          <at 1 0.1 RCIM_Cubic>
         ]}}}
      Scale Curve=10
     ))
    ScaleA=true
    ScaleRGB=true
   )
   &SolveForcesAndVelocity(Clamp Velocity=false, Limit Acceleration=false)
   &ScaleSpriteSize(Scale Sprite Size Mode=::NewEnumerator0, Uniform Scale Factor=0.6)
   &ScaleSpriteSize(Scale Factor=&Multiply_Vector2D_ByFloat(
     Float=&FloatFromCurve(FloatCurve=DataInterface() {propertyValues {Curve {keys [
          {arriveTangent=1.4708, interpMode=RCIM_Cubic, leaveTangent=1.4708, tangentMode=RCTM_User, time=0, value=1.3}
          {arriveTangent=0.7, interpMode=RCIM_Cubic, leaveTangent=0.7, time=0.6, value=1.8}
          {interpMode=RCIM_Cubic, tangentMode=RCTM_User, time=1, value=2}
         ]}}})
     Vector2D=<1 1>
    )) {module=ScaleSpriteSize001}
   &ScaleColor(Scale Alpha=$System.TimeOfDay, Scale RGB=&VectorFromFloat(Value=1), ScaleA=true, ScaleRGB=true) {module=ScaleColor002}
  ]
  rendererClasses [&NiagaraSpriteRendererProperties]
  renderers [{rendererClass=&NiagaraSpriteRendererProperties}]
 }
 Flash_Flare {
  emitterUpdateScript [
   &EmitterState(Life Cycle Mode=::NewEnumerator1, Loop Behavior=::NewEnumerator1, Loop Duration=1, UseLoopDelay=false)
   &SpawnBurst_Instantaneous(Spawn Count=1, Spawn Time=0.1, Use Spawn Probability=false)
  ]
  particleSpawnScript [
   &InitializeParticle(
    Color=<rgba 1 0.2274 0.0317 1>
    Lifetime=0.6
    Sprite Size=<600 1200>
    Sprite Size Mode=::NewEnumerator4
    Sprite UV Mode=::NewEnumerator3
    UsePositionOffset=false
    Write Material Random=true
   )
   &DynamicMaterialParameters(
    Index 0 Param 1=1
    Index 0 Param 2=1
    Module.Index0Float0Write=true
    Module.Index0Float1Write=true
    Module.Index0Float2Write=true
    Module.Index0Float3Write=true
   )
   /SetVariables(Particles.ParticleAlignmentVector:Vector3f=<1 0 0>) {isSetParameters=true}
   &SystemLocation(MaskSpawnGroup=false, Offset=<250 0 0>)
  ]
  particleUpdateScript [
   &ParticleState()
   &ScaleSpriteSize(Scale Sprite Size Mode=::NewEnumerator0, Uniform Scale Factor=2)
   /SetVariables(
    Particles.FireColor:LinearColor=&MakeLinearColorFromVectorAndFloat(
     "Vector (RGB)"=&VectorFromFloat(Value=&Multiply_Float(
       A=2
       B=&FloatFromCurve(FloatCurve=DataInterface() {propertyValues {Curve {keys [
            {arriveTangent=0.0345, interpMode=RCIM_Cubic, leaveTangent=0.0345, tangentMode=RCTM_User, time=0.0729, value=1}
            <at 0.9716 0.4635 RCIM_Cubic>
           ]}}})
      ))
     "Float (Alpha)"=&FloatFromCurve(FloatCurve=DataInterface() {propertyValues {Curve {keys [<at 0 1 RCIM_Cubic>, {arriveTangent=-1, interpMode=RCIM_Cubic, leaveTangent=-1, time=0.3, value=0.3}, <at 1 0 RCIM_Cubic>]}}})
    )
    Particles.FireSize:Vector2f=&Multiply_Vector2D(A=$Particles.SpriteSize, B=&Multiply_Vector2D_ByFloat(Float=1, Vector2D=<0.2 0.2>))
   ) {isSetParameters=true}
   &ScaleColor(
    Scale Alpha=&FloatFromCurve(FloatCurve=DataInterface() {propertyValues {Curve {keys [
         {arriveTangent=16.2782, interpMode=RCIM_Cubic, leaveTangent=16.2782, tangentMode=RCTM_User, time=0, value=0}
         {arriveTangent=0.1596, interpMode=RCIM_Cubic, leaveTangent=0.1596, tangentMode=RCTM_User, time=0.1, value=1}
         {arriveTangent=-1.1111, interpMode=RCIM_Cubic, leaveTangent=-1.1111, time=0.3, value=0.3}
         {arriveTangent=0.0036, interpMode=RCIM_Cubic, leaveTangent=0.0036, tangentMode=RCTM_User, time=1, value=0}
        ]}}})
    Scale RGB=&VectorFromFloat(Value=1)
    ScaleA=true
    ScaleRGB=true
   )
   &ScaleColor(
    Scale Alpha=$System.TimeOfDay
    Scale RGB=&VectorFromFloat(Value=$System.TimeOfDay)
    ScaleA=true
    ScaleRGB=true
   ) {module=ScaleColor001}
   &FN_Light_Attributes(
    Color=$Particles.Color
    Create Color Attribute=true
    Create Radius Attribute=true
    Radius=500
    Scale Color=&FloatFromCurve(
     FloatCurve=DataInterface() {propertyValues {Curve {keys [
         <at 0 0.05 RCIM_Cubic>
         {arriveTangent=-0.05, interpMode=RCIM_Cubic, leaveTangent=-0.05, time=0.1, value=0.95}
         {interpMode=RCIM_Cubic, leaveTangent=-8.0089, time=1, value=0}
        ]}}}
     Scale Curve=100
    )
   )
  ]
  rendererClasses [&NiagaraSpriteRendererProperties, &NiagaraLightRendererProperties]
  renderers [
   {rendererClass=&NiagaraSpriteRendererProperties}
   {rendererClass=&NiagaraLightRendererProperties, rendererIndex=1}
  ]
 }
 Flash_Flare002 {
  emitterUpdateScript [
   &EmitterState()
   &SpawnBurst_Instantaneous(Spawn Count=1, Use Spawn Probability=true)
  ]
  particleSpawnScript [
   &InitializeParticle(Lifetime=3, Sprite Size Mode=::NewEnumerator3, Uniform Sprite Size=600, UsePositionOffset=false, Write Material Random=true)
   /SetVariables(
    Particles.DecalFade=1
    Particles.DecalOrientation:Quat4f=&OrientMeshToVector(Look At Direction=<0 0 1>, Reference Axis Vector=<1 0 0>)
    Particles.DecalSize:Vector3f=<300 700 700>
   ) {isSetParameters=true}
  ]
  particleUpdateScript [
   &ParticleState()
   /SetVariables(Particles.DecalFade:NiagaraFloat=&Lerp_Float(
     A=0
     B=1
     Alpha=&FloatFromCurve(FloatCurve=DataInterface() {propertyValues {Curve {keys [
          {arriveTangent=12.3499, interpMode=RCIM_Cubic, leaveTangent=12.3499, tangentMode=RCTM_User, time=0, value=0}
          {arriveTangent=0.9375, leaveTangent=0.9375, time=0.2, value=1}
          {arriveTangent=-1.25, interpMode=RCIM_Cubic, leaveTangent=-1.25, time=0.6954, value=0.6902}
          {arriveTangent=-0.0037, interpMode=RCIM_Cubic, leaveTangent=-0.0037, tangentMode=RCTM_User, time=1, value=0}
         ]}}})
    )) {isSetParameters=true}
  ]
  rendererClasses [&NiagaraDecalRendererProperties]
  renderers [{rendererClass=&NiagaraDecalRendererProperties, rendererIndex=0}]
 }
 Swirls_Local001 {
  emitterSpawnScript [/SetVariables(Emitter.RandomInt:NiagaraInt32=&RandomRangeInt(Minimum=0, Maximum=999)) {isSetParameters=true}]
  emitterUpdateScript [
   &EmitterState(Life Cycle Mode=::NewEnumerator1, Loop Behavior=::NewEnumerator1, Loop Duration=0.7, UseLoopDelay=false)
   &SpawnRate(SpawnRate=500, Use Spawn Probability=false) {module=SpawnRate001}
  ]
  particleSpawnScript [
   &PartitionParticles(Number Of Partitions=4, Write Partition to Ribbon ID=::NewEnumerator0)
   &InitializeParticle(
    Lifetime=&Multiply_Float(
     A=&SelectFloatFromArray(Direct Array Index=$Particles.PartitionParticles.Partition, Float Selection Array=DataInterface() {propertyValues {FloatData [1,1,1,1]}})
     B=0.3
    )
    Ribbon Width=20
    Ribbon Width Mode=::NewEnumerator1
    UsePositionOffset=false
    Write Material Random=true
   )
  ]
  particleUpdateScript [
   &ParticleState()
   /SetVariables(Particles.RandomVector:Vector3f=<0 0 500>) {isSetParameters=true}
   &ShapeLocation(
    Cylinder Distribution=::NewEnumerator2
    Cylinder Height=10
    Cylinder Height Midpoint=0
    Cylinder Radius=&Multiply_Float(
     A=&Multiply_Float(
      A=50
      B=&FloatFromCurve(
       FloatCurve=DataInterface() {propertyValues {Curve {keys [<at 0 1 RCIM_Cubic>, {arriveTangent=-1.2319, interpMode=RCIM_Cubic, leaveTangent=-1.2319, tangentMode=RCTM_User, time=1, value=0.65}]}}}
       Scale Curve=3
      )
     )
     B=&SelectFloatFromArray(
      Direct Array Index=$Particles.PartitionParticles.Partition
      Float Selection Array=DataInterface() {propertyValues {FloatData [0.6,0.7,0.6,0.5]}}
     )
    )
    MaskSpawnGroup=false
    Offset=&MakeVector(
     X=&FloatFromCurve(
      FloatCurve=DataInterface() {propertyValues {Curve {keys [
          {arriveTangent=0.8615, interpMode=RCIM_Cubic, leaveTangent=0.8615, tangentMode=RCTM_User, time=0, value=0}
          {arriveTangent=2.5042, interpMode=RCIM_Cubic, leaveTangent=2.5042, tangentMode=RCTM_User, time=1, value=1}
         ]}}}
      Scale Curve=&SelectFloatFromArray(
       Direct Array Index=$Particles.PartitionParticles.Partition
       Float Selection Array=DataInterface() {propertyValues {FloatData [300,350,380,450]}}
      )
     )
     Y=0
     Z=0
    )
    Offset Mode=::NewEnumerator0
    Rotation Angle:NiagaraFloat=&Multiply_Float(A=$System.Age, B=-0.4)
    Rotation Mode=::NewEnumerator1
    Shape Primitive=::NewEnumerator1
    U Position=&Add_Float(
     A=&Multiply_Float(A=&Add_Float(A=$Particles.NormalizedAge, B=-0.5), B=0.7)
     B=&Multiply_FloatByInt(Float=0.25, Integer=$Particles.PartitionParticles.Partition)
    )
    UseOwnerScale=true
    V Position=0
   )
   &ScaleRibbonWidth(Ribbon Width Scale=&FloatFromCurve(
     FloatCurve=DataInterface() {propertyValues {Curve {keys [
         {tangentMode=RCTM_User, time=0, value=1}
         {arriveTangent=-0.0405, interpMode=RCIM_Cubic, leaveTangent=-0.0405, tangentMode=RCTM_User, time=0.7009, value=1}
         {arriveTangent=-8.0089, interpMode=RCIM_Cubic, leaveTangent=-8.0089, tangentMode=RCTM_User, time=1, value=0}
        ]}}}
     Scale Curve=3
    ))
   &ScaleColor(
    Scale Alpha=&FloatFromCurve(
     FloatCurve=DataInterface() {propertyValues {Curve {keys [
         {arriveTangent=-0.0969, interpMode=RCIM_Cubic, leaveTangent=-0.0969, tangentMode=RCTM_User, time=0.5, value=0.9869}
         {arriveTangent=-0.6092, interpMode=RCIM_Cubic, leaveTangent=-0.6092, tangentMode=RCTM_User, time=0.8094, value=0.7839}
         {arriveTangent=-4.1106, interpMode=RCIM_Cubic, leaveTangent=-4.1106, time=0.891, value=0.1279}
         {interpMode=RCIM_Cubic, leaveTangent=-2.2222, time=1.0098, value=-0.0399}
        ]}}}
     CurveIndex=$Emitter.Age
    )
    ScaleA=true
    ScaleRGB=true
   ) {module=ScaleColor002}
   &ScaleColor(Scale RGB=&VectorFromFloat(Value=$System.TimeOfDay), ScaleA=true, ScaleRGB=true) {module=ScaleColor001}
   &DynamicMaterialParameters(
    Index 0 Param 1=&FloatFromCurve(
     FloatCurve=DataInterface() {propertyValues {Curve {keys [
         {arriveTangent=-2.2284, interpMode=RCIM_Cubic, leaveTangent=-2.2284, tangentMode=RCTM_User, time=0, value=1}
         {arriveTangent=-0.2527, interpMode=RCIM_Cubic, leaveTangent=-0.2527, tangentMode=RCTM_User, time=1, value=0}
        ]}}}
     CurveIndex=$Emitter.Age
     Scale Curve=-1.5
    )
    Index 0 Param 3=&FloatFromCurve(
     FloatCurve=DataInterface() {propertyValues {Curve {keys [
         <at 0 1 RCIM_Cubic>
         {arriveTangent=-0.7209, interpMode=RCIM_Cubic, leaveTangent=-0.7209, tangentMode=RCTM_User, time=0.3351, value=0.2896}
         {arriveTangent=-0.2188, interpMode=RCIM_Cubic, leaveTangent=-0.2188, tangentMode=RCTM_User, time=1, value=0}
        ]}}}
     CurveIndex=$Emitter.Age
     Scale Curve=4
    )
    Index 0 Param 4=&SelectFloatFromArray(
     Direct Array Index=$Particles.PartitionParticles.Partition
     Float Selection Array=DataInterface() {propertyValues {FloatData [0.2,0.7,1.25,0.45]}}
    )
    Module.Index0Param0Write=true
    Module.Index0Param1Write=true
    Module.Index0Param2Write=true
    Module.Index0Param3Write=true
   )
  ]
  rendererClasses [&NiagaraRibbonRendererProperties]
  renderers [{rendererClass=&NiagaraRibbonRendererProperties}]
 }
}
summary {
 systemName=NS_CoffeePuff_FirePillarA
 emitters [
  {emitterName=Flash_Flare, rendererClasses [&NiagaraSpriteRendererProperties, &NiagaraLightRendererProperties]}
  {emitterName=Embers_Burst, simTarget=GPUComputeSim, rendererClasses [&NiagaraSpriteRendererProperties]}
  {emitterName=Fire_Sprite, rendererClasses [&NiagaraSpriteRendererProperties]}
  {emitterName=FireFast_V_aligned, rendererClasses [&NiagaraSpriteRendererProperties]}
  {emitterName=Swirls_Local001, rendererClasses [&NiagaraRibbonRendererProperties]}
  {emitterName=Flash_Flare002, rendererClasses [&NiagaraDecalRendererProperties]}
 ]
 userVariables []
}
