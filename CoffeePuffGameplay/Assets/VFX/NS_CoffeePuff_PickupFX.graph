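# Niagara pickup effect for CoffeePuff that emits fire sprites and glowing embers from two thin cylinders mirrored on either side of the origin.
# Fire_Sprite slowly spawns camera-offset, rising sprites that grow and fade, with randomized material parameters.
# GPU-simulated Embers_Burst spawns continuously at a rate switched by owner speed, adding curl-noise motion, speed-based stretch and HDR color cooling.
# Ember size also scales with a system-level DistanceScalar.
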NiagaraSystem /CoffeePuffGameplay/Assets/VFX/NS_CoffeePuff_PickupFX
dependencies {
 usedDataInterfaces [
  &NiagaraDataInterfaceArrayFloat
  &NiagaraDataInterfaceCurve
  &NiagaraDataInterfaceColorCurve
  &NiagaraDataInterfaceVectorCurve
 ]
 usedDynamicInputs [&VectorFromCurve, &FN_LOD_Distance_Remap]
 usedModules [
  &SystemState
  &self:SystemSpawnScript.SystemScriptSource.SystemScriptGraph.NiagaraNodeAssignment_0.SetVariables_#1
 ]
}
emitters {
 Embers_Burst {
  simTarget=GPUComputeSim
  emitterUpdateScript [
   &EmitterState()
   &SpawnRate(
    SpawnRate=&MakeCustomFloatFromBool(
     Bool=&SetBoolByFloatComparison(A=&VectorLength(Vector=$Engine.Owner.Velocity), B=200)
     True Float=12
     False Float=15
    )
    Use Spawn Probability=false
   )
  ]
  particleSpawnScript [
   &PartitionParticles()
   &InitializeParticle(
    Color=<rgba 1 1 1 2>
    Lifetime=&UniformRangedFloat(Minimum=0.3, Maximum=1.2)
    Mass Mode=::NewEnumerator1
    Position Offset=&MakeVector(
     X=&SelectFloatFromArray(Direct Array Index=$Particles.PartitionParticles.Partition, Float Selection Array=DataInterface() {propertyValues {FloatData [25,-25]}})
     Y=0
     Z=0
    )
    Sprite Size=&Vector2DFromFloat(Value=&FloatFromCurve(
      FloatCurve=DataInterface() {propertyValues {Curve {keys [<at 0 1.2>,<at 0.1257 0.9847>,<at 0.2996 0.5>,<at 1 0.3>]}}}
      CurveIndex=&UniformRangedFloat(Minimum=0, Maximum=1)
      Scale Curve=2
     ))
    Sprite Size Mode=::NewEnumerator4
    UsePositionOffset=true
    Write Material Random=true
   ) {module=InitializeParticle001}
   &ShapeLocation(
    Cylinder Height=35
    Cylinder Radius=1
    MaskSpawnGroup=false
    Override Seed=false
    Rotation Mode=::NewEnumerator2
    Shape Primitive=::NewEnumerator1
    Surface Only=false
    UseOwnerScale=true
    Yaw / Pitch / Roll=<0 0 0.3>
   )
  ]
  particleUpdateScript [
   &ParticleState()
   &ScaleSpriteSize(
    Scale Sprite Size Mode=::NewEnumerator0
    Uniform Scale Factor=&Lerp_Float(A=1, B=2.5, Alpha=$System.DistanceScalar)
   )
   &ScaleSpriteSize(Uniform Curve Sprite Scale=DataInterface() {propertyValues {Curve {keys [{arriveTangent=2.4391, interpMode=RCIM_Cubic, leaveTangent=2.4391, tangentMode=RCTM_User, time=0, value=0.325}, <at 1 1 RCIM_Cubic>]}}}) {module=ScaleSpriteSize001}
   &AccelerationForce(Acceleration=&UniformRangedVector(Minimum=<0 0 50>, Maximum=<0 0 100>))
   &Drag(Drag=5, Use Linear Drag=true, Use Rotational Drag=false)
   /Niagara/Modules/Update/Forces/CurlNoiseForce(Noise Frequency=10, Noise Strength=1000, Pan Noise=true, Pan Noise Field=<0.1 0.1 0.5>)
   &GravityForce(Gravity=&RandomRangeVector(Minimum=<0 0 900>, Maximum=<0 0 -500>))
   &SolveForcesAndVelocity(Clamp Velocity=false, Limit Acceleration=false)
   &ScaleSpriteSizeBySpeed(
    Max Scale Factor=<0.8 5>
    Min Scale Factor=<0.8 0.8>
    Sample Scale Factor By Curve=true
    Scale Factor Curve=DataInterface() {propertyValues {Curve {keys [
        <at -0.0077 0.0765 RCIM_Cubic>
        {arriveTangent=1.2463, interpMode=RCIM_Cubic, leaveTangent=1.2463, tangentMode=RCTM_User, time=0.9985, value=0.9549}
       ]}}}
    Velocity Threshold=300
   )
   &ScaleColor(
    Scale Alpha=&FloatFromCurve(FloatCurve=DataInterface() {propertyValues {Curve {keys [<at 0.25 1>,<at 1 0>]}}})
    Scale RGB=&MakeVectorFromLinearColorRGB(Color=&ColorFromCurve(ColorCurve=DataInterface() {propertyValues {
        AlphaCurve {keys [<at 0 1 RCIM_Cubic>,<at 0.7107 1>]}
        BlueCurve {keys [<at 0 14.6701 RCIM_Cubic>,<at 0.368 3.776>,<at 0.855 0.1753 RCIM_Cubic>]}
        GreenCurve {keys [<at 0 36.7481 RCIM_Cubic>,<at 0.368 13.2325>,<at 0.855 0.3243 RCIM_Cubic>]}
        RedCurve {keys [<at 0 50 RCIM_Cubic>,<at 0.368 25>,<at 0.855 1 RCIM_Cubic>]}
       }}))
    ScaleA=true
    ScaleRGB=true
   )
  ]
  rendererClasses [&NiagaraSpriteRendererProperties]
  renderers [{rendererClass=&NiagaraSpriteRendererProperties}]
 }
 Fire_Sprite {
  emitterUpdateScript [
   &EmitterState()
   &SpawnRate(SpawnRate=1.5, Use Spawn Probability=false)
  ]
  particleSpawnScript [
   &PartitionParticles()
   &InitializeParticle(
    Lifetime Max=2.8
    Lifetime Min=2.5
    Lifetime Mode=::NewEnumerator1
    Position=/Niagara/DynamicInputs/Helpers/SimulationPosition()
    Position Mode=::NewEnumerator1
    Position Offset=&MakeVector(
     X=&SelectFloatFromArray(Direct Array Index=$Particles.PartitionParticles.Partition, Float Selection Array=DataInterface() {propertyValues {FloatData [28,-28]}})
     Y=0
     Z=-15
    )
    Sprite Rotation Mode=::NewEnumerator2
    Sprite Size Mode=::NewEnumerator3
    Uniform Sprite Size=&Multiply_Float(A=&UniformRangedFloat(Minimum=50, Maximum=80), B=0.3)
    UsePositionOffset=true
    Write Material Random=true
   ) {module=InitializeParticle001}
   &DynamicMaterialParameters(
    Index 0 Param 2=&RandomRangeFloat(Minimum=0, Maximum=4)
    Index 0 Param 4=2.5
    Module.Param0WriteEnabled=false
    Module.Param1WriteEnabled=true
    Module.Param2WriteEnabled=false
    Module.Param3WriteEnabled=true
   ) {module=DynamicMaterialParameters001}
   &CameraOffset(Camera Offset Amount=&UniformRangedFloat(Minimum=20, Maximum=30))
   &ShapeLocation(
    Cylinder Height=20
    Cylinder Radius=1
    MaskSpawnGroup=false
    Offset Coordinate Space=::Local
    Offset Mode=::NewEnumerator0
    Override Seed=false
    Rotation Mode=::NewEnumerator2
    Shape Primitive=::NewEnumerator1
    Surface Only=false
    UseOwnerScale=true
    Yaw / Pitch / Roll=<0 0 0.3>
   )
  ]
  particleUpdateScript [
   &ParticleState()
   &ScaleSpriteSize(Uniform Curve Sprite Scale=DataInterface() {propertyValues {Curve {keys [
        {interpMode=RCIM_Cubic, leaveTangent=1.9619, tangentMode=RCTM_Break, time=0, value=0.5}
        {arriveTangent=0.1394, tangentMode=RCTM_User, time=0.7, value=1}
       ]}}})
   &Drag(Drag=2, Use Linear Drag=true, Use Rotational Drag=false) {module=Drag001}
   &AccelerationForce(Acceleration=&UniformRangedVector(Minimum=<0 0 20>, Maximum=<0 0 25>))
   &ScaleColor(
    Linear Color Curve=DataInterface() {propertyValues {
      AlphaCurve {keys [{arriveTangent=-0.1596, interpMode=RCIM_Cubic, leaveTangent=-0.1596, tangentMode=RCTM_User, time=0.8085, value=1}, <at 0.982 0 RCIM_Cubic>]}
      BlueCurve {keys [
        <at 0.0505 0.6999 RCIM_Cubic>
        {arriveTangent=-0.0647, interpMode=RCIM_Cubic, leaveTangent=-0.0647, time=0.3055, value=1.1959}
        {arriveTangent=-1.0615, interpMode=RCIM_Cubic, leaveTangent=-1.0615, time=0.7879, value=0.6522}
        <at 0.9823 0.4774>
       ]}
      GreenCurve {keys [
        <at 0.0505 0.8143 RCIM_Cubic>
        {arriveTangent=-0.1676, interpMode=RCIM_Cubic, leaveTangent=-0.1676, time=0.3055, value=1.3562}
        {arriveTangent=-1.1361, interpMode=RCIM_Cubic, leaveTangent=-1.1361, time=0.7879, value=0.6908}
        <at 0.9823 0.5872>
       ]}
      RedCurve {keys [
        <at 0.0505 1 RCIM_Cubic>
        <at 0.3055 2 RCIM_Cubic>
        {arriveTangent=-1.4775, interpMode=RCIM_Cubic, leaveTangent=-1.4775, time=0.7879, value=1}
        <at 0.9823 1>
       ]}
     }}
    Scale Mode=::NewEnumerator2
   ) {module=ScaleColor001}
   &ScaleColor(
    Scale Alpha=&FloatFromCurve(FloatCurve=DataInterface() {propertyValues {Curve {keys [
         {arriveTangent=6.7724, interpMode=RCIM_Cubic, leaveTangent=6.7724, tangentMode=RCTM_User, time=0, value=0}
         {interpMode=RCIM_Cubic, tangentMode=RCTM_User, time=0.5, value=1}
         {arriveTangent=-3.0347, interpMode=RCIM_Cubic, leaveTangent=-3.0347, tangentMode=RCTM_User, time=1, value=0}
        ]}}})
    ScaleA=true
    ScaleRGB=true
   )
   &SolveForcesAndVelocity(Clamp Velocity=false, Limit Acceleration=false)
   &ScaleSpriteSize(Scale Factor=&Multiply_Vector2D_ByFloat(
     Float=&FloatFromCurve(FloatCurve=DataInterface() {propertyValues {Curve {keys [
          {arriveTangent=1.4708, interpMode=RCIM_Cubic, leaveTangent=1.4708, tangentMode=RCTM_User, time=0, value=0.75}
          {arriveTangent=1.25, interpMode=RCIM_Cubic, leaveTangent=1.25, time=0.6, value=1.625}
          {interpMode=RCIM_Cubic, tangentMode=RCTM_User, time=1, value=2}
         ]}}})
     Vector2D=<1 1>
    )) {module=ScaleSpriteSize001}
  ]
  rendererClasses [&NiagaraSpriteRendererProperties]
  renderers [{rendererClass=&NiagaraSpriteRendererProperties}]
 }
}
summary {
 systemName=NS_CoffeePuff_PickupFX
 emitters [
  {emitterName=Fire_Sprite, rendererClasses [&NiagaraSpriteRendererProperties]}
  {emitterName=Embers_Burst, simTarget=GPUComputeSim, rendererClasses [&NiagaraSpriteRendererProperties]}
 ]
 userVariables []
}
